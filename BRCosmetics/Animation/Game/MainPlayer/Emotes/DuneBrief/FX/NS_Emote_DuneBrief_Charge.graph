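# Niagara system for the blue charge-up effect in the DuneBrief emote, built from two emitters.
# Inward_Sparks ramps up sparks pulled inward from a sphere around a flickering core, with glows, a mesh renderer and a pulsing light.
# Complete_Ring bursts spinning, scaling ring sprites near the half-second mark, as the charge completes.
# Each emitter copies spawn group into VisibilityTag, letting one emitter vary lifetime, size and motion per layer across multiple renderers.
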NiagaraSystem /BRCosmetics/Animation/Game/MainPlayer/Emotes/DuneBrief/FX/NS_Emote_DuneBrief_Charge
dependencies {
 usedDataInterfaces [&NiagaraDataInterfaceCurve, &NiagaraDataInterfaceColorCurve]
 usedDynamicInputs []
 usedModules [&SystemState]
}
emitters {
 Complete_Ring {
  emitterUpdateScript [
   &EmitterState(Life Cycle Mode=::NewEnumerator1, Loop Behavior=::NewEnumerator1, Loop Duration=1, UseLoopDelay=false)
   &SpawnBurst_Instantaneous(Spawn Time=0.49, Use Spawn Probability=false, UseLoopCountLimit=false)
   &SpawnBurst_Instantaneous(Spawn Group=1, Spawn Time=0.49, Use Spawn Probability=false, UseLoopCountLimit=false) {module=SpawnBurst_Instantaneous001}
  ]
  particleSpawnScript [
   /SetVariables(Particles.VisibilityTag=&SpawnGroup()) {isSetParameters=true}
   &InitializeParticle(
    Color=<rgba 0.3 1.4317 10 1>
    Lifetime=&MaskFloat(Mask Source=0, Mask Parameter=?, Mask=0, Masked Float=0.35, Passthrough Float=0.4)
    Mesh Scale Mode=::NewEnumerator3
    Sprite Rotation Angle=&MaskFloat(
     Mask Source=0
     Mask Parameter=?
     Mask=0
     Masked Float=&RandomRangeFloat(Minimum=0, Maximum=1)
     Passthrough Float=45
    )
    Sprite Rotation Mode=::NewEnumerator0
    Sprite Size Mode=::NewEnumerator3
    Uniform Sprite Size=&MaskFloat(Mask Source=0, Mask Parameter=?, Mask=0, Masked Float=82.9769, Passthrough Float=200)
    UsePositionOffset=false
    Write Material Random=true
   )
   &InitialMeshOrientation(
    Mesh Orientation Mode=::NewEnumerator2
    Rotation=&MakeVector(X=&RandomRangeFloat(Minimum=0, Maximum=1), Y=0, Z=0)
    Use Rotation Vector=true
   )
   &CameraOffset(Camera Offset Amount=50)
  ]
  particleUpdateScript [
   &ParticleState()
   &UpdateMeshOrientation(Rotation Rate=&FloatFromCurve(
     FloatCurve=DataInterface() {propertyValues {Curve {keys [
         {arriveTangent=-3.0901, interpMode=RCIM_Cubic, leaveTangent=-3.0901, tangentMode=RCTM_User, time=0, value=1}
         {
          arriveTangent=-0.0612
          arriveTangentWeight=0.5956
          interpMode=RCIM_Cubic
          leaveTangent=-0.0612
          leaveTangentWeight=0.5956
          tangentMode=RCTM_User
          tangentWeightMode=RCTWM_WeightedBoth
          time=0.8
          value=0
         }
        ]}}}
     Scale Curve=-2
    ))
   &ScaleSpriteSize(
    Scale Sprite Size Mode=::NewEnumerator0
    Uniform Scale Factor=&MaskFloat(
     Mask Source=0
     Mask Parameter=?
     Mask=0
     Masked Float=&FloatFromCurve(FloatCurve=DataInterface() {propertyValues {Curve {keys [
          {arriveTangent=1.5335, interpMode=RCIM_Cubic, leaveTangent=1.5335, leaveTangentWeight=0.3333, tangentMode=RCTM_User, time=0, value=0.4}
          {
           arriveTangent=-0.0447
           arriveTangentWeight=0.6537
           interpMode=RCIM_Cubic
           leaveTangent=-0.0447
           leaveTangentWeight=0.6537
           tangentMode=RCTM_User
           tangentWeightMode=RCTWM_WeightedBoth
           time=0.9
           value=1.5
          }
         ]}}})
     Passthrough Float=&FloatFromCurve(FloatCurve=DataInterface() {propertyValues {Curve {
         defaultValue=1
         keys [
          {interpMode=RCIM_Cubic, tangentMode=RCTM_User, time=0, value=0}
          {arriveTangentWeight=0.0333, interpMode=RCIM_Cubic, tangentMode=RCTM_User, time=0.3, value=1}
          <at 1 0>
         ]
        }}})
    )
   )
   &DynamicMaterialParameters(
    Index 0 Param 1=&FloatFromCurve(FloatCurve=DataInterface() {propertyValues {Curve {keys [
         {arriveTangent=-0.016, interpMode=RCIM_Cubic, leaveTangent=-0.016, tangentMode=RCTM_User, time=0.1, value=0.5}
         {arriveTangent=-0.3283, interpMode=RCIM_Cubic, leaveTangent=-0.3283, tangentMode=RCTM_User, time=1, value=-1}
        ]}}})
    Index 0 Param 4=&FloatFromCurve(FloatCurve=DataInterface() {propertyValues {Curve {
        defaultValue=0.9
        keys [
         {interpMode=RCIM_Cubic, tangentMode=RCTM_User, time=0.5, value=0.8}
         {arriveTangent=-1.0158, interpMode=RCIM_Cubic, leaveTangent=-1.0158, tangentMode=RCTM_User, time=1, value=0.1}
        ]
       }}})
    Module.Index0Float0Write=true
    Module.Index0Float1Write=false
    Module.Index0Float2Write=false
    Module.Index0Float3Write=true
   )
   /SetVariables(Particles.Star DynP 00:Vector4f=&MakeVector4(
     X=&FloatFromCurve(FloatCurve=DataInterface() {propertyValues {Curve {keys [
          {interpMode=RCIM_Cubic, tangentMode=RCTM_User, time=0, value=1.5}
          {interpMode=RCIM_Cubic, tangentMode=RCTM_User, time=0.2866, value=3}
          {interpMode=RCIM_Cubic, tangentMode=RCTM_User, time=1, value=1.5}
         ]}}})
     Y=0
     Z=0
     W=0
    )) {isSetParameters=true}
   /SetVariables(Particles.SpriteRotation:NiagaraFloat=&Add_Float(
     A=$Particles.Initial.SpriteRotation
     B=&MaskFloat(
      Mask Source=0
      Mask Parameter=?
      Mask=0
      Masked Float=&FloatFromCurve(
       FloatCurve=DataInterface() {propertyValues {Curve {
          defaultValue=1
          keys [
           {interpMode=RCIM_Cubic, tangentMode=RCTM_User, time=0, value=1}
           {arriveTangentWeight=0.3333, interpMode=RCIM_Cubic, tangentMode=RCTM_User, time=1, value=0}
          ]
         }}}
       Scale Curve=50
      )
      Passthrough Float=&FloatFromCurve(
       FloatCurve=DataInterface() {propertyValues {Curve {
          defaultValue=1
          keys [
           {interpMode=RCIM_Cubic, leaveTangentWeight=0.9495, tangentMode=RCTM_User, time=0, value=0}
           {interpMode=RCIM_Cubic, tangentMode=RCTM_User, time=0.6, value=1}
           <at 1 1>
          ]
         }}}
       Scale Curve=180
      )
     )
    )) {isSetParameters=true}
   /SetVariables(Particles.RingColor:LinearColor=&Multiply_LinearColorByFloat(Linear Color=$Particles.Initial.Color, Float=1)) {isSetParameters=true}
   /SetVariables(Particles.RingGlowColor:LinearColor=&Multiply_LinearColorByFloat(Linear Color=$Particles.RingColor, Float=0.7)) {isSetParameters=true}
  ]
  rendererClasses [&NiagaraSpriteRendererProperties]
  renderers [
   {rendererClass=&NiagaraSpriteRendererProperties}
   {rendererClass=&NiagaraSpriteRendererProperties, rendererIndex=1}
   {rendererClass=&NiagaraSpriteRendererProperties, rendererIndex=2}
  ]
 }
 Inward_Sparks {
  emitterUpdateScript [
   &EmitterState(Life Cycle Mode=::NewEnumerator1, Loop Behavior=::NewEnumerator1, Loop Duration=0.5, UseLoopDelay=false)
   &SpawnRate(
    SpawnRate=&FloatFromCurve(
     FloatCurve=DataInterface() {propertyValues {Curve {keys [
         {arriveTangent=2.3661, interpMode=RCIM_Cubic, leaveTangent=2.3661, tangentMode=RCTM_User, time=0, value=0}
         <at 0.5984 1 RCIM_Cubic>
         <at 0.6429 0 RCIM_Constant>
         {arriveTangent=63.0133, interpMode=RCIM_Cubic, leaveTangent=63.0133, tangentMode=RCTM_User, time=0.8, value=1}
         <at 0.9992 5>
        ]}}}
     CurveIndex=$Emitter.NormalizedLoopAge
     Scale Curve=200
    )
    Use Spawn Probability=false
   )
   &SpawnBurst_Instantaneous(Spawn Group=1, Use Spawn Probability=false, UseLoopCountLimit=false)
   &SpawnRate(SpawnGroup=2, SpawnRate=8, Use Spawn Probability=false) {module=SpawnRate001}
   &SpawnBurst_Instantaneous(Spawn Group=3, Use Spawn Probability=false, UseLoopCountLimit=false) {module=SpawnBurst_Instantaneous001}
  ]
  particleSpawnScript [
   /SetVariables(Particles.VisibilityTag=&SpawnGroup()) {isSetParameters=true}
   &InitializeParticle(
    Color=<rgba 0.7 2.2977 7 1>
    Lifetime=&MergeHelper_FloatSwitch(0=0.25, 1=0.85, 2=0.25, 3=0.8, Index=$Particles.VisibilityTag, Num=::NewEnumerator2)
    Mesh Scale Mode=::NewEnumerator3
    Sprite Size Mode=::NewEnumerator3
    Uniform Sprite Size=3
    UsePositionOffset=false
    Write Material Random=true
   )
   &ShapeLocation(
    MaskSpawnGroup=false
    Override Seed=false
    Sphere Radius=&MaskFloat(Mask Source=0, Mask Parameter=?, Mask=0, Masked Float=100, Passthrough Float=0)
    UseOwnerScale=true
   )
   &AddVelocity(
    Module.Use Falloff Distance=false
    Override Seed=false
    Velocity Mode=::NewEnumerator1
    Velocity Speed=&MaskFloat(Mask Source=0, Mask Parameter=?, Mask=0, Masked Float=-1000, Passthrough Float=0)
   )
   &ApplyInitialForces()
   /SetVariables() {isSetParameters=true}
  ]
  particleUpdateScript [
   &ParticleState()
   &ScaleColor(
    Scale Alpha=&MaskFloat(Mask Source=0, Mask Parameter=?, Mask=1, Masked Float=1, Passthrough Float=1)
    ScaleA=true
    ScaleRGB=false
   )
   &FN_Sprite_Attributes(
    Camera Offset=10
    Color=<rgba 0.6896 1.6635 7 1>
    Create Camera Offset Attribute=true
    Create Color Attribute=true
    Create Sprite Rotation Attribute=true
    Create SpriteSize Attribute=true
    Module Output:ENiagara_AttributeSamplingApplyOutput=::NewEnumerator1
    SpriteRotation:NiagaraFloat=&FloatFromCurve(
     FloatCurve=DataInterface() {propertyValues {Curve {keys [{interpMode=RCIM_Cubic, tangentMode=RCTM_User, time=0, value=1}, {arriveTangent=-1.3184, tangentMode=RCTM_User, time=1, value=0}]}}}
     Scale Curve=-500
    )
    SpriteSize=&Multiply_Vector2D_ByFloat(
     Float=&FloatFromCurve(
      FloatCurve=DataInterface() {propertyValues {Curve {keys [
          {interpMode=RCIM_Cubic, tangentMode=RCTM_User, time=0, value=0}
          {arriveTangent=0.3292, interpMode=RCIM_Cubic, leaveTangent=0.3292, time=0.0393, value=0.1837}
          {arriveTangent=0.1194, interpMode=RCIM_Cubic, leaveTangent=0.1194, tangentMode=RCTM_User, time=0.0511, value=0.0168}
          {arriveTangent=0.696, interpMode=RCIM_Cubic, leaveTangent=0.696, time=0.0706, value=0.1328}
          {arriveTangent=1.5219, interpMode=RCIM_Cubic, leaveTangent=1.5219, time=0.1009, value=0.0515}
          {arriveTangent=-0.039, interpMode=RCIM_Cubic, leaveTangent=-0.039, time=0.1229, value=0.2124}
          {arriveTangent=1.1955, interpMode=RCIM_Cubic, leaveTangent=1.1955, time=0.146, value=0.0497}
          {arriveTangent=1.139, interpMode=RCIM_Cubic, leaveTangent=1.139, time=0.1723, value=0.2714}
          {arriveTangent=1.6019, interpMode=RCIM_Cubic, leaveTangent=1.6019, time=0.2, value=0.1112}
          {arriveTangent=-0.1604, interpMode=RCIM_Cubic, leaveTangent=-0.1604, time=0.2218, value=0.3506}
          {arriveTangent=0.8994, interpMode=RCIM_Cubic, leaveTangent=0.8994, time=0.2474, value=0.1036}
          {arriveTangent=0.5033, interpMode=RCIM_Cubic, leaveTangent=0.5033, time=0.2719, value=0.3957}
          {arriveTangent=-0.4117, interpMode=RCIM_Cubic, leaveTangent=-0.4117, time=0.2928, value=0.1264}
          {arriveTangent=-0.2451, interpMode=RCIM_Cubic, leaveTangent=-0.2451, time=0.3258, value=0.3735}
          {arriveTangent=0.3861, interpMode=RCIM_Cubic, leaveTangent=0.3861, time=0.3549, value=0.1112}
          {arriveTangent=0.2947, interpMode=RCIM_Cubic, leaveTangent=0.2947, time=0.3807, value=0.3947}
          {arriveTangent=-2.6403, interpMode=RCIM_Cubic, leaveTangent=-2.6403, time=0.4066, value=0.1264}
          {arriveTangent=1.2116, interpMode=RCIM_Cubic, leaveTangent=1.2116, tangentMode=RCTM_User, time=0.4371, value=0.2459}
          {arriveTangent=2.9233, interpMode=RCIM_Cubic, leaveTangent=2.9233, tangentMode=RCTM_User, time=0.7035, value=1}
          {arriveTangent=-0.1222, interpMode=RCIM_Cubic, leaveTangent=-0.1222, tangentMode=RCTM_User, time=0.7882, value=2}
          <at 1 0>
         ]}}}
      Scale Curve=&Sine(Normalized Angle=$Emitter.Age, Period=0.1, Scale=0.2, Bias=0.8)
     )
     Vector2D=<50 50>
    )
   )
   /SetVariables(Particles.SpeedScale:NiagaraFloat=&FloatFromCurve(FloatCurve=DataInterface() {propertyValues {Curve {
        defaultValue=1
        keys [
         {
          arriveTangent=-0.0103
          arriveTangentWeight=0.3894
          interpMode=RCIM_Cubic
          leaveTangent=-0.0103
          leaveTangentWeight=0.3894
          tangentMode=RCTM_User
          tangentWeightMode=RCTWM_WeightedBoth
          time=0.2
          value=0
         }
         {arriveTangent=2.4374, interpMode=RCIM_Cubic, leaveTangent=2.4374, tangentMode=RCTM_User, time=1, value=1}
        ]
       }}})) {isSetParameters=true}
   /SetVariables(
    Particles.RoundSparkColor:LinearColor=&MultiplyLinearColor(Linear Color=$Particles.Color, Scale Factor=&MakeVector4(X=1, Y=1, Z=2, W=1))
    Particles.RoundSparkGlowColor:LinearColor=&MultiplyLinearColor(Linear Color=$Particles.Color, Scale Factor=&MakeVector4(X=0.1, Y=0.1, Z=0.2, W=1))
    Particles.RoundSparkSize:Vector2f=&Multiply_Vector2D_ByFloat(
     Float=&FloatFromCurve(FloatCurve=DataInterface() {propertyValues {Curve {keys [
          {arriveTangent=25.1943, interpMode=RCIM_Cubic, leaveTangent=25.1943, tangentMode=RCTM_User, time=0, value=0}
          {interpMode=RCIM_Cubic, tangentMode=RCTM_User, time=0.1, value=1}
          <at 0.4 0>
         ]}}})
     Vector2D=<8 8>
    )
   ) {isSetParameters=true}
   /SetVariables(Particles.CoreCoreSize:Vector2f=&Multiply_Vector2D(A=$Particles.FN_Sprite_Attributes.SpriteSize, B=<-1 1>)) {isSetParameters=true}
   /SetVariables(Particles.RingScale:Vector3f=&Multiply_VectorByFloat(
     Vector=$Particles.Initial.Scale
     Float=&FloatFromCurve(
      FloatCurve=DataInterface() {propertyValues {Curve {
         defaultValue=1
         keys [
          {arriveTangent=0.7845, interpMode=RCIM_Cubic, leaveTangent=0.7845, tangentMode=RCTM_User, time=0, value=0.7}
          {
           arriveTangent=-0.0094
           arriveTangentWeight=0.8467
           interpMode=RCIM_Cubic
           leaveTangent=-0.0094
           leaveTangentWeight=0.8467
           tangentMode=RCTM_User
           tangentWeightMode=RCTWM_WeightedBoth
           time=1
           value=1
          }
         ]
        }}}
      Scale Curve=1.5
     )
    )) {isSetParameters=true}
   /SetVariables(Particles.Ring Dynamic Param:Vector4f=&MakeVector4(
     X=&FloatFromCurve(FloatCurve=DataInterface() {propertyValues {Curve {keys [
          {arriveTangent=-7.2529, interpMode=RCIM_Cubic, leaveTangent=-7.2529, tangentMode=RCTM_User, time=-0.0022, value=1.3}
          {arriveTangent=0.875, interpMode=RCIM_Cubic, leaveTangent=0.875, tangentMode=RCTM_User, time=0.5, value=0}
          {arriveTangent=3.8908, interpMode=RCIM_Cubic, leaveTangent=3.8908, tangentMode=RCTM_User, time=1, value=1.3}
         ]}}})
     Y=&FloatFromCurve(FloatCurve=DataInterface() {propertyValues {Curve {keys [
          {arriveTangent=5.2096, interpMode=RCIM_Cubic, leaveTangent=5.2096, tangentMode=RCTM_User, time=0, value=0.5}
          {arriveTangent=0.2132, interpMode=RCIM_Cubic, leaveTangent=0.2132, tangentMode=RCTM_User, time=1, value=3}
         ]}}})
     Z=&FloatFromCurve(FloatCurve=DataInterface() {propertyValues {Curve {
         defaultValue=1
         keys [{arriveTangent=2.1224, interpMode=RCIM_Cubic, leaveTangent=2.1224, tangentMode=RCTM_User, time=0, value=0}, <at 1 1 RCIM_Cubic>]
        }}})
     W=&FloatFromCurve(
      FloatCurve=DataInterface() {propertyValues {Curve {keys [
          {arriveTangent=2.3563, interpMode=RCIM_Cubic, leaveTangent=2.3563, tangentMode=RCTM_User, time=0, value=0}
          {tangentMode=RCTM_User, time=1, value=1}
         ]}}}
      Scale Curve=2
     )
    )) {isSetParameters=true}
   /SetVariables(
    Particles.DarkGlowColor:LinearColor=<rgba 0 0.0069 0.1 1>
    Particles.DarkGlowSize:Vector2f=&Multiply_Vector2D_ByFloat(Float=3, Vector2D=$Particles.FN_Sprite_Attributes.SpriteSize)
   ) {isSetParameters=true}
   &FN_Light_Attributes(
    Color=$Particles.Color
    Create Color Attribute=true
    Create Exponent Attribute=true
    Create Radius Attribute=true
    Enable Particle Light Enable=true
    Enable Visibility Tag=true
    Light Enabled=true
    Radius=&FloatFromCurve(
     FloatCurve=DataInterface() {propertyValues {Curve {
        defaultValue=1
        keys [
         {interpMode=RCIM_Cubic, tangentMode=RCTM_User, time=0, value=0.5}
         {arriveTangent=0.0497, interpMode=RCIM_Cubic, leaveTangent=0.0497, time=0.1357, value=0.6}
         {arriveTangent=0.5687, interpMode=RCIM_Cubic, leaveTangent=0.5687, time=0.2014, value=0.51}
         <at 0.4121 0.7572 RCIM_Cubic>
         {arriveTangent=0.232, interpMode=RCIM_Cubic, leaveTangent=0.232, time=0.4376, value=0.51}
         <at 0.5922 0.799 RCIM_Cubic>
         {arriveTangent=0.6816, interpMode=RCIM_Cubic, leaveTangent=0.6816, time=0.6382, value=0.51}
         {arriveTangent=-0.0466, interpMode=RCIM_Cubic, leaveTangent=-0.0466, time=0.8138, value=0.95}
         {arriveTangentWeight=0.2847, interpMode=RCIM_Cubic, leaveTangentWeight=0.0153, tangentMode=RCTM_User, time=0.8526, value=0.5}
         {interpMode=RCIM_Cubic, tangentMode=RCTM_User, time=0.943, value=1.1}
         <at 0.9963 0.5 RCIM_Cubic>
        ]
       }}}
     Scale Curve=150
    )
    Scale Color=&FloatFromCurve(
     FloatCurve=DataInterface() {propertyValues {Curve {
        defaultValue=1
        keys [
         {interpMode=RCIM_Cubic, tangentMode=RCTM_User, time=-0.076, value=0.01}
         <at 0.1357 0.2141 RCIM_Cubic>
         {arriveTangent=1.1906, interpMode=RCIM_Cubic, leaveTangent=1.1906, time=0.2014, value=0.01}
         <at 0.4121 0.5432 RCIM_Cubic>
         {arriveTangent=0.5208, interpMode=RCIM_Cubic, leaveTangent=0.5208, time=0.4376, value=0.01}
         <at 0.5922 0.637 RCIM_Cubic>
         {arriveTangent=1.8337, interpMode=RCIM_Cubic, leaveTangent=1.8337, time=0.6382, value=0.01}
         {arriveTangent=1.0696, interpMode=RCIM_Cubic, leaveTangent=1.0696, time=0.8138, value=1.0433}
         {interpMode=RCIM_Cubic, tangentMode=RCTM_User, time=0.8542, value=0.241}
         {arriveTangent=-1.6531, interpMode=RCIM_Cubic, leaveTangent=-1.6531, time=0.943, value=2}
         <at 1 0 RCIM_Cubic>
        ]
       }}}
     Scale Curve=10
    )
    VisibilityTag=1
   )
   &ScaleSpriteSize(
    Scale Factor=&Multiply_Vector2D_ByFloat(
     Float=&FloatFromCurve(FloatCurve=DataInterface() {propertyValues {Curve {keys [
          {arriveTangent=11.7633, interpMode=RCIM_Cubic, leaveTangent=11.7633, tangentMode=RCTM_User, time=0.2021, value=0}
          {arriveTangentWeight=0.0543, interpMode=RCIM_Cubic, leaveTangentWeight=0.3721, tangentMode=RCTM_User, time=0.4, value=1}
          {interpMode=RCIM_Cubic, tangentMode=RCTM_User, time=0.8, value=1}
          <at 1 0>
         ]}}})
     Vector2D=&Lerp_Vector2(A=<1 1>, B=<0 45>, Alpha=$Particles.SpeedScale)
    )
    Scale Sprite Size Mode=::NewEnumerator2
   )
   /SetVariables(
    Particles.GlowColor=&ScaleLinearColorByCurve(
     InitialValue=/Niagara/DynamicInputs/Add/Add_LinearColor(
      A:LinearColor=&Multiply_LinearColorByFloat(
       Linear Color=$Particles.Color
       Float=&FNTimeOfDayScalar(Night Scale=0.05, Morning Scale=0.05, Day Scale=0.05, Evening Scale=0.05)
      )
      B:LinearColor=<rgba 0.0077 0.0834 0.224 0>
     )
     ScaleCurve=DataInterface() {propertyValues {
       AlphaCurve {keys [{interpMode=RCIM_Cubic, tangentMode=RCTM_User, time=0.5, value=1}, <at 0.8175 1>, <at 1 0>]}
       BlueCurve {keys [<at 0 1>,<at 0.9932 1>]}
       GreenCurve {keys [<at 0 1>,<at 0.9932 1>]}
       RedCurve {keys [<at 0 1>,<at 0.9932 1>]}
      }}
     CurveIndex=$Particles.NormalizedAge
    )
    Particles.GlowSize=&Multiply_Vector2D(A=$Particles.SpriteSize, B=&Lerp_Vector2(A=<3 1.5>, B=<15 1.5>, Alpha=$Particles.SpeedScale))
   ) {isSetParameters=true}
   &ScaleVelocity(Velocity Scale=&Multiply_VectorByFloat(Vector=<1 1 1>, Float=$Particles.SpeedScale))
   &SolveForcesAndVelocity(Clamp Velocity=false, Limit Acceleration=false)
  ]
  rendererClasses [
   &NiagaraSpriteRendererProperties
   &NiagaraMeshRendererProperties
   &NiagaraLightRendererProperties
  ]
  renderers [
   {rendererClass=&NiagaraSpriteRendererProperties}
   {rendererClass=&NiagaraSpriteRendererProperties, rendererIndex=1}
   {rendererClass=&NiagaraSpriteRendererProperties, rendererIndex=2}
   {rendererClass=&NiagaraSpriteRendererProperties, rendererIndex=3}
   {rendererClass=&NiagaraSpriteRendererProperties, rendererIndex=4}
   {rendererClass=&NiagaraSpriteRendererProperties, rendererIndex=5}
   {rendererClass=&NiagaraSpriteRendererProperties, rendererIndex=6}
   {rendererClass=&NiagaraMeshRendererProperties, rendererIndex=7}
   {rendererClass=&NiagaraSpriteRendererProperties, rendererIndex=8}
   {rendererClass=&NiagaraLightRendererProperties, rendererIndex=9}
  ]
 }
}
summary {
 systemName=NS_Emote_DuneBrief_Charge
 emitters [
  {
   emitterName=Inward_Sparks
   rendererClasses [
    &NiagaraSpriteRendererProperties
    &NiagaraMeshRendererProperties
    &NiagaraLightRendererProperties
   ]
  }
  {emitterName=Complete_Ring, rendererClasses [&NiagaraSpriteRendererProperties]}
 ]
 userVariables []
}
