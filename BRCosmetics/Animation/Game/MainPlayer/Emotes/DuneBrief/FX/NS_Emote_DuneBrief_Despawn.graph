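# Niagara system for the despawn effect of the DuneBrief emote, built from three emitters.
# Decal flashes one blue, time-of-day-tinted sprite that grows, brightens and fades.
# Equip combines a delayed burst with early-loop rate spawning of small blue meshes, placed along a short vertical path and orbit-offset by curve-driven radii; its skeletal-mesh samplers are disabled.
# EquipTracer adds a bright white mesh with a curve-animated material parameter and alpha fade.
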NiagaraSystem /BRCosmetics/Animation/Game/MainPlayer/Emotes/DuneBrief/FX/NS_Emote_DuneBrief_Despawn
dependencies {
 usedDataInterfaces [
  &NiagaraDataInterfaceArrayFloat
  &NiagaraDataInterfaceArrayInt32
  &NiagaraDataInterfaceArrayFloat2
  &NiagaraDataInterfaceCurve
  &NiagaraDataInterfaceColorCurve
 ]
 usedDynamicInputs [
  &MergeHelper_ColorSwitch
  &MakeCustomIntFromBool
  &MergeHelper_FloatSwitch
 ]
 usedModules [
  &SystemState
  &self:SystemSpawnScript.SystemScriptSource.SystemScriptGraph.NiagaraNodeAssignment_1.SetVariables_#1
 ]
}
emitters {
 Decal {
  emitterUpdateScript [
   &EmitterState()
   &SpawnBurst_Instantaneous(Spawn Count=1, Use Spawn Probability=false) {module=SpawnBurst_Instantaneous001}
  ]
  particleSpawnScript [
   &PartitionParticles()
   &InitializeParticle(
    Color=&FNTimeOfDayColor(
     Night Color=<rgba 0.1121 0.5265 1 1>
     Morning Color=<rgba 0.1121 0.5265 1 1>
     Day Color=<rgba 0.1121 0.5265 1 1>
     Evening Color=<rgba 0.2242 1.0529 2 1>
    )
    Lifetime=0.3
    Position Offset=<0 0 15>
    Position Offset Coordinate Space=::Local
    Sprite Size Mode=::NewEnumerator3
    Uniform Sprite Size=40
    UsePositionOffset=true
    Write Material Random=true
   ) {module=InitializeParticle001}
   &CameraOffset(Camera Offset Amount=20)
  ]
  particleUpdateScript [
   &ParticleState()
   &ScaleSpriteSize(Uniform Curve Sprite Scale=DataInterface() {propertyValues {Curve {keys [{interpMode=RCIM_Cubic, tangentMode=RCTM_User, time=0, value=0.5}, {tangentMode=RCTM_User, time=0.5, value=1}, <at 1 1>]}}})
   &DynamicMaterialParameters(
    Index 0 Param 1=2
    Index 0 Param 2=1
    Module.Param0WriteEnabled=true
    Module.Param1WriteEnabled=true
    Module.Param2WriteEnabled=true
    Module.Param3WriteEnabled=true
   )
   &ScaleColor(
    Scale Alpha=&FloatFromCurve(FloatCurve=DataInterface() {propertyValues {Curve {keys [
         {interpMode=RCIM_Cubic, tangentMode=RCTM_User, time=0, value=0}
         {arriveTangent=-0.0306, interpMode=RCIM_Cubic, leaveTangent=-0.0306, tangentMode=RCTM_User, time=0.2, value=1}
         <at 1 0>
        ]}}})
    Scale RGB=&MakeVectorFromLinearColorRGB(Color=&ColorFromCurve(ColorCurve=DataInterface() {propertyValues {
        AlphaCurve {keys [<at 0 0 RCIM_Cubic>]}
        BlueCurve {keys [<at 0 1 RCIM_Cubic>, {arriveTangent=4.4002, interpMode=RCIM_Cubic, leaveTangent=4.4002, tangentMode=RCTM_User, time=1, value=3}]}
        GreenCurve {keys [<at 0 1 RCIM_Cubic>, {arriveTangent=4.4002, interpMode=RCIM_Cubic, leaveTangent=4.4002, tangentMode=RCTM_User, time=1, value=3}]}
        RedCurve {keys [<at 0 1 RCIM_Cubic>, {arriveTangent=4.4002, interpMode=RCIM_Cubic, leaveTangent=4.4002, tangentMode=RCTM_User, time=1, value=3}]}
       }}))
    ScaleA=true
    ScaleRGB=true
   )
  ]
  rendererClasses [&NiagaraSpriteRendererProperties]
  renderers [{rendererClass=&NiagaraSpriteRendererProperties}]
 }
 Equip {
  emitterUpdateScript [
   &EmitterState(Inactive Response=::NewEnumerator1, Life Cycle Mode=::NewEnumerator1, Loop Behavior=::NewEnumerator1, Loop Duration=0.65, UseLoopDelay=false)
   /Game/Effects/Niagara/Modules/Spawn/FN_Spawn_TriggeredBurst(Cooldown=0.01, Module.Use Burst Count Limit=false, Reset Mode=2, Spawn Count=20, UseLoopCountLimit=false) {enabled=false}
   &SpawnBurst_Instantaneous(Spawn Count=30, Spawn Time=0.25, Use Spawn Probability=false, UseLoopCountLimit=false)
   &SpawnRate(
    Spawn Probability=&MakeCustomFloatFromBool(
     Bool=&SetBoolByFloatComparison(A=$Emitter.NormalizedLoopAge, B=0.5, Comparison Type=::NewEnumerator6)
     True Float=1
     False Float=0
    )
    SpawnRate=200
    Use Spawn Probability=true
   )
  ]
  particleSpawnScript [
   &InitializeParticle(
    Color=&RandomRangeLinearColor(Minimum=<rgba 0.0107 0.076 0.2396 1>, Maximum=<rgba 0.0207 0.1158 0.4635 1>)
    Lifetime=&RandomRangeFloat(Minimum=0.15, Maximum=0.35)
    Mesh Scale Mode=::NewEnumerator3
    Mesh Uniform Scale=&RandomRangeFloat(Minimum=0.2, Maximum=0.75)
    Sprite Size Mode=::NewEnumerator3
    Uniform Sprite Size=&RandomRangeFloat(Minimum=10, Maximum=20)
    UsePositionOffset=false
    Write Material Random=true
   )
   /SetVariables(Particles.ParticleRandom:NiagaraFloat=&RandomRangeFloat(Minimum=0, Maximum=1)) {isSetParameters=true}
   /Game/Effects/Niagara/Modules/Curves/FN_ListCurve_SkeletalMeshSampler(
    Apply Position=true
    Curve Index:NiagaraFloat=&FloatFromCurve(
     FloatCurve=DataInterface() {propertyValues {Curve {keys [
         {interpMode=RCIM_Cubic, tangentMode=RCTM_User, time=0, value=0}
         {interpMode=RCIM_Cubic, tangentMode=RCTM_User, time=0.1284, value=1}
         {interpMode=RCIM_Cubic, tangentMode=RCTM_User, time=0.45, value=0}
        ]}}}
     CurveIndex=$Emitter.Age
    )
    Number of Inputs=1
    Offset_01:Vector3f=&self:ScratchDynamicInput(
     Theta:NiagaraFloat=$Particles.ParticleRandom
     Evaluation Type:NiagaraInt32=0
     Offset:Vector3f=<0 0 0>
     Rotation:Vector3f=&ReturnMeshOrientationAxis(Mesh Orientation Axis=::XAxis, Mesh Orientation Quat=$Particles.MeshOrientation)
     Radius:NiagaraFloat=10
     Random Seed:NiagaraInt32=0
     U Coverage:NiagaraFloat=0
    )
    Offset_02:Vector3f=&self:ScratchDynamicInput(
     Theta:NiagaraFloat=$Particles.ParticleRandom
     Evaluation Type:NiagaraInt32=0
     Offset:Vector3f=<0 0 -25>
     Rotation:Vector3f=&ReturnMeshOrientationAxis(Mesh Orientation Axis=::ZAxis, Mesh Orientation Quat=$Particles.MeshOrientation)
     Radius:NiagaraFloat=20
     Random Seed:NiagaraInt32=0
     U Coverage:NiagaraFloat=0
    )
    Offset_03:Vector3f=&self:ScratchDynamicInput(
     Theta:NiagaraFloat=$Particles.ParticleRandom
     Evaluation Type:NiagaraInt32=0
     Offset:Vector3f=<0 0 -50>
     Rotation:Vector3f=&ReturnMeshOrientationAxis(Mesh Orientation Axis=::XAxis, Mesh Orientation Quat=$Particles.MeshOrientation)
     Radius:NiagaraFloat=10
     Random Seed:NiagaraInt32=0
     U Coverage:NiagaraFloat=0
    )
    Skeletal Mesh:NiagaraDataInterfaceSkeletalMesh=$System.SkeletalMesh
   ) {enabled=false}
   &FN_ListCurve_SkeletalMeshSampler(
    Apply Position=true
    Number of Inputs=0
    Offset_01:Vector3f=&self:ScratchDynamicInput(
     Theta:NiagaraFloat=$Particles.ParticleRandom
     Evaluation Type:NiagaraInt32=0
     Offset:Vector3f=<0 0 0>
     Rotation:Vector3f=<0 90 0>
     Radius:NiagaraFloat=15
     Random Seed:NiagaraInt32=0
     U Coverage:NiagaraFloat=0
    )
    Offset_02:Vector3f=&self:ScratchDynamicInput(
     Theta:NiagaraFloat=$Particles.ParticleRandom
     Evaluation Type:NiagaraInt32=0
     Offset:Vector3f=<0 0 0>
     Rotation:Vector3f=<0 90 0>
     Radius:NiagaraFloat=15
     Random Seed:NiagaraInt32=0
     U Coverage:NiagaraFloat=0
    )
    Skeletal Mesh:NiagaraDataInterfaceSkeletalMesh=$System.SkeletalMesh
   ) {enabled=false, module=FN_ListCurve_SkeletalMeshSampler001}
   /SetVariables(Particles.Position=&AddVectorToPosition(
     Position=$Engine.Emitter.SimulationPosition
     Vector=&self:ScratchDynamicInput(
      Theta:NiagaraFloat=$Particles.ParticleRandom
      Evaluation Type:NiagaraInt32=0
      Offset:Vector3f=<0 0 0>
      Rotation:Vector3f=<0 0 0>
      Radius:NiagaraFloat=15
      Random Seed:NiagaraInt32=0
      U Coverage:NiagaraFloat=0
     )
    )) {enabled=false, isSetParameters=true}
   /Game/Effects/Niagara/Modules/Curves/FN_ListCurve(
    Apply Position=true
    Curve Index:NiagaraFloat=&FloatFromCurve(
     FloatCurve=DataInterface() {propertyValues {Curve {keys [
         {arriveTangent=0.5817, interpMode=RCIM_Cubic, leaveTangent=0.5817, leaveTangentWeight=0.1157, tangentMode=RCTM_User, time=0, value=0}
         {tangentMode=RCTM_User, time=0.3, value=1}
         <at 0.35 0.79>
        ]}}}
     CurveIndex=$Emitter.LoopedAge
    )
    Input Type=2
    Number Inputs Switch=0
    Position Coordinate Space=1
    Vector_01:Vector3f=&RandomRangeVector(Minimum=<0 0 20>, Maximum=<0 0 28>)
    Vector_02:Vector3f=<0 0 -10>
   )
  ]
  particleUpdateScript [
   &SolveForcesAndVelocity(Clamp Velocity=false, Limit Acceleration=false)
   &ParticleState()
   /SetVariables(
    Particles.Position=&self:ScratchDynamicInput(
     Theta:NiagaraFloat=$Particles.ParticleRandom
     Evaluation Type:NiagaraInt32=0
     Offset:Vector3f=Linked() {linkedVariable {name=Particles.Initial.Position, type=NiagaraPosition}}
     Rotation:Vector3f=<0 0 0>
     Radius:NiagaraFloat=&FloatFromCurve(
      FloatCurve=DataInterface() {propertyValues {Curve {keys [
          {arriveTangent=6.8794, interpMode=RCIM_Cubic, leaveTangent=6.8794, tangentMode=RCTM_User, time=0, value=1}
          {arriveTangent=0.8488, interpMode=RCIM_Cubic, leaveTangent=0.8488, tangentMode=RCTM_User, time=0.15, value=1.5}
          {arriveTangent=-1.1354, interpMode=RCIM_Cubic, leaveTangent=-1.1354, tangentMode=RCTM_User, time=0.3, value=1.5}
          {interpMode=RCIM_Cubic, tangentMode=RCTM_User, time=1, value=0.25}
         ]}}}
      Scale Curve=&FloatFromCurve(
       FloatCurve=DataInterface() {propertyValues {Curve {keys [{interpMode=RCIM_Cubic, tangentMode=RCTM_User, time=0, value=1}, {arriveTangent=-3.9037, tangentMode=RCTM_User, time=0.5, value=0}]}}}
       CurveIndex=$Emitter.LoopedAge
       Scale Curve=10
      )
     )
     Random Seed:NiagaraInt32=0
     U Coverage:NiagaraFloat=0
    )
    Particles.SpriteSize=&Multiply_Vector2D_ByFloat(
     Float=&FloatFromCurve(FloatCurve=DataInterface() {propertyValues {Curve {keys [
          {interpMode=RCIM_Cubic, tangentMode=RCTM_User, time=0, value=0}
          {interpMode=RCIM_Cubic, tangentMode=RCTM_User, time=0.25, value=1}
          {arriveTangent=-4.7366, interpMode=RCIM_Cubic, leaveTangent=-4.7366, tangentMode=RCTM_User, time=1, value=0}
         ]}}})
     Vector2D=$Particles.Initial.SpriteSize
    )
   ) {isSetParameters=true}
   &CameraOffset()
   &SpriteFacingAndAlignment(Sprite Facing=<0 1 0>, WriteAlignment=false, WriteFacing=false) {enabled=false}
   &ScaleMeshSize(Scale Factor=&Multiply_VectorByFloat(
     Vector=<1 1 1>
     Float=&FloatFromCurve(FloatCurve=DataInterface() {propertyValues {Curve {keys [
          {arriveTangent=2.9773, interpMode=RCIM_Cubic, leaveTangent=2.9773, tangentMode=RCTM_User, time=0, value=1}
          {tangentMode=RCTM_User, time=1, value=0.25}
         ]}}})
    ))
   &ScaleColor(
    Scale RGB=&Multiply_VectorByFloat(
     Vector=<1 1 1>
     Float=&FloatFromCurve(FloatCurve=DataInterface() {propertyValues {Curve {keys [{arriveTangent=-29.5871, interpMode=RCIM_Cubic, leaveTangent=-29.5871, tangentMode=RCTM_User, time=0, value=10}, <at 0.5 1>]}}})
    )
    ScaleA=true
    ScaleRGB=true
   )
  ]
  rendererClasses [&NiagaraMeshRendererProperties]
  renderers [
   {rendererClass=&NiagaraMeshRendererProperties, rendererIndex=0}
   {rendererClass=&NiagaraMeshRendererProperties, rendererIndex=1}
  ]
 }
 EquipTracer {
  emitterUpdateScript [
   &EmitterState(Inactive Response=::NewEnumerator1, Life Cycle Mode=::NewEnumerator1, Loop Behavior=::NewEnumerator1, Loop Duration=0.85, UseLoopDelay=false)
   &SpawnBurst_Instantaneous(Use Spawn Probability=false, UseLoopCountLimit=false)
  ]
  particleSpawnScript [&InitializeParticle(Color=<rgba 15 15 15 1>, Lifetime=0.3, Mesh Scale Mode=::NewEnumerator3, UsePositionOffset=false, Write Material Random=true)]
  particleUpdateScript [
   &ParticleState()
   &DynamicMaterialParameters(
    Index 0 Param 1=1
    Index 0 Param 2=1
    Index 0 Param 4=1
    Index 1 Param 3=&FloatFromCurve(FloatCurve=DataInterface() {propertyValues {Curve {keys [
         {arriveTangent=0.477, interpMode=RCIM_Cubic, leaveTangent=0.477, tangentMode=RCTM_User, time=0, value=-1}
         {arriveTangent=3.2002, arriveTangentWeight=0.3437, interpMode=RCIM_Cubic, leaveTangent=3.2002, tangentMode=RCTM_User, time=1, value=1}
        ]}}})
    Module.Index0Float0Write=true
    Module.Index0Float1Write=true
    Module.Index0Float2Write=true
    Module.Index0Float3Write=true
    Module.Index1Float0Write=true
    Module.Index1Float1Write=true
    Module.Index1Float2Write=true
    Module.Index1Float3Write=true
    Write Parameter Index 1=::NewEnumerator1
   )
   &ScaleColor(
    Scale Alpha=&FloatFromCurve(
     FloatCurve=DataInterface() {propertyValues {Curve {keys [
         {interpMode=RCIM_Cubic, tangentMode=RCTM_User, time=0, value=1}
         {arriveTangent=-1, interpMode=RCIM_Cubic, leaveTangent=-1, time=0.4847, value=0.8588}
         {arriveTangent=-0.6755, tangentMode=RCTM_User, time=1, value=0}
        ]}}}
     CurveIndex=$Emitter.NormalizedLoopAge
    )
    ScaleA=true
    ScaleRGB=true
   )
   &CameraOffset(Camera Offset Amount=2)
  ]
  rendererClasses [&NiagaraMeshRendererProperties]
  renderers [{rendererClass=&NiagaraMeshRendererProperties, rendererIndex=0}]
 }
}
summary {
 systemName=NS_Emote_DuneBrief_Despawn
 emitters [
  {emitterName=Equip, rendererClasses [&NiagaraMeshRendererProperties]}
  {emitterName=Decal, rendererClasses [&NiagaraSpriteRendererProperties]}
  {emitterName=EquipTracer, rendererClasses [&NiagaraMeshRendererProperties]}
 ]
 userVariables [User.VariantToon {type=NiagaraFloat, defaultValue=1, description=""}]
}
