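# Niagara effect for the DuneBrief emote: a blue energy explosion followed by a rain of glowing sparks.
# Explosion uses staggered spawn groups as visibility tags to layer sprites, meshes, a light and decal from one emitter.
# Energy_Rain launches sparks from a widening sphere at System.ExplosionPosition, splitting bright and dim by mass, with gravity, collision and particle lights.
# Sparks shows no renderers; User.ExplosionDelay is exposed, but its use is not shown.
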NiagaraSystem /BRCosmetics/Animation/Game/MainPlayer/Emotes/DuneBrief/FX/NS_Emote_DuneBrief_EnergyBurstRain
dependencies {
 usedDataInterfaces [&NiagaraDataInterfaceCurve, &NiagaraDataInterfaceColorCurve, &NiagaraDataInterfaceVectorCurve]
 usedDynamicInputs []
 usedModules [
  &self:SystemSpawnScript.SystemScriptSource.SystemScriptGraph.NiagaraNodeAssignment_1.SetVariables_#1
  &self:SystemSpawnScript.SystemScriptSource.SystemScriptGraph.NiagaraNodeAssignment_0.SetVariables_#2
  &SystemState
 ]
}
emitters {
 Energy_Rain {
  emitterUpdateScript [
   &EmitterState(
    Life Cycle Mode=::NewEnumerator1
    Loop Behavior=::NewEnumerator1
    Loop Delay=0.1
    Loop Duration=0.55
    Scalability Mode=::NewEnumerator1
    UseLoopDelay=true
   )
   &SpawnRate(
    SpawnRate=&FloatFromCurve(
     FloatCurve=DataInterface() {propertyValues {Curve {keys [{interpMode=RCIM_Cubic, tangentMode=RCTM_User, time=0.3, value=1}, <at 1 0.5>]}}}
     CurveIndex=$Emitter.NormalizedLoopAge
     Scale Curve=90
    )
    Use Spawn Probability=false
   )
   &SpawnBurst_Instantaneous(Spawn Count=11, Use Spawn Probability=false, UseLoopCountLimit=false)
  ]
  particleSpawnScript [
   /SetVariables(Particles.VisibilityTag=&SpawnGroup()) {isSetParameters=true}
   &InitializeParticle(
    Color Mode=::NewEnumerator0
    Lifetime=&MaskFloat(
     Mask Source=0
     Mask Parameter=?
     Mask=0
     Masked Float=&RandomRangeFloat(Minimum=1.3, Maximum=1.5)
     Passthrough Float=2.5
    )
    Mass Max:NiagaraFloat=&MakeFloatFromVector2d(Vector 2D=$System.SparkMassRange, Channel=::NewEnumerator1)
    Mass Min:NiagaraFloat=&MakeFloatFromVector2d(Vector 2D=$System.SparkMassRange)
    Mass Mode=::NewEnumerator3
    Sprite Size=&Multiply_Vector2D_ByFloat(Float=1.5, Vector2D=<1 1>)
    Sprite Size Mode=::NewEnumerator4
    UsePositionOffset=false
    Write Material Random=true
   )
   &ShapeLocation(
    MaskSpawnGroup=false
    Non Uniform Scale=&MakeVector(
     X=1
     Y=1
     Z=&FloatFromCurve(
      FloatCurve=DataInterface() {propertyValues {Curve {keys [{arriveTangent=-4.2638, interpMode=RCIM_Cubic, leaveTangent=-4.2638, tangentMode=RCTM_User, time=0, value=1}, <at 0.3 0>]}}}
      CurveIndex=$Emitter.Age
     )
    )
    Offset:NiagaraPosition=$System.ExplosionPosition
    Offset Mode=::NewEnumerator0
    Override Seed=false
    Shape Origin=$Engine.Emitter.SimulationPosition
    Sphere Radius=&FloatFromCurve(
     FloatCurve=DataInterface() {propertyValues {Curve {defaultValue=1, keys [{tangentMode=RCTM_User, time=0, value=0.05}, {tangentMode=RCTM_User, time=1, value=1}]}}}
     CurveIndex=$Emitter.NormalizedLoopAge
     Scale Curve=90
    )
    Sphere Surface Distribution=1
    UseOwnerScale=true
   )
   &AddVelocity(
    Module.Use Falloff Distance=false
    Origin Offset=&AddVectorToPosition(Position=$System.ExplosionPosition, Vector=&Lerp_Vector(A=<0 0 1>, B=<0 0 50>, Alpha=$Emitter.NormalizedLoopAge))
    Origin Offset Coordinate Space=::World
    Override Seed=false
    Velocity Mode=::NewEnumerator1
    Velocity Speed=&Multiply_Float(
     A=&RandomRangeFloat(Minimum=100, Maximum=150)
     B=&FloatFromCurve(
      FloatCurve=DataInterface() {propertyValues {Curve {keys [{arriveTangent=-0.4955, interpMode=RCIM_Cubic, leaveTangent=-0.4955, tangentMode=RCTM_User, time=0, value=2}, <at 1 1.5>]}}}
      CurveIndex=$Emitter.NormalizedLoopAge
     )
    )
   )
   &ApplyInitialForces()
   /SetVariables(Particles.Bright:NiagaraBool=&SetBoolByFloatComparison(
     A=$Particles.Mass
     B=&Add_Float(
      A=&MakeFloatFromVector2d(Vector 2D=$System.SparkMassRange)
      B=&Multiply_Float(
       A=&Subtract_Float(
        A=&MakeFloatFromVector2d(Vector 2D=$System.SparkMassRange, Channel=::NewEnumerator1)
        B=&MakeFloatFromVector2d(Vector 2D=$System.SparkMassRange)
       )
       B=0.5
      )
     )
    )) {isSetParameters=true}
   /SetVariables(Particles.Color=&MakeCustomLinearColorFromBool(Bool=$Particles.Bright, True Color=<rgba 0.2 0.65 2 1>, False Color=<rgba 0.0326 0.1549 0.549 0.2>)) {isSetParameters=true}
  ]
  particleUpdateScript [
   &ParticleState()
   &ScaleSpriteSize(Uniform Curve Sprite Scale=DataInterface() {propertyValues {Curve {keys [{tangentMode=RCTM_User, time=0, value=0.5}, {tangentMode=RCTM_User, time=0.3, value=1}]}}})
   &ScaleColor(
    Scale Alpha=&FloatFromCurve(FloatCurve=DataInterface() {propertyValues {Curve {keys [{interpMode=RCIM_Cubic, tangentMode=RCTM_User, time=0, value=1}, {arriveTangent=-3.5158, tangentMode=RCTM_User, time=1, value=0}]}}})
    Scale RGB=&VectorFromFloat(Value=&FloatFromCurve(
      FloatCurve=DataInterface() {propertyValues {Curve {keys [{arriveTangent=-11.5631, interpMode=RCIM_Cubic, leaveTangent=-11.5631, tangentMode=RCTM_User, time=0, value=50}, <at 0.6 1>]}}}
      CurveIndex=$Emitter.Age
     ))
    ScaleA=true
    ScaleRGB=true
   )
   &AccelerationForce(Acceleration=&Lerp_Vector(A=<0 0 -490>, B=<0 0 -600>, Alpha=$Particles.Mass))
   &Collision(
    Age Colliding Particles=true
    Analytical Collision Plane Position 1:NiagaraPosition=<0 0 1>
    Analytical Collision Plane Position 2:NiagaraPosition=<0 0 0>
    Analytical Collision Plane Space:ENiagaraCoordinateSpace=::Local
    CPU Collision Type=::NewEnumerator3
    Kill On Collision=2
    KillOnCollision=true
    Module.Control Rest State Via Surface Normal=false
    Module.EnableMaxCollisionCount=false
    Module.Randomize Collision Normal Vector=false
    Restitution=&RandomRangeFloat(Minimum=0.1, Maximum=0.4)
   ) {module=Collision001}
   &Drag(Use Linear Drag=true, Use Rotational Drag=false)
   &SolveForcesAndVelocity(Clamp Velocity=false, Limit Acceleration=false)
   &ScaleSpriteSizeBySpeed(Max Scale Factor=<2 10>, Sample Scale Factor By Curve=true)
   /SetVariables(
    Particles.GlowColor=&ScaleLinearColorByCurve(
     InitialValue=/Niagara/DynamicInputs/Add/Add_LinearColor(
      A:LinearColor=&Multiply_LinearColorByFloat(
       Linear Color=$Particles.Color
       Float=&FNTimeOfDayScalar(Night Scale=0.05, Morning Scale=0.05, Day Scale=0.05, Evening Scale=0.05)
      )
      B:LinearColor=&MakeCustomLinearColorFromBool(Bool=$Particles.Bright, True Color=<rgba 0.0276 0.1605 0.4073 0>, False Color=<rgba 0.0516 0.0933 0.6781 0>)
     )
     ScaleCurve=DataInterface() {propertyValues {
       AlphaCurve {keys [{interpMode=RCIM_Cubic, tangentMode=RCTM_User, time=0.5, value=1}, <at 0.8175 1>, <at 1 0>]}
       BlueCurve {keys [<at 0 1>,<at 0.9932 1>]}
       GreenCurve {keys [<at 0 1>,<at 0.9932 1>]}
       RedCurve {keys [<at 0 1>,<at 0.9932 1>]}
      }}
     CurveIndex=$Particles.NormalizedAge
    )
    Particles.GlowSize=&Multiply_Vector2D(A=&Multiply_Vector2D_ByFloat(Float=$Particles.Mass, Vector2D=$Particles.SpriteSize), B=<10 10>)
   ) {isSetParameters=true}
   /SetVariables(
    Particles.StreakColor:LinearColor=&MultiplyLinearColor(Linear Color=$System.StreakColor, Scale Factor=&MakeVector4_FromLinearColor(LinearColor=$Particles.Color))
    Particles.StreakSize:Vector2f=&Multiply_Vector2D(
     A=$Particles.SpriteSize
     B=&MakeCustomVector2DFromBool(Bool=$Particles.Bright, True Vector2D=<1 7>, False Vector2D=<1 7>)
    )
   ) {isSetParameters=true}
   &FN_Light_Attributes(
    Color=$Particles.Color
    Create Color Attribute=true
    Create Exponent Attribute=true
    Create Radius Attribute=true
    Enable Particle Light Enable=true
    Enable Visibility Tag=true
    Exponent=3
    Light Enabled=true
    Radius=9
    VisibilityTag=1
   )
  ]
  rendererClasses [&NiagaraSpriteRendererProperties, &NiagaraLightRendererProperties]
  renderers [
   {rendererClass=&NiagaraSpriteRendererProperties}
   {rendererClass=&NiagaraSpriteRendererProperties, rendererIndex=1}
   {rendererClass=&NiagaraSpriteRendererProperties, rendererIndex=2}
   {rendererClass=&NiagaraLightRendererProperties, rendererIndex=3}
  ]
 }
 Explosion {
  emitterUpdateScript [
   &EmitterState()
   &SpawnBurst_Instantaneous(Spawn Time=0.1, Use Spawn Probability=false, UseLoopCountLimit=false)
   &SpawnBurst_Instantaneous(Spawn Group=1, Spawn Time=0.12, Use Spawn Probability=false, UseLoopCountLimit=false) {module=SpawnBurst_Instantaneous001}
   &SpawnBurst_Instantaneous(Spawn Count=2, Spawn Group=2, Use Spawn Probability=false, UseLoopCountLimit=false) {module=SpawnBurst_Instantaneous002}
   &SpawnBurst_Instantaneous(Spawn Group=3, Use Spawn Probability=false, UseLoopCountLimit=false) {module=SpawnBurst_Instantaneous003}
   &SpawnBurst_Instantaneous(Spawn Group=4, Use Spawn Probability=false, UseLoopCountLimit=false) {module=SpawnBurst_Instantaneous005}
   &SpawnBurst_Instantaneous(Spawn Count=3, Spawn Group=5, Spawn Time=0.15, Use Spawn Probability=false, UseLoopCountLimit=false) {module=SpawnBurst_Instantaneous004}
   &SpawnBurst_Instantaneous(Spawn Group=6, Use Spawn Probability=false, UseLoopCountLimit=false) {module=SpawnBurst_Instantaneous006}
   &SpawnBurst_Instantaneous(Spawn Group=6, Use Spawn Probability=false, UseLoopCountLimit=false) {module=SpawnBurst_Instantaneous007}
  ]
  particleSpawnScript [
   /SetVariables(Particles.VisibilityTag=&SpawnGroup()) {isSetParameters=true}
   &InitializeParticle(
    Color=<rgba 0.91 2.925 13 1>
    Lifetime=&MergeHelper_FloatSwitch(0=1, 1=0.4, 2=0.35, 3=0.9, 4=1.2, 5=0.6, 6=2, 7=3, Index=&SpawnGroup(), Num=::NewEnumerator6)
    Mesh Scale Mode=::NewEnumerator3
    Mesh Uniform Scale=&MergeHelper_FloatSwitch(
     0=4.3542
     1=4.3542
     2=4.3542
     3=1
     4=2
     5=&Lerp_Float(A=0.55, B=0.8, Alpha=&MakeFloatFromInt(Int=&ReturnExecIndex()))
     Index=&SpawnGroup()
     Num=::NewEnumerator4
    )
    Position Offset=&TransformVector(Vector=Linked() {linkedVariable {name=System.ExplosionPosition, type=NiagaraPosition}}, Destination Space=::World)
    Sprite Size=/Niagara/DynamicInputs/Helpers/MaskVector2d(
     Mask Source:NiagaraInt32=0
     Mask Parameter:NiagaraInt32=?
     Mask:NiagaraInt32=0
     Masked Vector2D:Vector2f=<400 100>
     Passthrough Vector2D:Vector2f=<300 500>
    )
    Sprite Size Mode=::NewEnumerator4
    UsePositionOffset=true
    Write Material Random=true
   )
   &InitialMeshOrientation(
    Mesh Orientation Mode=::NewEnumerator2
    Rotation=&MakeVector(
     X=&MaskFloat(
      Mask Source=0
      Mask Parameter=?
      Mask=5
      Masked Float=&RandomRangeFloat(Minimum=0, Maximum=1)
      Passthrough Float=0
     )
     Y=0
     Z=0
    )
    Use Rotation Vector=true
   )
   /SetVariables(Particles.RaysRotation:NiagaraFloat=&RandomRangeFloat(Minimum=-30, Maximum=30)) {isSetParameters=true}
   /SetVariables(Particles.InitCameraOffset=100) {isSetParameters=true}
   /SetVariables(Particles.InitStarRot1=-45, Particles.InitStarRot2=45) {isSetParameters=true}
   &CameraOffset(Camera Offset Amount=&MergeHelper_FloatSwitch(0=20, 1=20, 2=100, 3=20, 4=20, 5=20, 6=20, Index=$Particles.VisibilityTag, Num=::NewEnumerator5)) {enabled=false}
  ]
  particleUpdateScript [
   &ParticleState()
   &ScaleMeshSize(Scale Factor=&VectorFromCurve(
     VectorCurve=DataInterface() {propertyValues {
       XCurve {keys [<at 0 1>,<at 1 1>]}
       YCurve {keys [{arriveTangent=1.3327, interpMode=RCIM_Cubic, leaveTangent=1.3327, tangentMode=RCTM_User, time=0, value=0}, <at 1 1>]}
       ZCurve {keys [{arriveTangent=1.3327, interpMode=RCIM_Cubic, leaveTangent=1.3327, tangentMode=RCTM_User, time=0, value=0}, <at 1 1>]}
      }}
     Scale Curve=<0.1 1 1>
    ))
   &ScaleSpriteSize(Uniform Curve Sprite Scale=DataInterface() {propertyValues {Curve {
       defaultValue=0
       keys [
        {arriveTangent=17.7165, interpMode=RCIM_Cubic, leaveTangent=17.7165, tangentMode=RCTM_User, time=0, value=1.5}
        {arriveTangent=-14.1365, interpMode=RCIM_Cubic, leaveTangent=-14.1365, tangentMode=RCTM_User, time=0.03, value=2}
        {arriveTangent=-1.163, interpMode=RCIM_Cubic, leaveTangent=-1.163, tangentMode=RCTM_User, time=0.2725, value=0.3175}
        <at 1 0.01>
       ]
      }}})
   &ScaleColor(
    Scale Alpha=&FloatFromCurve(FloatCurve=DataInterface() {propertyValues {Curve {keys [{interpMode=RCIM_Cubic, tangentMode=RCTM_User, time=0, value=1}, <at 1 0>]}}})
    Scale RGB=&Multiply_VectorByFloat(
     Vector=<1 1 1>
     Float=&FloatFromCurve(FloatCurve=DataInterface() {propertyValues {Curve {keys [{interpMode=RCIM_Cubic, tangentMode=RCTM_User, time=0, value=1}, <at 1 1>]}}})
    )
    ScaleA=true
    ScaleRGB=true
   )
   /SetVariables(Particles.StarDarkSize:Vector2f=&Multiply_Vector2D_ByFloat(
     Float=&FloatFromCurve(FloatCurve=DataInterface() {propertyValues {Curve {keys [{interpMode=RCIM_Cubic, tangentMode=RCTM_User, time=0, value=1.1}, <at 1 2>]}}})
     Vector2D=$Particles.SpriteSize
    )) {isSetParameters=true}
   /SetVariables(
    Particles.DarkGlowColor:LinearColor=&Multiply_LinearColorByFloat(
     Linear Color=<rgba 0.015 0.0417 0.15 1>
     Float=&FloatFromCurve(
      FloatCurve=DataInterface() {propertyValues {Curve {keys [
          {arriveTangent=-19.1949, interpMode=RCIM_Cubic, leaveTangent=-19.1949, tangentMode=RCTM_User, time=0, value=5}
          {interpMode=RCIM_Cubic, tangentMode=RCTM_User, time=0.1, value=1}
          <at 1 0>
         ]}}}
      Scale Curve=0.7
     )
    )
    Particles.DarkGlowDynP0:Vector4f=&MakeVector4(X=2, Y=1, Z=1, W=1)
    Particles.DarkGlowRoundSize:Vector2f=&Multiply_Vector2D_ByFloat(Float=0.8912, Vector2D=<247.6003 202.4127>)
    Particles.DarkGlowSize:Vector2f=&Multiply_Vector2D_ByFloat(Float=0.7, Vector2D=$Particles.SpriteSize)
   ) {isSetParameters=true}
   /SetVariables(
    Particles.GlowColor=&Multiply_LinearColorByFloat(Linear Color=$Particles.Color, Float=0.5)
    Particles.GlowSize=&Multiply_Vector2D_ByFloat(Float=0.6, Vector2D=$Particles.DarkGlowRoundSize)
   ) {isSetParameters=true}
   /SetVariables(
    Particles.RaysColor:LinearColor=&MultiplyLinearColor(
     Linear Color=$Particles.Initial.Color
     Scale Factor=/Niagara/DynamicInputs/Helpers/MaskVector4(
      Mask Source:NiagaraInt32=0
      Mask Parameter:NiagaraInt32=?
      Mask:NiagaraInt32=1
      Masked Vector4:Vector4f=&MakeVector4(
       X=1
       Y=1
       Z=1
       W=&FloatFromCurve(FloatCurve=DataInterface() {propertyValues {Curve {keys [{arriveTangent=-2.3141, interpMode=RCIM_Cubic, leaveTangent=-2.3141, tangentMode=RCTM_User, time=0, value=1}, <at 1 0>]}}})
      )
      Passthrough Vector4:Vector4f=&Vector4FromCurve(Vector4Curve=DataInterface() {propertyValues {
         AlphaCurve {keys [<at 0 1>,<at 1 0>]}
         BlueCurve {keys [<at 0 1>,<at 1 1>]}
         GreenCurve {keys [<at 0 1>,<at 1 1>]}
         RedCurve {keys [<at 0 1>,<at 1 1>]}
        }})
     )
    )
    Particles.RaysScale:Vector2f=&Multiply_Vector2D_ByFloat(
     Float=&MaskFloat(
      Mask Source=0
      Mask Parameter=?
      Mask=1
      Masked Float=&FloatFromCurve(FloatCurve=DataInterface() {propertyValues {Curve {
          defaultValue=1
          keys [
           {arriveTangent=2.2389, interpMode=RCIM_Cubic, leaveTangent=2.2389, tangentMode=RCTM_User, time=0, value=0.5}
           {arriveTangent=0.288, tangentMode=RCTM_User, time=1, value=2}
          ]
         }}})
      Passthrough Float=&FloatFromCurve(
       FloatCurve=DataInterface() {propertyValues {Curve {
          defaultValue=1
          keys [
           {arriveTangent=-1.3558, interpMode=RCIM_Cubic, leaveTangent=-1.3558, tangentMode=RCTM_User, time=0, value=1}
           {tangentMode=RCTM_User, time=1, value=0.5}
          ]
         }}}
       Scale Curve=2
      )
     )
     Vector2D=<300 300>
    )
   ) {isSetParameters=true}
   /SetVariables(Particles.ExpRingColor:LinearColor=&Multiply_LinearColorByFloat(
     Linear Color=$Particles.Initial.Color
     Float=&FloatFromCurve(
      FloatCurve=DataInterface() {propertyValues {Curve {keys [{arriveTangent=-0.186, interpMode=RCIM_Cubic, leaveTangent=-0.186, tangentMode=RCTM_User, time=0, value=1}, <at 1 0>]}}}
      Scale Curve=3
     )
    )) {isSetParameters=true}
   /SetVariables(Particles.ExpRingGlowColor:LinearColor=&MultiplyLinearColor(Linear Color=$Particles.ExpRingColor, Scale Factor=<1 1 2 1>)) {isSetParameters=true}
   /SetVariables(
    Particles.InitRingCamColor:LinearColor=&Multiply_LinearColorByFloat(
     Linear Color=$Particles.Initial.Color
     Float=&FloatFromCurve(FloatCurve=DataInterface() {propertyValues {Curve {keys [{arriveTangent=-2.9365, interpMode=RCIM_Cubic, leaveTangent=-2.9365, tangentMode=RCTM_User, time=0, value=1}, <at 0.8 0>]}}})
    )
    Particles.InitRingScale:Vector3f=&Multiply_VectorByFloat(
     Vector=<1 1 1>
     Float=&FloatFromCurve(FloatCurve=DataInterface() {propertyValues {Curve {
         defaultValue=1
         keys [
          {arriveTangent=17.8217, interpMode=RCIM_Cubic, leaveTangent=17.8217, tangentMode=RCTM_User, time=0, value=0}
          {tangentMode=RCTM_User, time=1, value=13}
         ]
        }}})
    )
   ) {isSetParameters=true}
   /SetVariables(
    Particles.InitGlowColor:LinearColor=&MultiplyLinearColor(
     Linear Color=<rgba 0.2974 0.8317 3 1>
     Scale Factor=&Vector4FromCurve(Vector4Curve=DataInterface() {propertyValues {
        AlphaCurve {keys [
          <at 0 1 RCIM_Cubic>
          <at 0.1387 3 RCIM_Constant>
          {arriveTangent=-3.483, interpMode=RCIM_Cubic, leaveTangent=-3.483, time=0.3849, value=0.7}
          <at 1 0>
         ]}
        BlueCurve {keys [<at 0 0.7 RCIM_Constant>,<at 0.2757 0.07 RCIM_Constant>,<at 0.4521 1.2983>,<at 1 0.3865>]}
        GreenCurve {keys [<at 0 0.7 RCIM_Constant>,<at 0.2757 0.03 RCIM_Constant>,<at 0.4521 1.2983>,<at 1 0.3865>]}
        RedCurve {keys [<at 0 0.7 RCIM_Constant>,<at 0.2757 0.03 RCIM_Constant>,<at 0.4521 1.2983>,<at 1 0.3865>]}
       }})
    )
    Particles.InitGlowSize:Vector2f=&Multiply_Vector2D_ByFloat(
     Float=&FloatFromCurve(FloatCurve=DataInterface() {propertyValues {Curve {keys [{arriveTangent=-2.9609, interpMode=RCIM_Cubic, leaveTangent=-2.9609, tangentMode=RCTM_User, time=0, value=1}, <at 1 0>]}}})
     Vector2D=<500 500>
    )
   ) {isSetParameters=true}
   /SetVariables(Particles.InitStarColor:LinearColor=&MultiplyLinearColor(
     Linear Color=$Particles.Initial.Color
     Scale Factor=&Vector4FromCurve(Vector4Curve=DataInterface() {propertyValues {
        AlphaCurve {keys [<at 0 1>,<at 1 0>]}
        BlueCurve {keys [<at 0.2845 0>,<at 0.2848 3>,<at 1 3>]}
        GreenCurve {keys [<at 0.2845 0>,<at 0.2848 3>,<at 1 3>]}
        RedCurve {keys [<at 0.2845 0>,<at 0.2848 3>,<at 1 3>]}
       }})
    )) {isSetParameters=true}
   /SetVariables(
    Particles.DarkShockColor:LinearColor=&MaskLinearColor(
     Mask Source=0
     Mask Parameter=?
     Mask=3
     Masked Linear Color=&ColorFromCurve(ColorCurve=DataInterface() {propertyValues {
        AlphaCurve {keys [<at 0.3128 0.5>,<at 1 0>]}
        BlueCurve {keys [<at 0.2002 0.9406>,<at 0.6805 0.9406>]}
        GreenCurve {keys [<at 0.2002 0.3843>,<at 0.6805 0.1841>]}
        RedCurve {keys [<at 0.2002 0.2871>,<at 0.6805 0.0519>]}
       }})
     Passthrough Linear Color=&ColorFromCurve(ColorCurve=DataInterface() {propertyValues {
        AlphaCurve {keys [<at 0.3128 2>,<at 1 0>]}
        BlueCurve {keys [<at 0.8574 0.1781>,<at 0.91 0.0625>]}
        GreenCurve {keys [<at 0.8574 0.0397>,<at 0.91 0.0242>]}
        RedCurve {keys [<at 0.8574 0.0054>,<at 0.91 0.0192>]}
       }})
    )
    Particles.DarkShockDynP00:Vector4f=&MaskVector4(
     Mask Source:NiagaraInt32=0
     Mask Parameter:NiagaraInt32=?
     Mask:NiagaraInt32=3
     Masked Vector4:Vector4f=&MakeVector4(
      X=&FloatFromCurve(FloatCurve=DataInterface() {propertyValues {Curve {keys [
           {arriveTangent=0.0769, interpMode=RCIM_Cubic, time=0, value=-0.3577}
           {arriveTangent=0.3577, interpMode=RCIM_Cubic, leaveTangent=0.3577, time=0.3, value=-0.1}
           {interpMode=RCIM_Cubic, leaveTangent=2.9562, time=1, value=0}
          ]}}})
      Y=&FloatFromCurve(FloatCurve=DataInterface() {propertyValues {Curve {keys [{arriveTangent=-0.5818, interpMode=RCIM_Cubic, leaveTangent=-0.5818, tangentMode=RCTM_User, time=0, value=0}, <at 1 2>]}}})
      Z=&FloatFromCurve(FloatCurve=DataInterface() {propertyValues {Curve {keys [<at 0 0>,<at 0.2 1>,<at 1 0.5>]}}})
      W=0
     )
     Passthrough Vector4:Vector4f=&MakeVector4(
      X=&FloatFromCurve(FloatCurve=DataInterface() {propertyValues {Curve {keys [{arriveTangent=-2.6331, interpMode=RCIM_Cubic, leaveTangent=-2.6331, tangentMode=RCTM_User, time=0, value=0.6}, <at 1 -0.5 RCIM_Cubic>]}}})
      Y=&FloatFromCurve(FloatCurve=DataInterface() {propertyValues {Curve {keys [{arriveTangent=-0.5818, interpMode=RCIM_Cubic, leaveTangent=-0.5818, tangentMode=RCTM_User, time=0, value=1}, <at 1 0.9>]}}})
      Z=&FloatFromCurve(FloatCurve=DataInterface() {propertyValues {Curve {keys [<at 0 0>,<at 0.15 1>,<at 1 1.2>]}}})
      W=0
     )
    )
    Particles.DarkShockScale:Vector3f=&Multiply_Vector(
     A=$Particles.Initial.Scale
     B=&MaskVector(
      Mask Source=0
      Mask Parameter=?
      Mask=3
      Masked Vector=&VectorFromFloat(Value=&FloatFromCurve(FloatCurve=DataInterface() {propertyValues {Curve {keys [
            {arriveTangent=4.8906, interpMode=RCIM_Cubic, leaveTangent=4.8906, tangentMode=RCTM_User, time=0, value=0.5}
            {arriveTangent=2.5, interpMode=RCIM_Cubic, leaveTangent=2.5, time=0.4279, value=1.9}
            {arriveTangent=0.5979, interpMode=RCIM_Cubic, leaveTangent=0.5979, tangentMode=RCTM_User, time=1, value=3}
           ]}}}))
      Passthrough Vector=&VectorFromFloat(Value=&FloatFromCurve(FloatCurve=DataInterface() {propertyValues {Curve {keys [<at 0 0>,<at 0.1 1>,<at 1 2>]}}}, Scale Curve=0.7))
     )
    )
   ) {isSetParameters=true}
   /SetVariables(
    Particles.BurstColor:LinearColor=&Multiply_LinearColorByFloat(
     Linear Color=$Particles.Initial.Color
     Float=&FloatFromCurve(FloatCurve=DataInterface() {propertyValues {Curve {keys [{arriveTangent=-1.1839, interpMode=RCIM_Cubic, leaveTangent=-1.1839, tangentMode=RCTM_User, time=0, value=1}, <at 1 0>]}}})
    )
    Particles.BurstScale:Vector3f=&Multiply_VectorByFloat(
     Vector=&Multiply_Vector(A=$Particles.Initial.Scale, B=&RandomRangeVector(Minimum=<0.8 0.8 1>, Maximum=<1 1 1>))
     Float=&FloatFromCurve(FloatCurve=DataInterface() {propertyValues {Curve {
         defaultValue=1
         keys [
          {arriveTangent=2.367, interpMode=RCIM_Cubic, leaveTangent=2.367, tangentMode=RCTM_User, time=0, value=0.3}
          {arriveTangent=0.3402, interpMode=RCIM_Cubic, leaveTangent=0.3402, tangentMode=RCTM_User, time=1, value=1.5}
         ]
        }}})
    )
    Particles.CoronaColor:LinearColor=&Multiply_LinearColorByFloat(Linear Color=$Particles.Color, Float=0.2)
    Particles.CoronaSize:Vector2f=&Multiply_Vector2D_ByFloat(
     Float=&FloatFromCurve(FloatCurve=DataInterface() {propertyValues {Curve {keys [
          {interpMode=RCIM_Cubic, leaveTangentWeight=0.3333, time=0, value=0.85}
          {arriveTangent=0.0788, arriveTangentWeight=0.3506, interpMode=RCIM_Cubic, leaveTangent=0.0788, tangentMode=RCTM_User, time=1, value=1}
         ]}}})
     Vector2D=<400 400>
    )
   ) {isSetParameters=true}
   &FN_Light_Attributes(
    Color=$Particles.Color
    Create Color Attribute=true
    Create Exponent Attribute=true
    Create Position Attribute=true
    Create Radius Attribute=true
    Enable Particle Light Enable=true
    Enable Visibility Tag=true
    Exponent=2
    Light Enabled=true
    Position=&AddVectorToPosition(
     Position=$Particles.Initial.Position
     Vector=&Lerp_Vector(A=<0 0 -150>, B=<0 0 -200>, Alpha=$StackContext.NormalizedAge)
    )
    Radius=&FloatFromCurve(
     FloatCurve=DataInterface() {propertyValues {Curve {keys [{arriveTangent=-0.3006, interpMode=RCIM_Cubic, leaveTangent=-0.3006, tangentMode=RCTM_User, time=0, value=1.5}, <at 1 1.3>]}}}
     Scale Curve=100
    )
    Scale Color=&FloatFromCurve(
     FloatCurve=DataInterface() {propertyValues {Curve {keys [{interpMode=RCIM_Cubic, tangentMode=RCTM_User, time=0, value=1}, <at 1 0>]}}}
     Scale Curve=100
    )
    VisibilityTag=1
   ) {module=FN_Light_Attributes001}
   /Niagara/Modules/Decal/Decal_Attributes(
    Color=1
    Color Value To Scale:LinearColor=$Particles.Color
    Decal Facing Mode=1
    Decal Size:Vector3f=<50 700 700>
    Orientation=1
    Scale Alpha=0.08
    ScaleA=true
    ScaleRGB=true
    Set Alignment=false
    Size=1
    Visible=1
   )
   &DynamicMaterialParameters(
    Index 0 Param 1=&MaskFloat(
     Mask Source=0
     Mask Parameter=?
     Mask=5
     Masked Float=0.4176
     Passthrough Float=&FloatFromCurve(FloatCurve=DataInterface() {propertyValues {Curve {keys [
          {arriveTangent=-3.6312, interpMode=RCIM_Cubic, leaveTangent=-3.6312, tangentMode=RCTM_User, time=0.1, value=1}
          {arriveTangent=-0.3283, interpMode=RCIM_Cubic, leaveTangent=-0.3283, tangentMode=RCTM_User, time=1, value=-0.5}
         ]}}})
    )
    Index 0 Param 2=0.5
    Index 0 Param 3=&MaskFloat(
     Mask Source=0
     Mask Parameter=?
     Mask=5
     Masked Float=&RandomRangeFloat(Minimum=0.1, Maximum=0.9)
     Passthrough Float=0
    )
    Index 0 Param 4=&MaskFloat(
     Mask Source=0
     Mask Parameter=?
     Mask=5
     Masked Float=&Add_Float(
      A=&RandomRangeFloat(Minimum=0, Maximum=1)
      B=&FloatFromCurve(FloatCurve=DataInterface() {propertyValues {Curve {
          defaultValue=1
          keys [
           {arriveTangent=1.6785, interpMode=RCIM_Cubic, leaveTangent=1.6785, tangentMode=RCTM_User, time=0, value=0}
           {interpMode=RCIM_Cubic, tangentMode=RCTM_User, time=1, value=1}
          ]
         }}})
     )
     Passthrough Float=0
    )
    Module.Index0Float0Write=true
    Module.Index0Float1Write=true
    Module.Index0Float2Write=true
    Module.Index0Float3Write=true
   )
  ]
  rendererClasses [
   &NiagaraSpriteRendererProperties
   &NiagaraMeshRendererProperties
   &NiagaraLightRendererProperties
   &NiagaraDecalRendererProperties
  ]
  renderers [
   {rendererClass=&NiagaraSpriteRendererProperties}
   {rendererClass=&NiagaraSpriteRendererProperties, rendererIndex=1}
   {rendererClass=&NiagaraSpriteRendererProperties, rendererIndex=2}
   {rendererClass=&NiagaraSpriteRendererProperties, rendererIndex=3}
   {rendererClass=&NiagaraSpriteRendererProperties, rendererIndex=4}
   {rendererClass=&NiagaraSpriteRendererProperties, rendererIndex=5}
   {rendererClass=&NiagaraMeshRendererProperties, rendererIndex=6}
   {rendererClass=&NiagaraMeshRendererProperties, rendererIndex=7}
   {rendererClass=&NiagaraMeshRendererProperties, rendererIndex=8}
   {rendererClass=&NiagaraMeshRendererProperties, rendererIndex=9}
   {rendererClass=&NiagaraMeshRendererProperties, rendererIndex=10}
   {rendererClass=&NiagaraSpriteRendererProperties, rendererIndex=11}
   {rendererClass=&NiagaraSpriteRendererProperties, rendererIndex=12}
   {rendererClass=&NiagaraSpriteRendererProperties, rendererIndex=13}
   {rendererClass=&NiagaraSpriteRendererProperties, rendererIndex=14}
   {rendererClass=&NiagaraMeshRendererProperties, rendererIndex=15}
   {rendererClass=&NiagaraMeshRendererProperties, rendererIndex=16}
   {rendererClass=&NiagaraMeshRendererProperties, rendererIndex=17}
   {rendererClass=&NiagaraSpriteRendererProperties, rendererIndex=18}
   {rendererClass=&NiagaraLightRendererProperties, rendererIndex=19}
   {rendererClass=&NiagaraDecalRendererProperties, rendererIndex=20}
  ]
 }
 Sparks {rendererClasses [], renderers []}
}
summary {
 systemName=NS_Emote_DuneBrief_EnergyBurstRain
 emitters [
  {emitterName=Energy_Rain, rendererClasses [&NiagaraSpriteRendererProperties, &NiagaraLightRendererProperties]}
  {
   emitterName=Explosion
   rendererClasses [
    &NiagaraSpriteRendererProperties
    &NiagaraMeshRendererProperties
    &NiagaraLightRendererProperties
    &NiagaraDecalRendererProperties
   ]
  }
  {emitterName=Sparks, rendererClasses []}
 ]
 userVariables [User.ExplosionDelay {type=NiagaraFloat, defaultValue=3.6, description=""}]
}
