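# Niagara effect for the DuneBrief emote's shot: a blue muzzle flash and upward-firing projectile with lingering residue.
# Muzzle layers meshes, sprites and a light, using spawn groups as visibility tags to vary lifetime, scale and color.
# Projectile eases from its start by System.ProjectileHeight, pinning core particles to that moving point while trail particles fall away.
# Residue adds spinning, expanding meshes and curl-noise sparks stretched by speed.
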NiagaraSystem /BRCosmetics/Animation/Game/MainPlayer/Emotes/DuneBrief/FX/NS_Emote_DuneBrief_Shot
dependencies {
 usedDataInterfaces [
  &NiagaraDataInterfaceVectorCurve
  &NiagaraDataInterfaceCurve
  &NiagaraDataInterfaceVector2DCurve
 ]
 usedDynamicInputs []
 usedModules [
  &self:SystemSpawnScript.SystemScriptSource.SystemScriptGraph.NiagaraNodeAssignment_0.SetVariables_#1
  &SystemState
 ]
}
emitters {
 Muzzle {
  emitterUpdateScript [
   &EmitterState(Life Cycle Mode=::NewEnumerator1, Loop Behavior=::NewEnumerator1, Loop Duration=0.3, UseLoopDelay=false)
   &SpawnBurst_Instantaneous(Spawn Count=3, Use Spawn Probability=false, UseLoopCountLimit=false)
   &SpawnBurst_Instantaneous(Spawn Group=1, Use Spawn Probability=false, UseLoopCountLimit=false) {module=SpawnBurst_Instantaneous001}
   &SpawnBurst_Instantaneous(Spawn Group=2, Use Spawn Probability=false, UseLoopCountLimit=false) {module=SpawnBurst_Instantaneous002}
   &SpawnBurst_Instantaneous(Spawn Group=3, Use Spawn Probability=false, UseLoopCountLimit=false) {module=SpawnBurst_Instantaneous003}
  ]
  particleSpawnScript [
   /SetVariables(Particles.VisibilityTag=&SpawnGroup()) {isSetParameters=true}
   &PartitionParticles(Number Of Partitions=3)
   &InitializeParticle(
    Color=<rgba 1 3.25 10 1>
    Lifetime=&MergeHelper_FloatSwitch(
     0=&Lerp_Float(A=0.6, B=0.8, Alpha=&ReturnNormalizedExecIndex())
     1=0.5
     2=0.35
     3=0.27
     Index=&SpawnGroup()
     Num=::NewEnumerator2
    )
    Mesh Scale Mode=::NewEnumerator3
    Sprite Size Mode=::NewEnumerator3
    UsePositionOffset=false
    Write Material Random=true
   )
   &InitialMeshOrientation(
    Mesh Orientation Mode=::NewEnumerator2
    Orientation Coordinate Space=::World
    Orientation Vector=<0 0 1>
    Rotation=&MakeVector(
     X=&MaskFloat(
      Mask Source=0
      Mask Parameter=?
      Mask=0
      Masked Float=&RandomRangeFloat(Minimum=0, Maximum=1)
      Passthrough Float=0
     )
     Y=0
     Z=&MaskFloat(Mask Source=0, Mask Parameter=?, Mask=1, Masked Float=0.25, Passthrough Float=0)
    )
    Use Rotation Vector=true
   )
   &AddVelocity(
    Rotation Coordinate Space=::World
    Velocity=<0 0 200>
    Velocity Speed Scale=&MaskFloat(
     Mask Source=0
     Mask Parameter=?
     Mask=0
     Masked Float=&Lerp_Float(A=0.5, B=3, Alpha=&ReturnNormalizedExecIndex())
     Passthrough Float=0
    )
   )
   &ApplyInitialForces()
   &CameraOffset(Camera Offset Amount=25)
  ]
  particleUpdateScript [
   &ParticleState()
   &ScaleMeshSize(Scale Factor=&MergeHelper_VectorSwitch(
     0=&VectorFromCurve(
      VectorCurve=DataInterface() {propertyValues {
        XCurve {keys [
          {arriveTangent=4.5025, interpMode=RCIM_Cubic, leaveTangent=4.5025, tangentMode=RCTM_User, time=0, value=0.5}
          {arriveTangentWeight=0.0658, interpMode=RCIM_Cubic, leaveTangentWeight=0.2932, tangentMode=RCTM_User, time=0.1834, value=1}
          <at 1 0.5>
         ]}
        YCurve {keys [{arriveTangent=1.4256, interpMode=RCIM_Cubic, leaveTangent=1.4256, tangentMode=RCTM_User, time=0, value=0.5}, <at 1 1>]}
        ZCurve {keys [{arriveTangent=1.4256, interpMode=RCIM_Cubic, leaveTangent=1.4256, tangentMode=RCTM_User, time=0, value=0.5}, <at 1 1>]}
       }}
      Scale Curve=&VectorFromCurve(
       VectorCurve=DataInterface() {propertyValues {
         XCurve {keys [<at 0 1>,<at 1 0.5>]}
         YCurve {keys [{arriveTangent=-1.779, interpMode=RCIM_Cubic, leaveTangent=-1.779, tangentMode=RCTM_User, time=0, value=3}, <at 3 0.3>]}
         ZCurve {keys [{arriveTangent=-1.779, interpMode=RCIM_Cubic, leaveTangent=-1.779, tangentMode=RCTM_User, time=0, value=3}, <at 3 0.3>]}
        }}
       CurveIndex=&MakeFloatFromInt(Int=$Particles.PartitionParticles.Partition)
      )
     )
     1=&Multiply_VectorByFloat(
      Vector=<0.1 2.5 2.5>
      Float=&FloatFromCurve(FloatCurve=DataInterface() {propertyValues {Curve {
          defaultValue=1
          keys [
           {arriveTangent=1.2014, interpMode=RCIM_Cubic, leaveTangent=1.2014, tangentMode=RCTM_User, time=0, value=0.5}
           {interpMode=RCIM_Cubic, tangentMode=RCTM_User, time=1, value=1}
          ]
         }}})
     )
     2=<0 0 0>
     3=&VectorFromCurve(VectorCurve=DataInterface() {propertyValues {
        XCurve {keys [
          {arriveTangent=13.4953, interpMode=RCIM_Cubic, leaveTangent=13.4953, tangentMode=RCTM_User, time=0, value=0.9}
          {arriveTangent=0.6687, interpMode=RCIM_Cubic, leaveTangent=0.6687, time=0.3067, value=4.099}
          {arriveTangent=-0.2537, interpMode=RCIM_Cubic, leaveTangent=-0.2537, tangentMode=RCTM_User, time=1, value=1.5687}
         ]}
        YCurve {keys [
          {arriveTangent=40.2032, interpMode=RCIM_Cubic, leaveTangent=40.2032, tangentMode=RCTM_User, time=0, value=0.5}
          {interpMode=RCIM_Cubic, tangentMode=RCTM_User, time=0.1, value=4}
          {interpMode=RCIM_Cubic, tangentMode=RCTM_User, time=1, value=1}
         ]}
        ZCurve {keys [
          {arriveTangent=40.2032, interpMode=RCIM_Cubic, leaveTangent=40.2032, tangentMode=RCTM_User, time=0, value=0.5}
          {interpMode=RCIM_Cubic, tangentMode=RCTM_User, time=0.1, value=4}
          {interpMode=RCIM_Cubic, tangentMode=RCTM_User, time=1, value=1}
         ]}
       }})
     Index=&SpawnGroup()
     Num=::NewEnumerator2
    ))
   &ScaleSpriteSize(
    Scale Sprite Size Mode=::NewEnumerator0
    Uniform Scale Factor=&MergeHelper_FloatSwitch(
     0=0
     1=&FloatFromCurve(
      FloatCurve=DataInterface() {propertyValues {Curve {keys [
          {arriveTangent=8.1464, interpMode=RCIM_Cubic, leaveTangent=8.1464, tangentMode=RCTM_User, time=0, value=0.5}
          {interpMode=RCIM_Cubic, tangentMode=RCTM_User, time=0.1376, value=1.3}
          <at 1 0.5>
         ]}}}
      Scale Curve=20
     )
     2=&FloatFromCurve(
      FloatCurve=DataInterface() {propertyValues {Curve {keys [
          {arriveTangent=24.4655, interpMode=RCIM_Cubic, leaveTangent=24.4655, tangentMode=RCTM_User, time=0, value=0.1043}
          {arriveTangent=-0.1043, interpMode=RCIM_Cubic, leaveTangent=-0.1043, time=0.1122, value=1}
          <at 1 0>
         ]}}}
      Scale Curve=30
     )
     3=?
     Index=&SpawnGroup()
     Num=::NewEnumerator1
    )
   )
   &ScaleColor(
    Scale Alpha=&MergeHelper_FloatSwitch(
     0=1
     1=&FloatFromCurve(
      FloatCurve=DataInterface() {propertyValues {Curve {keys [{interpMode=RCIM_Cubic, tangentMode=RCTM_User, time=0, value=1}, {interpMode=RCIM_Cubic, tangentMode=RCTM_User, time=0.1, value=1}, <at 1 0>]}}}
      Scale Curve=0.1
     )
     2=1
     3=?
     Index=&SpawnGroup()
     Num=::NewEnumerator1
    )
    Scale RGB=&Multiply_VectorByFloat(
     Vector=<1 1 1>
     Float=&MergeHelper_FloatSwitch(
      0=&FloatFromCurve(
       FloatCurve=DataInterface() {propertyValues {Curve {keys [{arriveTangent=-2.8386, interpMode=RCIM_Cubic, leaveTangent=-2.8386, tangentMode=RCTM_User, time=0, value=1}, <at 1 0.01>]}}}
       Scale Curve=2
      )
      1=&FloatFromCurve(FloatCurve=DataInterface() {propertyValues {Curve {keys [{arriveTangent=-9.6567, interpMode=RCIM_Cubic, leaveTangent=-9.6567, tangentMode=RCTM_User, time=0, value=3}, <at 1 0>]}}})
      2=&FloatFromCurve(FloatCurve=DataInterface() {propertyValues {Curve {keys [
           {arriveTangent=-9.6567, interpMode=RCIM_Cubic, leaveTangent=-9.6567, tangentMode=RCTM_User, time=0, value=5}
           <at 0.2034 1 RCIM_Cubic>
           <at 1 5>
          ]}}})
      3=0.5
      Index=&SpawnGroup()
      Num=::NewEnumerator2
     )
    )
    ScaleA=true
    ScaleRGB=true
   )
   /SetVariables(
    Particles.MuzzleFlashGlow:LinearColor=&Multiply_LinearColorByFloat(Linear Color=<rgba 0 0.1281 2 0.2>, Float=3)
    Particles.RingCamFacingColor:LinearColor=&Multiply_LinearColorByFloat(
     Linear Color=$Particles.Initial.Color
     Float=&FloatFromCurve(FloatCurve=DataInterface() {propertyValues {Curve {keys [{arriveTangent=-1.9029, interpMode=RCIM_Cubic, leaveTangent=-1.9029, tangentMode=RCTM_User, time=0, value=1}, <at 0.8 0>]}}})
    )
    Particles.StarGlowColor:LinearColor=&Multiply_LinearColorByFloat(Linear Color=<rgba 0 0.0641 1 1>, Float=1)
   ) {isSetParameters=true}
   /SetVariables(
    Particles.DarkGlowColor:LinearColor=&Multiply_LinearColorByFloat(
     Linear Color=<rgba 0 0.0025 0.05 1>
     Float=&FloatFromCurve(
      FloatCurve=DataInterface() {propertyValues {Curve {keys [{interpMode=RCIM_Cubic, tangentMode=RCTM_User, time=0, value=1}, <at 1 0>]}}}
      Scale Curve=0.5
     )
    )
    Particles.DarkGlowSize:Vector2f=&Multiply_Vector2D_ByFloat(Float=1.03, Vector2D=$Particles.SpriteSize)
   ) {isSetParameters=true}
   &Drag(Drag=5, Use Linear Drag=true, Use Rotational Drag=false)
   &SolveForcesAndVelocity(Clamp Velocity=false, Limit Acceleration=false)
   &FN_Light_Attributes(
    Color=$Particles.Color
    Create Color Attribute=true
    Create Exponent Attribute=true
    Create Radius Attribute=true
    Enable Particle Light Enable=true
    Enable Visibility Tag=true
    Light Enabled=true
    Radius=&FloatFromCurve(
     FloatCurve=DataInterface() {propertyValues {Curve {keys [{arriveTangent=-0.6132, interpMode=RCIM_Cubic, leaveTangent=-0.6132, tangentMode=RCTM_User, time=0, value=1}, <at 1 0.6>]}}}
     Scale Curve=200
    )
    Scale Color=&FloatFromCurve(
     FloatCurve=DataInterface() {propertyValues {Curve {keys [{interpMode=RCIM_Cubic, tangentMode=RCTM_User, time=0, value=1}, <at 1 0>]}}}
     Scale Curve=700
    )
    VisibilityTag=1
   )
   &FN_Sprite_Attributes(
    Create SpriteSize Attribute=true
    Module Output:ENiagara_AttributeSamplingApplyOutput=::NewEnumerator1
    SpriteSize=&Multiply_Vector2D_ByFloat(
     Float=&FloatFromCurve(FloatCurve=DataInterface() {propertyValues {Curve {keys [{interpMode=RCIM_Cubic, tangentMode=RCTM_User, time=0, value=1}, <at 1 0>]}}})
     Vector2D=<300 150>
    )
   ) {module=FN_Sprite_Attributes001}
   &FN_Sprite_Attributes(
    Alignment:Vector3f=<0 0 -1>
    Create Alignment Attribute=true
    Create Dynamic Material Parameter Attributes=true
    Create SpriteSize Attribute=true
    Index 0 Param 1:NiagaraFloat=&FloatFromCurve(FloatCurve=DataInterface() {propertyValues {Curve {keys [{interpMode=RCIM_Cubic, tangentMode=RCTM_User, time=0.3, value=1}, <at 1 0.3>]}}})
    Index 0 Param 4:NiagaraFloat=&FloatFromCurve(FloatCurve=DataInterface() {propertyValues {Curve {keys [{interpMode=RCIM_Cubic, tangentMode=RCTM_User, time=0, value=1}, {arriveTangent=-0.4944, tangentMode=RCTM_User, time=1, value=-1}]}}})
    Module Output:ENiagara_AttributeSamplingApplyOutput=::NewEnumerator1
    Module.Index0Float0Write=true
    Module.Index0Float1Write=false
    Module.Index0Float2Write=false
    Module.Index0Float3Write=true
    SpriteSize=&Vector2DFromCurve(
     Vector2Curve=DataInterface() {propertyValues {
       XCurve {keys [
         {arriveTangent=1.6065, interpMode=RCIM_Cubic, leaveTangent=1.6065, tangentMode=RCTM_User, time=0, value=0.3}
         {arriveTangent=0.875, interpMode=RCIM_Cubic, leaveTangent=0.875, time=0.0646, value=0.7723}
         {interpMode=RCIM_Cubic, tangentMode=RCTM_User, time=0.8, value=1}
         <at 1 0.75>
        ]}
       YCurve {keys [
         {arriveTangent=0.1345, interpMode=RCIM_Cubic, leaveTangent=0.1345, tangentMode=RCTM_User, time=0, value=0.3}
         {arriveTangent=0.1, interpMode=RCIM_Cubic, leaveTangent=0.1, tangentMode=RCTM_User, time=0.312, value=0.7986}
         {arriveTangent=-0.0177, interpMode=RCIM_Cubic, leaveTangent=-0.0177, tangentMode=RCTM_User, time=0.6647, value=0.4866}
         <at 1 0.65>
        ]}
      }}
     ScaleCurve=<101.7142 177.5088>
    )
   ) {module=FN_Sprite_Attributes002}
   /SetVariables(Particles.MuzzleGlowSize:Vector2f=&Multiply_Vector2D(A=$Particles.FN_Sprite_Attributes002.SpriteSize, B=<1.5712 1.319>)) {isSetParameters=true}
   &DynamicMaterialParameters(
    Index 0 Param 1=&FloatFromCurve(FloatCurve=DataInterface() {propertyValues {Curve {keys [
         {arriveTangent=2.3916, interpMode=RCIM_Cubic, leaveTangent=2.3916, leaveTangentWeight=0.3158, tangentMode=RCTM_User, time=0, value=-0.5}
         {arriveTangent=0.016, arriveTangentWeight=0.3, interpMode=RCIM_Cubic, leaveTangent=0.016, tangentMode=RCTM_User, time=1, value=0.7}
        ]}}})
    Index 0 Param 4=&FloatFromCurve(FloatCurve=DataInterface() {propertyValues {Curve {
        defaultValue=0.9
        keys [
         {interpMode=RCIM_Cubic, tangentMode=RCTM_User, time=0.5, value=1}
         {arriveTangent=-1.0158, interpMode=RCIM_Cubic, leaveTangent=-1.0158, tangentMode=RCTM_User, time=1, value=0}
        ]
       }}})
    Module.Index0Float0Write=true
    Module.Index0Float1Write=false
    Module.Index0Float2Write=false
    Module.Index0Float3Write=true
   )
  ]
  rendererClasses [
   &NiagaraMeshRendererProperties
   &NiagaraSpriteRendererProperties
   &NiagaraLightRendererProperties
  ]
  renderers [
   {rendererClass=&NiagaraMeshRendererProperties, rendererIndex=0}
   {rendererClass=&NiagaraSpriteRendererProperties, rendererIndex=1}
   {rendererClass=&NiagaraLightRendererProperties, rendererIndex=2}
   {rendererClass=&NiagaraSpriteRendererProperties, rendererIndex=3}
   {rendererClass=&NiagaraSpriteRendererProperties, rendererIndex=4}
   {rendererClass=&NiagaraSpriteRendererProperties, rendererIndex=5}
   {rendererClass=&NiagaraSpriteRendererProperties, rendererIndex=6}
   {rendererClass=&NiagaraSpriteRendererProperties, rendererIndex=7}
   {rendererClass=&NiagaraMeshRendererProperties, rendererIndex=8}
  ]
 }
 Projectile {
  emitterSpawnScript [/SetVariables(Emitter.InitPosition:NiagaraPosition=$Engine.Emitter.SimulationPosition) {isSetParameters=true}]
  emitterUpdateScript [
   &EmitterState(Inactive Response=::NewEnumerator1, Life Cycle Mode=::NewEnumerator1, Loop Behavior=::NewEnumerator1, Loop Duration=0.7, UseLoopDelay=false)
   &SpawnBurst_Instantaneous(Use Spawn Probability=false, UseLoopCountLimit=false) {module=SpawnBurst_Instantaneous001}
   &SpawnRate(
    SpawnGroup=1
    SpawnRate=&FloatFromCurve(
     FloatCurve=DataInterface() {propertyValues {Curve {keys [{interpMode=RCIM_Cubic, tangentMode=RCTM_User, time=0, value=1}, <at 0.99 0.1>]}}}
     CurveIndex=$Emitter.NormalizedLoopAge
     Scale Curve=11
    )
    Use Spawn Probability=false
   )
   &SpawnBurst_Instantaneous(Spawn Group=2, Use Spawn Probability=false, UseLoopCountLimit=false)
   &SpawnRate(SpawnGroup=2, SpawnRate=10, Use Spawn Probability=false) {module=SpawnRate001}
   /SetVariables(Emitter.ProjectilePOS:Vector3f=&AddVectorToPosition(
     Position=Linked() {linkedVariable {name=Emitter.InitPosition, type=Vector3f}}
     Vector=&Lerp_Vector(
      A=<0 0 0>
      B=Linked() {linkedVariable {name=System.ProjectileHeight, type=NiagaraPosition}}
      Alpha=&FloatFromCurve(
       FloatCurve=DataInterface() {propertyValues {Curve {
          defaultValue=1
          keys [
           {arriveTangent=7.2605, interpMode=RCIM_Cubic, leaveTangent=7.2605, tangentMode=RCTM_User, time=0, value=0}
           {arriveTangent=1.6712, interpMode=RCIM_Cubic, leaveTangent=1.6712, tangentMode=RCTM_User, time=0.1169, value=0.6102}
           <at 1 1 RCIM_Cubic>
          ]
         }}}
       CurveIndex=$Emitter.NormalizedLoopAge
      )
     )
    )) {isSetParameters=true}
  ]
  particleSpawnScript [
   &InitializeParticle(
    Color=<rgba 0.49 2.1409 7 1>
    Lifetime=&MaskFloat(Mask Source=0, Mask Parameter=?, Mask=0, Masked Float=0.7, Passthrough Float=0.3)
    Mesh Scale=<1 0.6 0.6>
    Mesh Scale Mode=::NewEnumerator4
    Sprite Size Mode=::NewEnumerator3
    Uniform Sprite Size=15
    UsePositionOffset=false
    Write Material Random=true
   )
   /SetVariables(Particles.VisibilityTag=&SpawnGroup()) {isSetParameters=true}
   &ShapeLocation(MaskSpawnGroup=false, Override Seed=false, Shape Origin:Vector3f=$Emitter.ProjectilePOS, Sphere Radius=0, UseOwnerScale=true) {module=ShapeLocation001}
   &InitialMeshOrientation(
    Mesh Orientation Mode=::NewEnumerator2
    Rotation=&MakeVector(X=0, Y=0, Z=&RandomRangeFloat(Minimum=0, Maximum=1))
    Use Rotation Vector=true
   )
   &AddVelocity(
    Rotation Coordinate Space=::World
    Velocity=&MaskVector(Mask Source=0, Mask Parameter=?, Mask=1, Masked Vector=<0 0 -200>, Passthrough Vector=<0 0 0>)
   )
   &ApplyInitialForces()
   &AlignSpriteToMeshOrientation(Mesh Orientation Relative Sprite Alignment Vector=<0 0 1>, SetAlignment=true, SetFacing=false)
   /SetVariables(Particles.CoreRotation:NiagaraFloat=&RandomRangeFloat(Minimum=0, Maximum=360)) {isSetParameters=true}
   &CameraOffset(Camera Offset Amount=&MaskFloat(Mask Source=0, Mask Parameter=?, Mask=0, Masked Float=-10, Passthrough Float=-3)) {enabled=false}
  ]
  particleUpdateScript [
   &ParticleState()
   &ScaleColor(
    Scale Alpha=&MaskFloat(
     Mask Source=0
     Mask Parameter=?
     Mask=0
     Masked Float=1
     Passthrough Float=&FloatFromCurve(FloatCurve=DataInterface() {propertyValues {Curve {keys [{interpMode=RCIM_Cubic, tangentMode=RCTM_User, time=0, value=1}, <at 1 0>]}}})
    )
    Scale RGB=&Multiply_VectorByFloat(
     Vector=<1 1 1>
     Float=&MergeHelper_FloatSwitch(0=2, 1=0.5, 2=1, 3=?, Index=&SpawnGroup(), Num=::NewEnumerator1)
    )
    ScaleA=true
    ScaleRGB=true
   )
   &ScaleSpriteSize(
    Scale Sprite Size Mode=::NewEnumerator0
    Uniform Scale Factor=&FloatFromCurve(FloatCurve=DataInterface() {propertyValues {Curve {
        defaultValue=1
        keys [
         {interpMode=RCIM_Cubic, tangentMode=RCTM_User, time=0, value=1}
         {interpMode=RCIM_Cubic, tangentMode=RCTM_User, time=0.0688, value=2}
         {interpMode=RCIM_Cubic, tangentMode=RCTM_User, time=0.1417, value=1}
         {interpMode=RCIM_Cubic, tangentMode=RCTM_User, time=0.2275, value=2}
         <at 0.3175 1 RCIM_Cubic>
         <at 0.3904 2 RCIM_Cubic>
         {arriveTangent=0.071, interpMode=RCIM_Cubic, leaveTangent=0.071, time=0.4912, value=1}
         <at 0.5877 2.014 RCIM_Cubic>
         {arriveTangentWeight=0.2333, interpMode=RCIM_Cubic, tangentMode=RCTM_User, time=0.7, value=1}
         {arriveTangent=-3.3333, interpMode=RCIM_Cubic, leaveTangent=-3.3333, time=0.8, value=2.5645}
         <at 1 0>
        ]
       }}})
   )
   &ScaleMeshSize(Scale Factor=&Multiply_VectorByFloat(
     Vector=&MakeVector(
      X=&FloatFromCurve(
       FloatCurve=DataInterface() {propertyValues {Curve {keys [{arriveTangent=3.5233, interpMode=RCIM_Cubic, leaveTangent=3.5233, tangentMode=RCTM_User, time=0, value=0.5}, <at 1 2>]}}}
       Scale Curve=&FloatFromCurve(
        FloatCurve=DataInterface() {propertyValues {Curve {keys [{interpMode=RCIM_Cubic, tangentMode=RCTM_User, time=0.8, value=1}, <at 1 0.3>]}}}
        CurveIndex=$Emitter.NormalizedLoopAge
       )
      )
      Y=1
      Z=1
     )
     Float=&FloatFromCurve(FloatCurve=DataInterface() {propertyValues {Curve {
         defaultValue=1
         keys [
          {arriveTangent=1.9393, interpMode=RCIM_Cubic, leaveTangent=1.9393, tangentMode=RCTM_User, time=0, value=0.3}
          {arriveTangent=0.1131, interpMode=RCIM_Cubic, tangentMode=RCTM_User, time=0.3602, value=0.7}
          {arriveTangent=-0.9779, tangentMode=RCTM_User, time=1, value=0.2713}
         ]
        }}})
    ))
   /SetVariables(
    Particles.CoreGlowColor:LinearColor=&MultiplyLinearColor(Linear Color=<rgba 0.0037 0.0666 0.3698 1>, Scale Factor=<1 1 2 1>)
    Particles.CoreGlowSize:Vector2f=&Multiply_Vector2D_ByFloat(Float=4, Vector2D=$Particles.SpriteSize)
   ) {isSetParameters=true}
   /SetVariables(
    Particles.ProjectileDynP00:Vector4f=&MakeVector4(
     X=&FloatFromCurve(FloatCurve=DataInterface() {propertyValues {Curve {keys [{interpMode=RCIM_Cubic, tangentMode=RCTM_User, time=0, value=1}, <at 1 0>]}}})
     Y=0
     Z=0
     W=0
    )
    Particles.ProjectileSize:Vector2f=&Multiply_Vector2D(
     A=$Particles.Initial.SpriteSize
     B=&Vector2DFromCurve(
      Vector2Curve=DataInterface() {propertyValues {
        XCurve {keys [
          {arriveTangent=14.4902, interpMode=RCIM_Cubic, leaveTangent=14.4902, tangentMode=RCTM_User, time=0, value=0.9535}
          {interpMode=RCIM_Cubic, tangentMode=RCTM_User, time=0.1238, value=1.5}
          <at 1 0.5>
         ]}
        YCurve {keys [
          {arriveTangent=5.3135, interpMode=RCIM_Cubic, leaveTangent=5.3135, tangentMode=RCTM_User, time=0, value=0.5}
          {interpMode=RCIM_Cubic, tangentMode=RCTM_User, time=0.3811, value=1.5}
          <at 0.7692 0.5>
         ]}
       }}
      ScaleCurve=&Multiply_Vector2D_ByFloat(
       Float=&FloatFromCurve(
        FloatCurve=DataInterface() {propertyValues {Curve {keys [{arriveTangent=-3.3326, interpMode=RCIM_Cubic, leaveTangent=-3.3326, tangentMode=RCTM_User, time=0.3, value=1}, <at 1 0>]}}}
        CurveIndex=$Emitter.NormalizedLoopAge
       )
       Vector2D=<2.9086 7.4651>
      )
     )
    )
   ) {isSetParameters=true}
   /SetVariables(
    Particles.ProjectileDarkColor:LinearColor=&MultiplyLinearColor(
     Linear Color=$Particles.Color
     Scale Factor=/Niagara/DynamicInputs/Multiply/Multiply_Vector4_ByFloat(Vector4:Vector4f=<0.01 0.02 1 1>, Float:NiagaraFloat=1)
    )
    Particles.ProjectileDarkSize:Vector2f=&Multiply_Vector2D_ByFloat(Float=1.5, Vector2D=$Particles.ProjectileSize)
   ) {isSetParameters=true}
   /SetVariables(
    Particles.ProjectileGlowSize:Vector2f=&Multiply_Vector2D(A=$Particles.ProjectileSize, B=<2.9296 2.3195>)
    Particles.ProjectileHotColor:LinearColor=&MultiplyLinearColor(Linear Color=$Particles.Color, Scale Factor=<1.7 1.7 1.7 1>)
    Particles.ProjectileHotSize:Vector2f=&Multiply_Vector2D(
     A=$Particles.ProjectileSize
     B=&Vector2DFromCurve(
      Vector2Curve=DataInterface() {propertyValues {
        XCurve {keys [
          {interpMode=RCIM_Cubic, tangentMode=RCTM_User, time=0, value=1}
          {interpMode=RCIM_Cubic, tangentMode=RCTM_User, time=0.4721, value=0.3}
          {interpMode=RCIM_Cubic, tangentMode=RCTM_User, time=1, value=1}
         ]}
        YCurve {keys [
          {interpMode=RCIM_Cubic, tangentMode=RCTM_User, time=0, value=0.3}
          {interpMode=RCIM_Cubic, tangentMode=RCTM_User, time=0.4312, value=1}
          <at 1.0268 0.3>
         ]}
       }}
      ScaleCurve=<0.7 0.7>
     )
    )
    Particles.ProjectxileGlowColor:LinearColor=&Multiply_LinearColorByFloat(Linear Color=<rgba 0.0037 0.0666 0.3698 1>, Float=1)
   ) {isSetParameters=true}
   /SetVariables(
    Particles.CoreBurstDarkColor:LinearColor=&MultiplyLinearColor(
     Linear Color=$Particles.Color
     Scale Factor=&Multiply_Vector4_ByFloat(Vector4:Vector4f=<0.01 0.02 1 4>, Float:NiagaraFloat=4)
    )
    Particles.CoreBurstDarkSize:Vector2f=&Multiply_Vector2D_ByFloat(
     Float=&MakeFloatFromVector2d(Vector 2D=$Particles.ProjectileHotSize, Channel=::NewEnumerator1)
     Vector2D=<0.3 0.3>
    )
   ) {isSetParameters=true}
   /SetVariables(
    Particles.CoreBurstHotColor:LinearColor=&MultiplyLinearColor(
     Linear Color=$Particles.Color
     Scale Factor=&Multiply_Vector4_ByFloat(Vector4:Vector4f=<1.7 1.7 1.7 4>, Float:NiagaraFloat=3)
    )
    Particles.CoreBurstHotSize:Vector2f=&Multiply_Vector2D_ByFloat(
     Float=&MakeFloatFromVector2d(Vector 2D=$Particles.ProjectileHotSize, Channel=::NewEnumerator1)
     Vector2D=<0.33 0.33>
    )
   ) {enabled=false, isSetParameters=true}
   /SetVariables(Particles.Position=&MaskVector(
     Mask Source=0
     Mask Parameter=?
     Mask=1
     Masked Vector=Linked() {linkedVariable {name=Particles.Position, type=NiagaraPosition}}
     Passthrough Vector=$Emitter.ProjectilePOS
    )) {isSetParameters=true}
   &DynamicMaterialParameters(
    Index 0 Param 1=&FloatFromCurve(FloatCurve=DataInterface() {propertyValues {Curve {keys [
         {arriveTangent=-0.016, interpMode=RCIM_Cubic, leaveTangent=-0.016, tangentMode=RCTM_User, time=0.1, value=0}
         {arriveTangent=0.011, interpMode=RCIM_Cubic, leaveTangent=0.011, tangentMode=RCTM_User, time=1, value=0.8}
        ]}}})
    Index 0 Param 4=&FloatFromCurve(FloatCurve=DataInterface() {propertyValues {Curve {
        defaultValue=0.9
        keys [
         {
          arriveTangent=-3.432
          arriveTangentWeight=0.1667
          interpMode=RCIM_Cubic
          leaveTangent=-3.432
          leaveTangentWeight=0.1667
          tangentMode=RCTM_User
          time=0.5
          value=1
         }
         {
          arriveTangent=-1.0158
          arriveTangentWeight=0.2376
          interpMode=RCIM_Cubic
          leaveTangent=-1.0158
          leaveTangentWeight=0.2376
          tangentMode=RCTM_User
          time=1
          value=0
         }
        ]
       }}})
    Module.Index0Float0Write=true
    Module.Index0Float1Write=false
    Module.Index0Float2Write=false
    Module.Index0Float3Write=true
   )
   &SolveForcesAndVelocity(Clamp Velocity=false, Limit Acceleration=false)
  ]
  rendererClasses [&NiagaraSpriteRendererProperties, &NiagaraMeshRendererProperties]
  renderers [
   {rendererClass=&NiagaraSpriteRendererProperties}
   {rendererClass=&NiagaraSpriteRendererProperties, rendererIndex=1}
   {rendererClass=&NiagaraSpriteRendererProperties, rendererIndex=2}
   {rendererClass=&NiagaraMeshRendererProperties, rendererIndex=3}
   {rendererClass=&NiagaraSpriteRendererProperties, rendererIndex=4}
   {rendererClass=&NiagaraSpriteRendererProperties, rendererIndex=5}
   {rendererClass=&NiagaraSpriteRendererProperties, rendererIndex=6}
   {rendererClass=&NiagaraSpriteRendererProperties, rendererIndex=7}
   {rendererClass=&NiagaraSpriteRendererProperties, rendererIndex=8}
   {rendererClass=&NiagaraSpriteRendererProperties, rendererIndex=9}
  ]
 }
 Residue {
  emitterUpdateScript [
   &EmitterState(Life Cycle Mode=::NewEnumerator1, Loop Behavior=::NewEnumerator1, Loop Duration=0.2833, UseLoopDelay=false)
   &SpawnBurst_Instantaneous(Spawn Count=5, Spawn Time=0.01, Use Spawn Probability=false, UseLoopCountLimit=false)
   &SpawnBurst_Instantaneous(Spawn Count=7, Spawn Group=1, Use Spawn Probability=false, UseLoopCountLimit=false) {module=SpawnBurst_Instantaneous001}
   &SpawnRate(
    SpawnGroup=1
    SpawnRate=&FloatFromCurve(
     FloatCurve=DataInterface() {propertyValues {Curve {keys [{interpMode=RCIM_Cubic, tangentMode=RCTM_User, time=0, value=1}, <at 1 0>]}}}
     CurveIndex=$Emitter.NormalizedLoopAge
     Scale Curve=50
    )
    Use Spawn Probability=false
   )
  ]
  particleSpawnScript [
   &InitializeParticle(
    Color=<rgba 0.7 2.2977 7 1>
    Lifetime=&MaskFloat(
     Mask Source=0
     Mask Parameter=?
     Mask=1
     Masked Float=&RandomRangeFloat(Minimum=0.3, Maximum=1.3)
     Passthrough Float=&RandomRangeFloat(Minimum=0.3, Maximum=1)
    )
    Mesh Scale=&Multiply_VectorByFloat(
     Vector=&VectorFromCurve(
      VectorCurve=DataInterface() {propertyValues {
        XCurve {keys [
          {arriveTangent=0.2343, interpMode=RCIM_Cubic, leaveTangent=0.2343, tangentMode=RCTM_User, time=0, value=0.3}
          {arriveTangent=0.0757, tangentMode=RCTM_User, time=1, value=0.6}
         ]}
        YCurve {keys [
          {arriveTangent=0.2343, interpMode=RCIM_Cubic, leaveTangent=0.2343, tangentMode=RCTM_User, time=0, value=0.3}
          {arriveTangent=0.0757, tangentMode=RCTM_User, time=1, value=0.6}
         ]}
        ZCurve {keys [
          {arriveTangent=0.0635, interpMode=RCIM_Cubic, leaveTangent=0.0635, tangentMode=RCTM_User, time=0, value=4}
          {arriveTangent=-0.0741, tangentMode=RCTM_User, time=1, value=1}
         ]}
       }}
      CurveIndex=&ReturnNormalizedExecIndex()
      Scale Curve=&MakeVector(X=1, Y=1, Z=&RandomRangeFloat(Minimum=0.8, Maximum=1.3))
     )
     Float=1
    )
    Mesh Scale Mode=::NewEnumerator4
    Sprite Size=<1 3>
    Sprite Size Mode=::NewEnumerator4
    UsePositionOffset=false
    Write Material Random=true
   )
   &InitialMeshOrientation(
    Mesh Orientation Mode=::NewEnumerator2
    Orientation Coordinate Space=::World
    Rotation=&MakeVector(
     X=&MakeCustomFloatFromBool(Bool=&RandomBool(), True Float=-0.03, False Float=0.03)
     Y=&MakeCustomFloatFromBool(Bool=&RandomBool(), True Float=-0.03, False Float=0.03)
     Z=&Lerp_Float(A=0, B=1, Alpha=&ReturnNormalizedExecIndex())
    )
    Use Rotation Vector=true
   )
   &AddVelocity(
    Cone Angle=5
    Cone Axis=<0 0 1>
    Module.UseConeFalloff=false
    Module.UseVelDistribution=false
    Override Seed=false
    Rotation Coordinate Space=::World
    Velocity Mode=::NewEnumerator2
    Velocity Speed=&MaskFloat(
     Mask Source=0
     Mask Parameter=?
     Mask=1
     Masked Float=&RandomRangeFloat(Minimum=30, Maximum=300)
     Passthrough Float=&RandomRangeFloat(Minimum=150, Maximum=-150)
    )
   )
   &ApplyInitialForces()
   /SetVariables(Particles.VisibilityTag=&SpawnGroup()) {isSetParameters=true}
  ]
  particleUpdateScript [
   &ParticleState()
   &DynamicMaterialParameters(
    Index 0 Param 1=&FloatFromCurve(FloatCurve=DataInterface() {propertyValues {Curve {keys [
         {arriveTangent=-1.8691, interpMode=RCIM_Cubic, leaveTangent=-1.8691, tangentMode=RCTM_User, time=0, value=1}
         {interpMode=RCIM_Cubic, tangentMode=RCTM_User, time=1, value=0.1}
        ]}}})
    Module.Param0WriteEnabled=true
    Module.Param1WriteEnabled=false
    Module.Param2WriteEnabled=false
    Module.Param3WriteEnabled=false
   )
   &ScaleMeshSize(Scale Factor=&Multiply_VectorByFloat(
     Vector=&Multiply_VectorByFloat(Vector=$Engine.Owner.Scale, Float=1)
     Float=&FloatFromCurve(
      FloatCurve=DataInterface() {propertyValues {Curve {keys [
          {arriveTangent=3.1849, interpMode=RCIM_Cubic, leaveTangent=3.1849, tangentMode=RCTM_User, time=0, value=0}
          {arriveTangent=0.9995, interpMode=RCIM_Cubic, leaveTangent=0.9995, tangentMode=RCTM_User, time=0.3, value=0.6}
          {arriveTangent=-0.0405, interpMode=RCIM_Cubic, leaveTangent=-0.0405, tangentMode=RCTM_User, time=1, value=1}
         ]}}}
      Scale Curve=2.8
     )
    ))
   &UpdateMeshOrientation(
    Rotation Rate=&FloatFromCurve(
     FloatCurve=DataInterface() {propertyValues {Curve {keys [{arriveTangent=-1.6054, interpMode=RCIM_Cubic, leaveTangent=-1.6054, tangentMode=RCTM_User, time=0, value=1}, <at 1 0.01>]}}}
     Scale Curve=-5
    )
    Rotation Vector=<0 0 1>
   )
   &ScaleColor(
    Scale Alpha=&FloatFromCurve(
     FloatCurve=DataInterface() {propertyValues {Curve {keys [
         {arriveTangent=13.4402, interpMode=RCIM_Cubic, leaveTangent=13.4402, tangentMode=RCTM_User, time=0, value=0}
         {arriveTangent=0.9995, interpMode=RCIM_Cubic, leaveTangent=0.9995, tangentMode=RCTM_User, time=0.1, value=1}
         {arriveTangent=-0.0405, interpMode=RCIM_Cubic, leaveTangent=-0.0405, tangentMode=RCTM_User, time=1, value=0}
        ]}}}
     Scale Curve=0.5
    )
    Scale RGB=&Multiply_VectorByFloat(
     Vector=<1 1 1>
     Float=&FloatFromCurve(FloatCurve=DataInterface() {propertyValues {Curve {keys [{interpMode=RCIM_Cubic, tangentMode=RCTM_User, time=0, value=1}, <at 1 0.1>]}}})
    )
    ScaleA=true
    ScaleRGB=true
   )
   &Drag(Drag=2, Use Linear Drag=true, Use Rotational Drag=false)
   &CurlNoiseForce(
    Bias Noise=false
    Noise Frequency=100
    Noise Strength=&MaskFloat(Mask Source=0, Mask Parameter=?, Mask=1, Masked Float=100, Passthrough Float=0)
    Pan Noise=false
   )
   &SolveForcesAndVelocity(Clamp Velocity=false, Limit Acceleration=false)
   &ScaleSpriteSizeBySpeed(Max Scale Factor=<1 5>)
   /SetVariables(
    Particles.SparkGlowColor:LinearColor=&MultiplyLinearColor(Linear Color=$Particles.Color, Scale Factor=<0.7 0.7 1.1 0.3>)
    Particles.SparkGlowSize:Vector2f=&Multiply_Vector2D(A=$Particles.SpriteSize, B=<2 1>)
   ) {isSetParameters=true}
  ]
  rendererClasses [&NiagaraMeshRendererProperties, &NiagaraSpriteRendererProperties]
  renderers [
   {rendererClass=&NiagaraMeshRendererProperties, rendererIndex=0}
   {rendererClass=&NiagaraSpriteRendererProperties, rendererIndex=1}
   {rendererClass=&NiagaraSpriteRendererProperties, rendererIndex=2}
  ]
 }
}
summary {
 systemName=NS_Emote_DuneBrief_Shot
 emitters [
  {
   emitterName=Muzzle
   rendererClasses [
    &NiagaraMeshRendererProperties
    &NiagaraSpriteRendererProperties
    &NiagaraLightRendererProperties
   ]
  }
  {emitterName=Residue, rendererClasses [&NiagaraMeshRendererProperties, &NiagaraSpriteRendererProperties]}
  {emitterName=Projectile, rendererClasses [&NiagaraSpriteRendererProperties, &NiagaraMeshRendererProperties]}
 ]
 userVariables []
}
