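# Niagara wind-burst effect for a shot in the DuneBrief player emote.
# Inward_Wind meshes spin and gather, while Glow_Body places outline-glow meshes on sampled skeleton bones.
# System delay variables time Burst_spikes (sprites, radial streaks, light) and Shockwave_Air; a duplicate spikes emitter is disabled.
# Spawn groups select per-layer curves within shared emitters, with time-of-day and front-end versus gameplay scalars adjusting brightness, alpha and size.
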NiagaraSystem /BRCosmetics/Animation/Game/MainPlayer/Emotes/DuneBrief/FX/NS_Emote_DuneBrief_Shot_WindBurst
dependencies {
 usedDataInterfaces [
  &NiagaraDataInterfaceCurve
  &NiagaraDataInterfaceColorCurve
  &NiagaraDataInterfaceVectorCurve
  &NiagaraDataInterfaceVector2DCurve
  &NiagaraDataInterfaceSkeletalMesh
 ]
 usedDynamicInputs []
 usedModules [
  &self:SystemSpawnScript.SystemScriptSource.SystemScriptGraph.NiagaraNodeAssignment_1.SetVariables_#1
  &self:SystemSpawnScript.SystemScriptSource.SystemScriptGraph.NiagaraNodeAssignment_2.SetVariables_#2
  &SystemState
  &self:SystemSpawnScript.SystemScriptSource.SystemScriptGraph.NiagaraNodeAssignment_0.SetVariables_#3
 ]
}
emitters {
 Burst_spikes {
  emitterUpdateScript [
   &EmitterState(Life Cycle Mode=::NewEnumerator1, Loop Behavior=::NewEnumerator1, Loop Delay=$System.AntiBurstDelay, Loop Duration=0.2, UseLoopDelay=true)
   &SpawnBurst_Instantaneous(Use Spawn Probability=false, UseLoopCountLimit=false) {module=SpawnBurst_Instantaneous002}
   &SpawnBurst_Instantaneous(Spawn Group=1, Spawn Time=0.1, Use Spawn Probability=false, UseLoopCountLimit=false) {module=SpawnBurst_Instantaneous003}
   &SpawnRate(SpawnRate=14, Use Spawn Probability=false) {enabled=false, module=SpawnRate001}
   &SpawnRate(SpawnGroup=2, SpawnRate=10, Use Spawn Probability=false) {enabled=false}
   &SpawnBurst_Instantaneous(Spawn Count=10, Spawn Group=2, Use Spawn Probability=false, UseLoopCountLimit=false) {module=SpawnBurst_Instantaneous005}
   &SpawnBurst_Instantaneous(Spawn Count=1, Spawn Group=3, Use Spawn Probability=false)
   &SpawnBurst_Instantaneous(Spawn Group=4, Use Spawn Probability=false, UseLoopCountLimit=false) {module=SpawnBurst_Instantaneous001}
   &SpawnBurst_Instantaneous(Spawn Time=1.3833, Use Spawn Probability=false, UseLoopCountLimit=false) {module=SpawnBurst_Instantaneous004}
  ]
  particleSpawnScript [
   &PartitionParticles()
   /SetVariables(Particles.VisibilityTag=&SpawnGroup()) {isSetParameters=true, module=SetVariables_#4}
   &InitializeParticle(
    Color=<rgba 0.366 0.9806 3 0.7>
    Lifetime=&MergeHelper_FloatSwitch(
     0=0.3
     1=0.4
     2=&RandomRangeFloat(Minimum=0.2, Maximum=0.3)
     3=0.3
     4=0.5
     Index=&SpawnGroup()
     Num=::NewEnumerator3
    )
    Position Mode=::NewEnumerator0
    Sprite Rotation Angle=0.002
    Sprite Rotation Mode=::NewEnumerator1
    Sprite Size=/Niagara/DynamicInputs/Helpers/MaskVector2d(
     Mask Source:NiagaraInt32=0
     Mask Parameter:NiagaraInt32=?
     Mask:NiagaraInt32=2
     Masked Vector2D:Vector2f=&RandomRangeVector2D(Minimum=<20 100>, Maximum=<100 180>)
     Passthrough Vector2D:Vector2f=<325 375>
    )
    Sprite Size Mode=::NewEnumerator4
    Write Material Random=true
   )
   &SpriteFacingAndAlignment(Sprite Alignment=<0 0 1>, WriteAlignment=true, WriteFacing=true)
   &TorusLocation(
    Align Torus With Axis=false
    Handle Radius=0
    Large Radius:NiagaraFloat=&MaskFloat(Mask Source=0, Mask Parameter=?, Mask=2, Masked Float=10, Passthrough Float=0)
    MaskSpawnGroup=false
    Module.Override Randomness=false
   )
   &AddVelocityFromPoint(
    Module.Use Falloff Distance=false
    Origin Offset=&RandomRangeVector(Minimum=<0 0 -5>, Maximum=<0 0 -10>)
    Velocity Strength=&MaskFloat(
     Mask Source=0
     Mask Parameter=?
     Mask=2
     Masked Float=&RandomRangeFloat(Minimum=600, Maximum=800)
     Passthrough Float=0
    )
   )
   &CameraOffset(Camera Offset Amount=&MaskFloat(Mask Source=1, Mask Parameter=$Particles.PartitionParticles.Partition, Mask=0, Masked Float=-30, Passthrough Float=30))
   &SystemLocation(MaskSpawnGroup=false, Offset=<0 0 10>)
   &SubUVAnimation(End Frame=0, SubUV Animation Mode:ENiagara_SubUVLookupMode=::NewEnumerator2)
  ]
  particleUpdateScript [
   &ParticleState()
   &ScaleSpriteSize(
    Scale Factor=&Multiply_Vector2D_ByFloat(
     Float=&FloatFromCurve(FloatCurve=DataInterface() {propertyValues {Curve {keys [
          {arriveTangent=0.476, interpMode=RCIM_Cubic, leaveTangent=0.476, tangentMode=RCTM_User, time=0, value=0.8}
          {arriveTangent=0.0816, tangentMode=RCTM_User, time=1, value=1.1}
         ]}}})
     Vector2D=&MergeHelper_Vector2Switch(
      0=&Vector2DFromCurve(
       Vector2Curve=DataInterface() {propertyValues {
         XCurve {
          defaultValue=1
          keys [
           {arriveTangent=0.6865, interpMode=RCIM_Cubic, leaveTangent=0.6865, tangentMode=RCTM_User, time=0, value=0.8}
           {arriveTangentWeight=0.3333, interpMode=RCIM_Cubic, tangentMode=RCTM_User, time=1, value=1.5}
          ]
         }
         YCurve {
          defaultValue=1
          keys [
           {interpMode=RCIM_Cubic, tangentMode=RCTM_User, time=0, value=0.7}
           {arriveTangent=0.007, arriveTangentWeight=0.3333, interpMode=RCIM_Cubic, leaveTangent=0.007, tangentMode=RCTM_User, time=1, value=0.5}
          ]
         }
        }}
       ScaleCurve=<1 1>
      )
      1=&Vector2DFromCurve(
       Vector2Curve=DataInterface() {propertyValues {
         XCurve {
          defaultValue=1
          keys [
           {interpMode=RCIM_Cubic, tangentMode=RCTM_User, time=0, value=1}
           {arriveTangentWeight=0.3333, interpMode=RCIM_Cubic, tangentMode=RCTM_User, time=1, value=0.8}
          ]
         }
         YCurve {
          defaultValue=1
          keys [
           {arriveTangent=0.3161, interpMode=RCIM_Cubic, leaveTangent=0.3161, tangentMode=RCTM_User, time=0, value=0.9}
           {arriveTangent=0.0449, arriveTangentWeight=0.3333, interpMode=RCIM_Cubic, leaveTangent=0.0449, tangentMode=RCTM_User, time=1, value=1}
          ]
         }
        }}
       ScaleCurve=<1 1>
      )
      2=&Vector2DFromCurve(
       Vector2Curve=DataInterface() {propertyValues {XCurve {defaultValue=0.8, keys [<at 0 1>,<at 1 0.8>]}, YCurve {defaultValue=0.7, keys [<at 0 0.5>,<at 1 1.7>]}}}
       ScaleCurve=&Multiply_Vector2D_ByFloat(Float=&FN_FrontEndFloatModifier(Gameplay=0.6, Front End=0.5), Vector2D=<1 1>)
      )
      3=&Multiply_Vector2D_ByFloat(
       Float=&FloatFromCurve(FloatCurve=DataInterface() {propertyValues {Curve {
           defaultValue=1
           keys [
            {arriveTangent=9.228, interpMode=RCIM_Cubic, leaveTangent=9.228, tangentMode=RCTM_User, time=0, value=0.1}
            {interpMode=RCIM_Cubic, tangentMode=RCTM_User, time=0.1, value=1}
           ]
          }}})
       Vector2D=<2.5 2.5>
      )
      4=<1 1>
      Index=&SpawnGroup()
      Num=::NewEnumerator3
     )
    )
    Scale Sprite Size Mode=::NewEnumerator2
   )
   &ScaleColor(
    Scale Alpha=&Multiply_Float(
     A=&FloatFromCurve(
      FloatCurve=DataInterface() {propertyValues {Curve {keys [{interpMode=RCIM_Cubic, tangentMode=RCTM_User, time=0, value=1}, <at 0.5 1>, {interpMode=RCIM_Cubic, tangentMode=RCTM_User, time=1, value=0}]}}}
      Scale Curve=&FN_FrontEndFloatModifier(Gameplay=1, Front End=0.3)
     )
     B=&MergeHelper_FloatSwitch(
      0=0.5
      1=0.5
      2=&FloatFromCurve(FloatCurve=DataInterface() {propertyValues {Curve {keys [<at 0 0>,<at 0.1 1>]}}}, Scale Curve=$System.TOD)
      3=1
      4=1
      Index=$Particles.VisibilityTag
      Num=::NewEnumerator3
     )
    )
    Scale RGB=&Multiply_VectorByFloat(
     Vector=&MergeHelper_VectorSwitch(
      0=&MakeVectorFromLinearColorRGB(Color=&ColorFromCurve(ColorCurve=DataInterface() {propertyValues {
          AlphaCurve {keys [<at 0 0.5 RCIM_Cubic>]}
          BlueCurve {keys [<at 0 1 RCIM_Cubic>,<at 1 1>]}
          GreenCurve {keys [<at 0 1 RCIM_Cubic>,<at 1 0.2918>]}
          RedCurve {keys [<at 0 1 RCIM_Cubic>,<at 1 0.1022>]}
         }}))
      1=&MakeVectorFromLinearColorRGB(Color=&ColorFromCurve(ColorCurve=DataInterface() {propertyValues {
          AlphaCurve {keys [<at 0 0.5 RCIM_Cubic>]}
          BlueCurve {keys [<at 0 1 RCIM_Cubic>,<at 1 1>]}
          GreenCurve {keys [<at 0 1 RCIM_Cubic>,<at 1 0.2918>]}
          RedCurve {keys [<at 0 1 RCIM_Cubic>,<at 1 0.1022>]}
         }}))
      2=&MakeVectorFromLinearColorRGB(Color=&ColorFromCurve(ColorCurve=DataInterface() {propertyValues {
          AlphaCurve {keys [<at 0 1 RCIM_Cubic>]}
          BlueCurve {keys [<at 0 3 RCIM_Cubic>,<at 1 0.3198>]}
          GreenCurve {keys [<at 0 1.5339 RCIM_Cubic>,<at 1 0.0933>]}
          RedCurve {keys [<at 0 1.2406 RCIM_Cubic>,<at 1 0.0327>]}
         }}))
      3=&MakeVectorFromLinearColorRGB(Color=&ColorFromCurve(ColorCurve=DataInterface() {propertyValues {
          AlphaCurve {keys [<at 0 1 RCIM_Cubic>]}
          BlueCurve {keys [<at 0 1 RCIM_Cubic>,<at 1 0.3198>]}
          GreenCurve {keys [<at 0 0.5117 RCIM_Cubic>,<at 1 0.0933>]}
          RedCurve {keys [<at 0 0.414 RCIM_Cubic>,<at 1 0.0327>]}
         }}))
      4=&MakeVectorFromLinearColorRGB(Color=&ColorFromCurve(ColorCurve=DataInterface() {propertyValues {
          AlphaCurve {keys [<at 0 1 RCIM_Cubic>]}
          BlueCurve {keys [<at 0 1000 RCIM_Cubic>,<at 0.9007 0.3198>,<at 1 0>]}
          GreenCurve {keys [<at 0 427.9519 RCIM_Cubic>,<at 0.9007 0.0933>,<at 1 0>]}
          RedCurve {keys [<at 0 313.5418 RCIM_Cubic>,<at 0.9007 0.0327>,<at 1 0>]}
         }}))
      Index=&SpawnGroup()
      Num=::NewEnumerator3
     )
     Float=&FN_FrontEndFloatModifier(Gameplay=0.6, Front End=1)
    )
    ScaleA=true
    ScaleRGB=true
   )
   &DynamicMaterialParameters(
    Index 0 Param 1=&MergeHelper_FloatSwitch(
     0=&FloatFromCurve(FloatCurve=DataInterface() {propertyValues {Curve {keys [
          {arriveTangent=2.4757, interpMode=RCIM_Cubic, leaveTangent=2.4757, tangentMode=RCTM_User, time=0, value=0.6}
          {interpMode=RCIM_Cubic, tangentMode=RCTM_User, time=0.1184, value=0.8}
          {arriveTangent=-1.0941, interpMode=RCIM_Cubic, leaveTangent=-1.0941, tangentMode=RCTM_User, time=0.5, value=0.6}
          {arriveTangent=-0.1249, interpMode=RCIM_Cubic, leaveTangent=-0.1249, tangentMode=RCTM_User, time=1, value=0.3}
         ]}}})
     1=&FloatFromCurve(FloatCurve=DataInterface() {propertyValues {Curve {keys [
          {arriveTangent=2.4757, interpMode=RCIM_Cubic, leaveTangent=2.4757, tangentMode=RCTM_User, time=0, value=0.6}
          <at 0.23 0.8>
          {arriveTangent=-1.0941, interpMode=RCIM_Cubic, leaveTangent=-1.0941, tangentMode=RCTM_User, time=0.5, value=0.5}
          {arriveTangent=-0.1249, interpMode=RCIM_Cubic, leaveTangent=-0.1249, tangentMode=RCTM_User, time=1, value=0.3}
         ]}}})
     2=&FloatFromCurve(FloatCurve=DataInterface() {propertyValues {Curve {keys [
          {arriveTangent=0.8473, interpMode=RCIM_Cubic, leaveTangent=0.8473, tangentMode=RCTM_User, time=0, value=0.5362}
          {arriveTangent=-0.1249, interpMode=RCIM_Cubic, leaveTangent=-0.1249, tangentMode=RCTM_User, time=1, value=0.3}
         ]}}})
     3=&RandomRangeFloat(Minimum=0.4, Maximum=0.6)
     4=0
     Index=&SpawnGroup()
     Num=::NewEnumerator3
    )
    Index 0 Param 4=&Add_Float(
     A=&FloatFromCurve(FloatCurve=DataInterface() {propertyValues {Curve {
         defaultValue=1
         keys [
          {arriveTangent=1.2435, interpMode=RCIM_Cubic, leaveTangent=1.2435, tangentMode=RCTM_User, time=0, value=0}
          {interpMode=RCIM_Cubic, tangentMode=RCTM_User, time=1, value=1}
         ]
        }}})
     B=&RandomRangeFloat(Minimum=0, Maximum=1)
    )
    Module.Param0WriteEnabled=true
    Module.Param1WriteEnabled=true
    Module.Param2WriteEnabled=true
    Module.Param3WriteEnabled=true
   )
   &FN_Light_Attributes(
    Color=$Particles.Initial.Color
    Create Color Attribute=true
    Create Exponent Attribute=true
    Create Position Attribute=true
    Create Radius Attribute=true
    Enable Particle Light Enable=true
    Enable Visibility Tag=true
    Exponent=3
    Position=<0 0 50>
    Radius=&FloatFromCurve(
     FloatCurve=DataInterface() {propertyValues {Curve {keys [
         {arriveTangent=4.9681, interpMode=RCIM_Cubic, leaveTangent=4.9681, tangentMode=RCTM_User, time=0, value=0.5}
         {interpMode=RCIM_Cubic, tangentMode=RCTM_User, time=0.1, value=1}
         <at 0.9813 0.5>
        ]}}}
     Scale Curve=225
    )
    Scale Color=&FloatFromCurve(FloatCurve=DataInterface() {propertyValues {Curve {keys [{arriveTangent=-2.4738, interpMode=RCIM_Cubic, leaveTangent=-2.4738, tangentMode=RCTM_User, time=0, value=1}, <at 1 0>]}}})
   )
   &Drag(Drag=6, Use Linear Drag=true, Use Rotational Drag=false)
   &SolveForcesAndVelocity(Clamp Velocity=false, Limit Acceleration=false)
  ]
  rendererClasses [&NiagaraSpriteRendererProperties, &NiagaraLightRendererProperties]
  renderers [
   {rendererClass=&NiagaraSpriteRendererProperties}
   {rendererClass=&NiagaraSpriteRendererProperties, rendererIndex=1}
   {rendererClass=&NiagaraSpriteRendererProperties, rendererIndex=2}
   {rendererClass=&NiagaraSpriteRendererProperties, rendererIndex=3}
   {rendererClass=&NiagaraLightRendererProperties, rendererIndex=4}
  ]
 }
 Burst_spikes001 {
  bEnabled=false
  emitterUpdateScript [
   &EmitterState(Life Cycle Mode=::NewEnumerator1, Loop Behavior=::NewEnumerator1, Loop Delay=$System.AntiBurstDelay, Loop Duration=0.2, UseLoopDelay=true)
   &SpawnBurst_Instantaneous(Use Spawn Probability=false, UseLoopCountLimit=false) {module=SpawnBurst_Instantaneous002}
   &SpawnBurst_Instantaneous(Spawn Group=1, Spawn Time=0.1, Use Spawn Probability=false, UseLoopCountLimit=false) {module=SpawnBurst_Instantaneous003}
   &SpawnRate(SpawnRate=14, Use Spawn Probability=false) {enabled=false, module=SpawnRate001}
   &SpawnRate(SpawnGroup=2, SpawnRate=10, Use Spawn Probability=false) {enabled=false}
   &SpawnBurst_Instantaneous(Spawn Count=10, Spawn Group=2, Use Spawn Probability=false, UseLoopCountLimit=false) {module=SpawnBurst_Instantaneous005}
   &SpawnBurst_Instantaneous(Spawn Count=1, Spawn Group=3, Use Spawn Probability=false)
   &SpawnBurst_Instantaneous(Spawn Group=4, Use Spawn Probability=false, UseLoopCountLimit=false) {module=SpawnBurst_Instantaneous001}
   &SpawnBurst_Instantaneous(Spawn Time=1.3833, Use Spawn Probability=false, UseLoopCountLimit=false) {module=SpawnBurst_Instantaneous004}
  ]
  particleSpawnScript [
   &PartitionParticles()
   /SetVariables(Particles.VisibilityTag=&SpawnGroup()) {isSetParameters=true, module=SetVariables_#4}
   &InitializeParticle(
    Color=<rgba 0.366 0.9806 3 0.7>
    Lifetime=&MergeHelper_FloatSwitch(
     0=0.3
     1=0.4
     2=&RandomRangeFloat(Minimum=0.2, Maximum=0.3)
     3=0.3
     4=0.5
     Index=&SpawnGroup()
     Num=::NewEnumerator3
    )
    Position Mode=::NewEnumerator0
    Sprite Rotation Angle=0.002
    Sprite Rotation Mode=::NewEnumerator1
    Sprite Size=&MaskVector2d(
     Mask Source:NiagaraInt32=0
     Mask Parameter:NiagaraInt32=?
     Mask:NiagaraInt32=2
     Masked Vector2D:Vector2f=&RandomRangeVector2D(Minimum=<20 100>, Maximum=<100 180>)
     Passthrough Vector2D:Vector2f=<325 375>
    )
    Sprite Size Mode=::NewEnumerator4
    Write Material Random=true
   )
   &SpriteFacingAndAlignment(Sprite Alignment=<0 0 1>, WriteAlignment=true, WriteFacing=true)
   &TorusLocation(
    Align Torus With Axis=false
    Handle Radius=0
    Large Radius:NiagaraFloat=&MaskFloat(Mask Source=0, Mask Parameter=?, Mask=2, Masked Float=10, Passthrough Float=0)
    MaskSpawnGroup=false
    Module.Override Randomness=false
   )
   &AddVelocityFromPoint(
    Module.Use Falloff Distance=false
    Origin Offset=&RandomRangeVector(Minimum=<0 0 -5>, Maximum=<0 0 -10>)
    Velocity Strength=&MaskFloat(
     Mask Source=0
     Mask Parameter=?
     Mask=2
     Masked Float=&RandomRangeFloat(Minimum=600, Maximum=800)
     Passthrough Float=0
    )
   )
   &CameraOffset(Camera Offset Amount=&MaskFloat(Mask Source=1, Mask Parameter=$Particles.PartitionParticles.Partition, Mask=0, Masked Float=-30, Passthrough Float=30))
   &SystemLocation(MaskSpawnGroup=false, Offset=<0 0 10>)
   &SubUVAnimation(End Frame=0, SubUV Animation Mode:ENiagara_SubUVLookupMode=::NewEnumerator2)
  ]
  particleUpdateScript [
   &ParticleState()
   &ScaleSpriteSize(
    Scale Factor=&Multiply_Vector2D_ByFloat(
     Float=&FloatFromCurve(FloatCurve=DataInterface() {propertyValues {Curve {keys [
          {arriveTangent=0.476, interpMode=RCIM_Cubic, leaveTangent=0.476, tangentMode=RCTM_User, time=0, value=0.8}
          {arriveTangent=0.0816, tangentMode=RCTM_User, time=1, value=1.1}
         ]}}})
     Vector2D=&MergeHelper_Vector2Switch(
      0=&Vector2DFromCurve(
       Vector2Curve=DataInterface() {propertyValues {
         XCurve {
          defaultValue=1
          keys [
           {arriveTangent=0.6865, interpMode=RCIM_Cubic, leaveTangent=0.6865, tangentMode=RCTM_User, time=0, value=0.8}
           {arriveTangentWeight=0.3333, interpMode=RCIM_Cubic, tangentMode=RCTM_User, time=1, value=1.5}
          ]
         }
         YCurve {
          defaultValue=1
          keys [
           {interpMode=RCIM_Cubic, tangentMode=RCTM_User, time=0, value=0.7}
           {arriveTangent=0.007, arriveTangentWeight=0.3333, interpMode=RCIM_Cubic, leaveTangent=0.007, tangentMode=RCTM_User, time=1, value=0.5}
          ]
         }
        }}
       ScaleCurve=<1 1>
      )
      1=&Vector2DFromCurve(
       Vector2Curve=DataInterface() {propertyValues {
         XCurve {
          defaultValue=1
          keys [
           {interpMode=RCIM_Cubic, tangentMode=RCTM_User, time=0, value=1}
           {arriveTangentWeight=0.3333, interpMode=RCIM_Cubic, tangentMode=RCTM_User, time=1, value=0.8}
          ]
         }
         YCurve {
          defaultValue=1
          keys [
           {arriveTangent=0.3161, interpMode=RCIM_Cubic, leaveTangent=0.3161, tangentMode=RCTM_User, time=0, value=0.9}
           {arriveTangent=0.0449, arriveTangentWeight=0.3333, interpMode=RCIM_Cubic, leaveTangent=0.0449, tangentMode=RCTM_User, time=1, value=1}
          ]
         }
        }}
       ScaleCurve=<1 1>
      )
      2=&Vector2DFromCurve(
       Vector2Curve=DataInterface() {propertyValues {XCurve {defaultValue=0.8, keys [<at 0 1>,<at 1 0.8>]}, YCurve {defaultValue=0.7, keys [<at 0 0.5>,<at 1 1.7>]}}}
       ScaleCurve=&Multiply_Vector2D_ByFloat(Float=&FN_FrontEndFloatModifier(Gameplay=0.6, Front End=0.5), Vector2D=<1 1>)
      )
      3=&Multiply_Vector2D_ByFloat(
       Float=&FloatFromCurve(FloatCurve=DataInterface() {propertyValues {Curve {
           defaultValue=1
           keys [
            {arriveTangent=9.228, interpMode=RCIM_Cubic, leaveTangent=9.228, tangentMode=RCTM_User, time=0, value=0.1}
            {interpMode=RCIM_Cubic, tangentMode=RCTM_User, time=0.1, value=1}
           ]
          }}})
       Vector2D=<2.5 2.5>
      )
      4=<1 1>
      Index=&SpawnGroup()
      Num=::NewEnumerator3
     )
    )
    Scale Sprite Size Mode=::NewEnumerator2
   )
   &ScaleColor(
    Scale Alpha=&Multiply_Float(
     A=&FloatFromCurve(
      FloatCurve=DataInterface() {propertyValues {Curve {keys [{interpMode=RCIM_Cubic, tangentMode=RCTM_User, time=0, value=1}, <at 0.5 1>, {interpMode=RCIM_Cubic, tangentMode=RCTM_User, time=1, value=0}]}}}
      Scale Curve=&FN_FrontEndFloatModifier(Gameplay=1, Front End=0.3)
     )
     B=&MergeHelper_FloatSwitch(
      0=0.5
      1=0.5
      2=&FloatFromCurve(FloatCurve=DataInterface() {propertyValues {Curve {keys [<at 0 0>,<at 0.1 1>]}}}, Scale Curve=$System.TOD)
      3=1
      4=1
      Index=$Particles.VisibilityTag
      Num=::NewEnumerator3
     )
    )
    Scale RGB=&Multiply_VectorByFloat(
     Vector=&MergeHelper_VectorSwitch(
      0=&MakeVectorFromLinearColorRGB(Color=&ColorFromCurve(ColorCurve=DataInterface() {propertyValues {
          AlphaCurve {keys [<at 0 0.5 RCIM_Cubic>]}
          BlueCurve {keys [<at 0 1 RCIM_Cubic>,<at 1 1>]}
          GreenCurve {keys [<at 0 1 RCIM_Cubic>,<at 1 0.2918>]}
          RedCurve {keys [<at 0 1 RCIM_Cubic>,<at 1 0.1022>]}
         }}))
      1=&MakeVectorFromLinearColorRGB(Color=&ColorFromCurve(ColorCurve=DataInterface() {propertyValues {
          AlphaCurve {keys [<at 0 0.5 RCIM_Cubic>]}
          BlueCurve {keys [<at 0 1 RCIM_Cubic>,<at 1 1>]}
          GreenCurve {keys [<at 0 1 RCIM_Cubic>,<at 1 0.2918>]}
          RedCurve {keys [<at 0 1 RCIM_Cubic>,<at 1 0.1022>]}
         }}))
      2=&MakeVectorFromLinearColorRGB(Color=&ColorFromCurve(ColorCurve=DataInterface() {propertyValues {
          AlphaCurve {keys [<at 0 1 RCIM_Cubic>]}
          BlueCurve {keys [<at 0 3 RCIM_Cubic>,<at 1 0.3198>]}
          GreenCurve {keys [<at 0 1.5339 RCIM_Cubic>,<at 1 0.0933>]}
          RedCurve {keys [<at 0 1.2406 RCIM_Cubic>,<at 1 0.0327>]}
         }}))
      3=&MakeVectorFromLinearColorRGB(Color=&ColorFromCurve(ColorCurve=DataInterface() {propertyValues {
          AlphaCurve {keys [<at 0 1 RCIM_Cubic>]}
          BlueCurve {keys [<at 0 1 RCIM_Cubic>,<at 1 0.3198>]}
          GreenCurve {keys [<at 0 0.5117 RCIM_Cubic>,<at 1 0.0933>]}
          RedCurve {keys [<at 0 0.414 RCIM_Cubic>,<at 1 0.0327>]}
         }}))
      4=&MakeVectorFromLinearColorRGB(Color=&ColorFromCurve(ColorCurve=DataInterface() {propertyValues {
          AlphaCurve {keys [<at 0 1 RCIM_Cubic>]}
          BlueCurve {keys [<at 0 1000 RCIM_Cubic>,<at 0.9007 0.3198>,<at 1 0>]}
          GreenCurve {keys [<at 0 427.9519 RCIM_Cubic>,<at 0.9007 0.0933>,<at 1 0>]}
          RedCurve {keys [<at 0 313.5418 RCIM_Cubic>,<at 0.9007 0.0327>,<at 1 0>]}
         }}))
      Index=&SpawnGroup()
      Num=::NewEnumerator3
     )
     Float=&FN_FrontEndFloatModifier(Gameplay=0.6, Front End=1)
    )
    ScaleA=true
    ScaleRGB=true
   )
   &DynamicMaterialParameters(
    Index 0 Param 1=&MergeHelper_FloatSwitch(
     0=&FloatFromCurve(FloatCurve=DataInterface() {propertyValues {Curve {keys [
          {arriveTangent=2.4757, interpMode=RCIM_Cubic, leaveTangent=2.4757, tangentMode=RCTM_User, time=0, value=0.6}
          {interpMode=RCIM_Cubic, tangentMode=RCTM_User, time=0.1184, value=0.8}
          {arriveTangent=-1.0941, interpMode=RCIM_Cubic, leaveTangent=-1.0941, tangentMode=RCTM_User, time=0.5, value=0.6}
          {arriveTangent=-0.1249, interpMode=RCIM_Cubic, leaveTangent=-0.1249, tangentMode=RCTM_User, time=1, value=0.3}
         ]}}})
     1=&FloatFromCurve(FloatCurve=DataInterface() {propertyValues {Curve {keys [
          {arriveTangent=2.4757, interpMode=RCIM_Cubic, leaveTangent=2.4757, tangentMode=RCTM_User, time=0, value=0.6}
          <at 0.23 0.8>
          {arriveTangent=-1.0941, interpMode=RCIM_Cubic, leaveTangent=-1.0941, tangentMode=RCTM_User, time=0.5, value=0.5}
          {arriveTangent=-0.1249, interpMode=RCIM_Cubic, leaveTangent=-0.1249, tangentMode=RCTM_User, time=1, value=0.3}
         ]}}})
     2=&FloatFromCurve(FloatCurve=DataInterface() {propertyValues {Curve {keys [
          {arriveTangent=0.8473, interpMode=RCIM_Cubic, leaveTangent=0.8473, tangentMode=RCTM_User, time=0, value=0.5362}
          {arriveTangent=-0.1249, interpMode=RCIM_Cubic, leaveTangent=-0.1249, tangentMode=RCTM_User, time=1, value=0.3}
         ]}}})
     3=&RandomRangeFloat(Minimum=0.4, Maximum=0.6)
     4=0
     Index=&SpawnGroup()
     Num=::NewEnumerator3
    )
    Index 0 Param 4=&Add_Float(
     A=&FloatFromCurve(FloatCurve=DataInterface() {propertyValues {Curve {
         defaultValue=1
         keys [
          {arriveTangent=1.2435, interpMode=RCIM_Cubic, leaveTangent=1.2435, tangentMode=RCTM_User, time=0, value=0}
          {interpMode=RCIM_Cubic, tangentMode=RCTM_User, time=1, value=1}
         ]
        }}})
     B=&RandomRangeFloat(Minimum=0, Maximum=1)
    )
    Module.Param0WriteEnabled=true
    Module.Param1WriteEnabled=true
    Module.Param2WriteEnabled=true
    Module.Param3WriteEnabled=true
   )
   /SetVariables(Particles.OutlineColor:LinearColor=&MakeLinearColorFromVectorAndFloat(
     "Vector (RGB)"=&Multiply_Vector(
      A=&MakeVectorFromLinearColorRGB(Color=<rgba 2 2 2 0.5>)
      B=&MakeVectorFromLinearColorRGB(Color=&ColorFromCurve(ColorCurve=DataInterface() {propertyValues {
          AlphaCurve {keys [<at 0 1 RCIM_Cubic>]}
          BlueCurve {keys [<at 0.2266 1.5 RCIM_Cubic>,<at 1 1>]}
          GreenCurve {keys [<at 0.2266 1.0431 RCIM_Cubic>,<at 1 0.2918>]}
          RedCurve {keys [<at 0.2266 0.6 RCIM_Cubic>,<at 1 0.1022>]}
         }}))
     )
     "Float (Alpha)"=&FloatFromCurve(
      FloatCurve=DataInterface() {propertyValues {Curve {keys [{interpMode=RCIM_Cubic, tangentMode=RCTM_User, time=0, value=1}, <at 0.5 1>, {interpMode=RCIM_Cubic, tangentMode=RCTM_User, time=1, value=0}]}}}
      Scale Curve=0.5
     )
    )) {isSetParameters=true}
   &FN_Light_Attributes(
    Color=$Particles.Initial.Color
    Create Color Attribute=true
    Create Exponent Attribute=true
    Create Position Attribute=true
    Create Radius Attribute=true
    Enable Particle Light Enable=true
    Enable Visibility Tag=true
    Exponent=3
    Position=<0 0 50>
    Radius=&FloatFromCurve(
     FloatCurve=DataInterface() {propertyValues {Curve {keys [
         {arriveTangent=4.9681, interpMode=RCIM_Cubic, leaveTangent=4.9681, tangentMode=RCTM_User, time=0, value=0.5}
         {interpMode=RCIM_Cubic, tangentMode=RCTM_User, time=0.1, value=1}
         <at 0.9813 0.5>
        ]}}}
     Scale Curve=225
    )
    Scale Color=&FloatFromCurve(FloatCurve=DataInterface() {propertyValues {Curve {keys [{arriveTangent=-2.4738, interpMode=RCIM_Cubic, leaveTangent=-2.4738, tangentMode=RCTM_User, time=0, value=1}, <at 1 0>]}}})
   )
   &Drag(Drag=6, Use Linear Drag=true, Use Rotational Drag=false)
   &SolveForcesAndVelocity(Clamp Velocity=false, Limit Acceleration=false)
  ]
  rendererClasses [&NiagaraSpriteRendererProperties, &NiagaraLightRendererProperties]
  renderers [
   {rendererClass=&NiagaraSpriteRendererProperties}
   {rendererClass=&NiagaraSpriteRendererProperties, rendererIndex=1}
   {rendererClass=&NiagaraSpriteRendererProperties, rendererIndex=2}
   {rendererClass=&NiagaraSpriteRendererProperties, rendererIndex=3}
   {rendererClass=&NiagaraSpriteRendererProperties, rendererIndex=4}
   {rendererClass=&NiagaraSpriteRendererProperties, rendererIndex=5}
   {rendererClass=&NiagaraSpriteRendererProperties, rendererIndex=6}
   {rendererClass=&NiagaraLightRendererProperties, rendererIndex=7}
  ]
 }
 Glow_Body {
  emitterUpdateScript [
   /SetVariables(Emitter.GlowOutlineThickness:NiagaraFloat=&FloatFromCurve(
     FloatCurve=DataInterface() {propertyValues {Curve {keys [
         {arriveTangent=55.7531, interpMode=RCIM_Cubic, leaveTangent=55.7531, tangentMode=RCTM_User, time=0, value=0}
         {arriveTangent=10, interpMode=RCIM_Cubic, leaveTangent=10, time=0.0454, value=3}
         <at 0.1 1 RCIM_Constant>
         {interpMode=RCIM_Cubic, tangentMode=RCTM_User, time=0.72, value=5}
         <at 1 0>
        ]}}}
     CurveIndex=$Emitter.Age
    )) {isSetParameters=true}
   &EmitterState(Life Cycle Mode=::NewEnumerator1, Loop Behavior=::NewEnumerator1, Loop Duration=0.1, UseLoopDelay=false)
   &SpawnBurst_Instantaneous(Spawn Count=7, Use Spawn Probability=false, UseLoopCountLimit=false)
  ]
  particleSpawnScript [
   &InitializeParticle(Lifetime=1.5, Mesh Scale Mode=::NewEnumerator4, Position Offset=<0 0 100>, UsePositionOffset=true, Write Material Random=true)
   &PartitionParticles(Number Of Partitions=9)
  ]
  particleUpdateScript [
   &ParticleState()
   &ScaleColor(
    Scale Alpha=&FloatFromCurve(
     FloatCurve=DataInterface() {propertyValues {Curve {keys [
         <at 0 1 RCIM_Cubic>
         <at 0.1943 1>
         <at 0.444 1 RCIM_Constant>
         <at 0.4869 1 RCIM_Constant>
         <at 0.5709 1 RCIM_Constant>
         <at 0.6187 1 RCIM_Constant>
         {interpMode=RCIM_Cubic, tangentMode=RCTM_User, time=0.7152, value=1}
         <at 1 0>
        ]}}}
     Scale Curve=100
    )
    Scale RGB=&Multiply_VectorByFloat(Vector=<1 1 1>, Float=1)
    ScaleA=true
    ScaleRGB=false
   ) {module=ScaleColor001}
   &SkeletalMeshLocation(
    Bone / Socket Index=&Add_Integer(A=$Particles.PartitionParticles.Partition, B=1)
    Bone Sampling Mode=::NewEnumerator4
    MaskSpawnGroup=false
    Orientation Sampling=::NewEnumerator1
    Skeletal Mesh=$System.Skeletal Mesh
   )
   &CameraOffset(Camera Offset Amount=30)
  ]
  rendererClasses [&NiagaraMeshRendererProperties]
  renderers [{rendererClass=&NiagaraMeshRendererProperties, rendererIndex=0}]
 }
 Inward_Wind {
  emitterUpdateScript [
   &EmitterState(Life Cycle Mode=::NewEnumerator1, Loop Behavior=::NewEnumerator1, Loop Duration=0.5, UseLoopDelay=false)
   &SpawnBurst_Instantaneous(Spawn Count=2, Use Spawn Probability=false, UseLoopCountLimit=false)
   &SpawnRate(
    SpawnRate=&FloatFromCurve(
     FloatCurve=DataInterface() {propertyValues {Curve {keys [
         {arriveTangent=4.7309, interpMode=RCIM_Cubic, leaveTangent=4.7309, leaveTangentWeight=19.8086, tangentMode=RCTM_User, time=0, value=1}
         {arriveTangentWeight=0.3333, interpMode=RCIM_Cubic, tangentMode=RCTM_User, time=0.5, value=12}
        ]}}}
     CurveIndex=$Emitter.NormalizedLoopAge
    )
    Use Spawn Probability=false
   )
   &SpawnRate(
    SpawnGroup=1
    SpawnRate=&FloatFromCurve(
     FloatCurve=DataInterface() {propertyValues {Curve {keys [
         {arriveTangent=4.7309, interpMode=RCIM_Cubic, leaveTangent=4.7309, leaveTangentWeight=19.8086, tangentMode=RCTM_User, time=0, value=5}
         {arriveTangentWeight=0.3333, interpMode=RCIM_Cubic, tangentMode=RCTM_User, time=0.5, value=20}
        ]}}}
     CurveIndex=$Emitter.NormalizedLoopAge
    )
    Use Spawn Probability=false
   ) {module=SpawnRate001}
   &SpawnBurst_Instantaneous(Spawn Count=3, Spawn Group=1, Spawn Time=0.35, Use Spawn Probability=false, UseLoopCountLimit=false) {module=SpawnBurst_Instantaneous001}
  ]
  particleSpawnScript [
   /SetVariables(Particles.VisibilityTag=&SpawnGroup()) {isSetParameters=true}
   &InitializeParticle(
    Lifetime=&MergeHelper_FloatSwitch(
     0=&Lerp_Float(A=0.7, B=0.2, Alpha=$Emitter.NormalizedLoopAge)
     1=0.5
     2=0
     3=?
     Index=&SpawnGroup()
     Num=::NewEnumerator1
    )
    Mesh Scale=&MergeHelper_VectorSwitch(
     0=&Multiply_VectorByFloat(
      Vector=&MakeVector(X=0.55, Y=0.55, Z=&RandomRangeFloat(Minimum=1, Maximum=2))
      Float=&RandomRangeFloat(Minimum=1, Maximum=1.3)
     )
     1=&RandomRangeVector(Minimum=<1 0.4 0.4>, Maximum=<3 0.6 0.6>)
     2=<1 1 1>
     3=?
     Index=&SpawnGroup()
     Num=::NewEnumerator1
    )
    Mesh Scale Mode=::NewEnumerator4
    UsePositionOffset=false
    Write Material Random=true
   )
   &SystemLocation(
    MaskSpawnGroup=false
    Offset=&MergeHelper_VectorSwitch(0=<0 0 20>, 1=<0 0 0>, 2=?, 3=?, Index=&SpawnGroup(), Num=::NewEnumerator0)
   )
   &InitialMeshOrientation(
    Mesh Orientation Mode=::NewEnumerator2
    Orientation Coordinate Space=::Local
    Rotation=&MergeHelper_VectorSwitch(
     0=&RandomRangeVector(Minimum=<0.07 -0.07 -1>, Maximum=<-0.1 0.07 1>)
     1=&RandomRangeVector(Minimum=<0 0 0>, Maximum=<0 0 1>)
     2=<0 0 0>
     3=?
     Index=&SpawnGroup()
     Num=::NewEnumerator1
    )
    Use Rotation Vector=true
   )
   &ShapeLocation(MaskSpawnGroup=false, Override Seed=false, Ring Radius=15, Shape Primitive=::NewEnumerator4, UseOwnerScale=true)
  ]
  particleUpdateScript [
   &ParticleState()
   &ScaleMeshSize(Scale Factor=&MergeHelper_VectorSwitch(
     0=&Multiply_Vector(
      A=&MakeVector(
       X=&FloatFromCurve(
        FloatCurve=DataInterface() {propertyValues {Curve {keys [
            {interpMode=RCIM_Cubic, tangentMode=RCTM_User, time=0, value=1.5}
            {arriveTangent=-1.6426, arriveTangentWeight=0.3836, interpMode=RCIM_Cubic, leaveTangent=-1.6426, tangentMode=RCTM_User, time=1, value=0.1}
           ]}}}
        Scale Curve=1.5
       )
       Y=&FloatFromCurve(
        FloatCurve=DataInterface() {propertyValues {Curve {keys [
            {interpMode=RCIM_Cubic, tangentMode=RCTM_User, time=0, value=1.5}
            {arriveTangent=-1.9074, arriveTangentWeight=0.3836, interpMode=RCIM_Cubic, leaveTangent=-1.9074, tangentMode=RCTM_User, time=1, value=0.1}
           ]}}}
        Scale Curve=1.5
       )
       Z=&FloatFromCurve(FloatCurve=DataInterface() {propertyValues {Curve {
           defaultValue=1.5
           keys [{interpMode=RCIM_Cubic, tangentMode=RCTM_User, time=0, value=1}, {interpMode=RCIM_Cubic, tangentMode=RCTM_User, time=1, value=2}]
          }}})
      )
      B=&VectorFromCurve(
       VectorCurve=DataInterface() {propertyValues {
         XCurve {keys [{interpMode=RCIM_Cubic, tangentMode=RCTM_User, time=0, value=1}, {interpMode=RCIM_Cubic, tangentMode=RCTM_User, time=0.7, value=2}, <at 1 1>]}
         YCurve {keys [{interpMode=RCIM_Cubic, tangentMode=RCTM_User, time=0, value=1}, {interpMode=RCIM_Cubic, tangentMode=RCTM_User, time=0.7, value=2}, <at 1 1>]}
         ZCurve {keys [
           {interpMode=RCIM_Cubic, tangentMode=RCTM_User, time=0, value=0.3}
           {interpMode=RCIM_Cubic, tangentMode=RCTM_User, time=0.3, value=1}
           <at 0.8 0.1>
          ]}
        }}
       CurveIndex=$Emitter.Age
      )
     )
     1=&Multiply_VectorByFloat(
      Vector=<1 1 1>
      Float=&FloatFromCurve(
       FloatCurve=DataInterface() {propertyValues {Curve {keys [{interpMode=RCIM_Cubic, tangentMode=RCTM_User, time=0, value=1}, <at 1 0>]}}}
       Scale Curve=3.5
      )
     )
     2=<0 0 0>
     3=?
     Index=&SpawnGroup()
     Num=::NewEnumerator1
    ))
   &UpdateMeshOrientation(
    Rotation Rate=&MergeHelper_FloatSwitch(
     0=&FloatFromCurve(
      FloatCurve=DataInterface() {propertyValues {Curve {keys [
          {arriveTangent=0.0755, interpMode=RCIM_Cubic, leaveTangent=0.0755, tangentMode=RCTM_User, time=0, value=0.3}
          {arriveTangent=1.6672, arriveTangentWeight=0.3333, interpMode=RCIM_Cubic, leaveTangent=1.6672, tangentMode=RCTM_User, time=1, value=1}
         ]}}}
      CurveIndex=$Emitter.Age
      Scale Curve=8
     )
     1=0
     2=0
     3=?
     Index=&SpawnGroup()
     Num=::NewEnumerator1
    )
    Rotation Vector=<0 0 1>
   )
   &ScaleColor(
    Scale Alpha=&Multiply_Float(
     A=&FloatFromCurve(
      FloatCurve=DataInterface() {propertyValues {Curve {keys [
          {interpMode=RCIM_Cubic, tangentMode=RCTM_User, time=0, value=0}
          {interpMode=RCIM_Cubic, tangentMode=RCTM_User, time=0.2, value=1}
          {arriveTangentWeight=0.2771, interpMode=RCIM_Cubic, leaveTangentWeight=0.1497, tangentMode=RCTM_User, time=0.7194, value=0.9865}
          {tangentMode=RCTM_User, time=1, value=0}
         ]}}}
      Scale Curve=&FloatFromCurve(
       FloatCurve=DataInterface() {propertyValues {Curve {keys [{interpMode=RCIM_Cubic, tangentMode=RCTM_User, time=0.7, value=1}, <at 1.5 0>]}}}
       CurveIndex=$Emitter.Age
      )
     )
     B=0.1
    )
    Scale RGB=&Multiply_VectorByFloat(
     Vector=&MergeHelper_VectorSwitch(
      0=&Multiply_VectorByFloat(
       Vector=&MakeVectorFromLinearColorRGB(Color=&ColorFromCurve(
         ColorCurve=DataInterface() {propertyValues {
           AlphaCurve {keys [<at 0 1 RCIM_Cubic>,<at 1 0 RCIM_Cubic>]}
           BlueCurve {keys [
             {interpMode=RCIM_Cubic, leaveTangentWeight=0.121, tangentMode=RCTM_User, time=0.0959, value=1}
             {interpMode=RCIM_Cubic, leaveTangent=786.035, time=0.4588, value=7}
            ]}
           GreenCurve {keys [
             {interpMode=RCIM_Cubic, leaveTangentWeight=0.121, tangentMode=RCTM_User, time=0.0959, value=1}
             {interpMode=RCIM_Cubic, leaveTangent=157.9159, time=0.4588, value=1.9758}
            ]}
           RedCurve {keys [
             {interpMode=RCIM_Cubic, leaveTangentWeight=0.121, tangentMode=RCTM_User, time=0.0959, value=1}
             {interpMode=RCIM_Cubic, leaveTangent=32.2921, time=0.4588, value=0.301}
            ]}
          }}
         CurveIndex=$Emitter.Age
        ))
       Float=&FN_FrontEndFloatModifier(Gameplay=0.7, Front End=0.7)
      )
      1=&MakeVectorFromLinearColorRGB(Color=&ColorFromCurve(
        ColorCurve=DataInterface() {propertyValues {
          AlphaCurve {keys [<at 0 0.5 RCIM_Cubic>,<at 0.852 1>,<at 1 0 RCIM_Cubic>]}
          BlueCurve {keys [
            {interpMode=RCIM_Cubic, leaveTangentWeight=0.121, tangentMode=RCTM_User, time=0.0959, value=0.5}
            {arriveTangentWeight=0.235, interpMode=RCIM_Cubic, leaveTangentWeight=0.0723, tangentMode=RCTM_User, time=0.5074, value=1}
            <at 0.6212 3>
           ]}
          GreenCurve {keys [
            {interpMode=RCIM_Cubic, leaveTangentWeight=0.121, tangentMode=RCTM_User, time=0.0959, value=0.5}
            {arriveTangentWeight=0.235, interpMode=RCIM_Cubic, leaveTangentWeight=0.0723, tangentMode=RCTM_User, time=0.5074, value=0.4352}
            <at 0.6212 0.8555>
           ]}
          RedCurve {keys [
            {interpMode=RCIM_Cubic, leaveTangentWeight=0.121, tangentMode=RCTM_User, time=0.0959, value=0.5}
            {arriveTangentWeight=0.235, interpMode=RCIM_Cubic, leaveTangentWeight=0.0723, tangentMode=RCTM_User, time=0.5074, value=0.2469}
            <at 0.6212 0.1406>
           ]}
         }}
        CurveIndex=$Emitter.Age
       ))
      2=<0 0 0>
      3=?
      Index=&SpawnGroup()
      Num=::NewEnumerator1
     )
     Float=$System.TOD
    )
    ScaleA=true
    ScaleRGB=true
   )
   &AccelerationForce(Acceleration=&MaskVector(Mask Source=0, Mask Parameter=?, Mask=0, Masked Vector=<0 0 300>, Passthrough Vector=<0 0 0>))
   &SolveForcesAndVelocity(Clamp Velocity=false, Limit Acceleration=false)
   /SetVariables(Particles.GroundDynP00:Vector4f=&MakeVector4(
     X=&FloatFromCurve(FloatCurve=DataInterface() {propertyValues {Curve {keys [
          <at 0.5 0 RCIM_Cubic>
          {arriveTangent=2.6293, arriveTangentWeight=0.1667, interpMode=RCIM_Cubic, leaveTangent=2.6293, tangentMode=RCTM_User, time=1, value=1}
         ]}}})
     Y=&FloatFromCurve(FloatCurve=DataInterface() {propertyValues {Curve {
         defaultValue=1
         keys [
          {interpMode=RCIM_Cubic, leaveTangentWeight=0.8807, tangentMode=RCTM_User, time=0, value=0.5}
          {arriveTangentWeight=0.3072, interpMode=RCIM_Cubic, tangentMode=RCTM_User, time=0.9213, value=1}
         ]
        }}})
     Z=&FloatFromCurve(FloatCurve=DataInterface() {propertyValues {Curve {
         defaultValue=1
         keys [
          <at 0 0 RCIM_Cubic>
          {arriveTangent=1.4436, arriveTangentWeight=0.3333, interpMode=RCIM_Cubic, leaveTangent=1.4436, tangentMode=RCTM_User, time=1, value=3}
         ]
        }}})
     W=&FloatFromCurve(FloatCurve=DataInterface() {propertyValues {Curve {
         defaultValue=1
         keys [{arriveTangent=3.1821, interpMode=RCIM_Cubic, leaveTangent=3.1821, tangentMode=RCTM_User, time=0, value=0}, <at 1 2 RCIM_Cubic>]
        }}})
    )) {isSetParameters=true}
  ]
  rendererClasses [&NiagaraMeshRendererProperties]
  renderers [
   {rendererClass=&NiagaraMeshRendererProperties, rendererIndex=0}
   {rendererClass=&NiagaraMeshRendererProperties, rendererIndex=1}
  ]
 }
 Shockwave_Air {
  emitterUpdateScript [
   &EmitterState(Life Cycle Mode=::NewEnumerator1, Loop Behavior=::NewEnumerator1, Loop Delay=$System.BurstDelay, Loop Duration=0.35, UseLoopDelay=true)
   &SpawnBurst_Instantaneous(Spawn Count=3, Use Spawn Probability=false, UseLoopCountLimit=false)
   &SpawnRate(SpawnRate=9, Use Spawn Probability=false)
   &SpawnBurst_Instantaneous(Spawn Count=3, Spawn Time=1.0333, Use Spawn Probability=false, UseLoopCountLimit=false) {module=SpawnBurst_Instantaneous001}
  ]
  particleSpawnScript [
   /Niagara/Modules/Spawn/Initialization/InitializeParticle(
    Color=<rgba 1 1 1 6>
    Lifetime=0.2
    Mesh Scale=&RandomRangeVector(Minimum=<1 2 2>, Maximum=<2 2 2>)
    Module.Write Color=true
    Module.Write Lifetime=true
    Module.Write Mass=true
    Module.Write Material Random=true
    Module.Write Position=false
    Module.Write Scale=true
    Module.Write SpriteRotation=false
    Module.Write SpriteSize=true
    Position=&SimulationPosition()
    Sprite Size=<5 5>
   ) {enabled=false}
   /Niagara/Modules/Spawn/Initialization/V2/InitializeParticle(
    Color=<rgba 1 1 1 6>
    Lifetime=0.2
    Mesh Scale Max=<2.2 1.5 1.5>
    Mesh Scale Min=<1.7 1.5 1.5>
    Mesh Scale Mode=::NewEnumerator2
    UsePositionOffset=false
    Write Material Random=true
   ) {module=InitializeParticle001}
   &SystemLocation(MaskSpawnGroup=false, Offset=<0 0 5>)
   &InitialMeshOrientation(
    Rotation=&MakeVector(X=0, Y=0, Z=&RandomRangeFloat(Minimum=0, Maximum=1))
    Use Orientation Vector=true
    Use Rotation Vector=true
   )
   &InitialMeshOrientation(
    Rotation=&RandomRangeVector(Minimum=<0 0 0>, Maximum=<0 0 0.15>)
    Use Orientation Vector=true
    Use Rotation Vector=true
   ) {enabled=false, module=InitialMeshOrientation001}
  ]
  particleUpdateScript [
   &ParticleState()
   &SolveForcesAndVelocity(Clamp Velocity=false, Limit Acceleration=false)
   /SetVariables(Particles.AirColor:LinearColor=&MultiplyLinearColor(
     Linear Color=&ColorFromCurve(
      ColorCurve=DataInterface() {propertyValues {
        AlphaCurve {keys [<at 0 1 RCIM_Cubic>,<at 1 0.1 RCIM_Cubic>]}
        BlueCurve {keys [<at 0 1 RCIM_Cubic>,<at 0.3254 1 RCIM_Cubic>]}
        GreenCurve {keys [<at 0 0.4688 RCIM_Cubic>,<at 0.3254 1 RCIM_Cubic>]}
        RedCurve {keys [<at 0 0.3219 RCIM_Cubic>,<at 0.3254 1 RCIM_Cubic>]}
       }}
      CurveIndex=$Emitter.Age
     )
     Scale Factor=&MakeVector4(
      X=1
      Y=1
      Z=1
      W=&FloatFromCurve(FloatCurve=DataInterface() {propertyValues {Curve {keys [{interpMode=RCIM_Cubic, tangentMode=RCTM_User, time=0, value=1}, <at 1 0>]}}})
     )
    )) {isSetParameters=true}
   &ScaleColor(
    Scale Alpha=&Multiply_Float(
     A=&FN_FrontEndFloatModifier(Gameplay=1, Front End=1)
     B=&Multiply_Float(
      A=&FloatFromCurve(FloatCurve=DataInterface() {propertyValues {Curve {keys [<at 0 0 RCIM_Cubic>,<at 0.026 1 RCIM_Cubic>,<at 1 0>]}}}, Scale Curve=0.5)
      B=&FNTimeOfDayScalar(Night Scale=0.3, Morning Scale=0.2, Evening Scale=0.2)
     )
    )
    Scale RGB=&MakeVectorFromLinearColorRGB(Color=&ColorFromCurve(
      ColorCurve=DataInterface() {propertyValues {
        AlphaCurve {keys [<at 0 1 RCIM_Cubic>,<at 1 0.5 RCIM_Cubic>]}
        BlueCurve {keys [<at 0 1 RCIM_Cubic>,<at 0.6962 0.5 RCIM_Cubic>]}
        GreenCurve {keys [<at 0 0.2504 RCIM_Cubic>,<at 0.6962 0.5 RCIM_Cubic>]}
        RedCurve {keys [<at 0 0.043 RCIM_Cubic>,<at 0.6962 0.5 RCIM_Cubic>]}
       }}
      CurveIndex=$Emitter.Age
     ))
    ScaleA=true
    ScaleRGB=true
   ) {module=ScaleColor001}
   &ScaleMeshSize(Scale Factor=&VectorFromCurve(
     VectorCurve=DataInterface() {propertyValues {
       XCurve {keys [<at 0 0.5>,<at 1 1.5>]}
       YCurve {keys [
         {arriveTangent=1.8533, interpMode=RCIM_Cubic, leaveTangent=1.8533, tangentMode=RCTM_User, time=0, value=0.7}
         {arriveTangent=0.4675, tangentMode=RCTM_User, time=1, value=1.9581}
        ]}
       ZCurve {keys [
         {arriveTangent=1.8533, interpMode=RCIM_Cubic, leaveTangent=1.8533, tangentMode=RCTM_User, time=0, value=0.7}
         {arriveTangent=0.4675, tangentMode=RCTM_User, time=1, value=2.3581}
        ]}
      }}
     Scale Curve=&VectorFromCurve(
      VectorCurve=DataInterface() {propertyValues {
        XCurve {keys [<at 0 1>,<at 1 1>]}
        YCurve {keys [{interpMode=RCIM_Cubic, tangentMode=RCTM_User, time=0, value=1}, <at 1 1.5>]}
        ZCurve {keys [{interpMode=RCIM_Cubic, tangentMode=RCTM_User, time=0, value=1}, <at 1 1.5>]}
       }}
      CurveIndex=$Emitter.Age
     )
    ))
   &DynamicMaterialParameters(
    Index 0 Param 1=&FloatFromCurve(FloatCurve=DataInterface() {propertyValues {Curve {keys [<at 0 0 RCIM_Cubic>,<at 1 0.28 RCIM_Cubic>]}}})
    Index 0 Param 2=&RandomRangeFloat(Minimum=0, Maximum=1)
    Index 0 Param 3=100
    Index 0 Param 4=&FloatFromCurve(FloatCurve=DataInterface() {propertyValues {Curve {keys [<at 0 0>, {arriveTangent=-3.0374, leaveTangent=2.0371, time=0.2, value=0}, {interpMode=RCIM_Cubic, tangentMode=RCTM_Break, time=1, value=0}]}}})
    Module.Param0WriteEnabled=true
    Module.Param1WriteEnabled=true
    Module.Param2WriteEnabled=true
    Module.Param3WriteEnabled=true
   )
  ]
  rendererClasses [&NiagaraMeshRendererProperties]
  renderers [
   {rendererClass=&NiagaraMeshRendererProperties, rendererIndex=0}
   {rendererClass=&NiagaraMeshRendererProperties, rendererIndex=1}
  ]
 }
 Shockwave_Wind {
  emitterUpdateScript [
   &EmitterState(Life Cycle Mode=::NewEnumerator1, Loop Behavior=::NewEnumerator1, Loop Delay=$System.BurstDelay, Loop Duration=0.3, UseLoopDelay=true)
   &SpawnBurst_Instantaneous(Spawn Count=5, Use Spawn Probability=false, UseLoopCountLimit=false)
   &SpawnRate(
    SpawnRate=&FloatFromCurve(
     FloatCurve=DataInterface() {propertyValues {Curve {keys [
         {interpMode=RCIM_Cubic, tangentMode=RCTM_User, time=0, value=25}
         {arriveTangent=-23.6136, interpMode=RCIM_Cubic, leaveTangent=-23.6136, tangentMode=RCTM_User, time=1, value=15}
        ]}}}
     CurveIndex=$Emitter.NormalizedLoopAge
    )
    Use Spawn Probability=false
   )
   &SpawnBurst_Instantaneous(Spawn Count=3, Spawn Group=1, Use Spawn Probability=false, UseLoopCountLimit=false) {module=SpawnBurst_Instantaneous001}
  ]
  particleSpawnScript [
   /SetVariables(Particles.VisibilityTag=&SpawnGroup()) {isSetParameters=true}
   &InitializeParticle(
    Color=<rgba 3.125 3.9093 5 0.3>
    Lifetime Max=0.4
    Lifetime Min=0.3
    Lifetime Mode=::NewEnumerator1
    Mesh Scale=&MergeHelper_VectorSwitch(
     0=&Multiply_VectorByFloat(
      Vector=&MakeVector(X=0.8, Y=0.8, Z=&RandomRangeFloat(Minimum=1, Maximum=2))
      Float=&RandomRangeFloat(Minimum=1, Maximum=1.3)
     )
     1=&RandomRangeVector(Minimum=<1 0.5 0.5>, Maximum=<1 1 1>)
     2=?
     3=?
     Index=&SpawnGroup()
     Num=::NewEnumerator0
    )
    Mesh Scale Mode=::NewEnumerator4
    UsePositionOffset=false
    Write Material Random=true
   )
   &SystemLocation(
    MaskSpawnGroup=false
    Offset=&MergeHelper_VectorSwitch(0=<0 0 50>, 1=<0 0 0>, 2=?, 3=?, Index=&SpawnGroup(), Num=::NewEnumerator0)
   )
   &InitialMeshOrientation(
    Mesh Orientation Mode=::NewEnumerator2
    Orientation Coordinate Space=::Local
    Rotation=&MergeHelper_VectorSwitch(
     0=&RandomRangeVector(Minimum=<0.05 -0.1 0>, Maximum=<-0.05 0.1 1>)
     1=&RandomRangeVector(Minimum=<0 0 0>, Maximum=<0 0 1>)
     2=?
     3=?
     Index=&SpawnGroup()
     Num=::NewEnumerator0
    )
    Use Rotation Vector=true
   )
   &ShapeLocation(MaskSpawnGroup=false, Override Seed=false, Ring Radius=5, Shape Primitive=::NewEnumerator4, UseOwnerScale=true)
  ]
  particleUpdateScript [
   &ParticleState()
   &ScaleMeshSize(Scale Factor=&Multiply_VectorByFloat(
     Vector=&Multiply_VectorByFloat(
      Vector=&MakeVector(
       X=&FloatFromCurve(FloatCurve=DataInterface() {propertyValues {Curve {keys [{arriveTangent=1.09, interpMode=RCIM_Cubic, leaveTangent=1.09, tangentMode=RCTM_User, time=0, value=1}, <at 1 1.5>]}}})
       Y=&FloatFromCurve(FloatCurve=DataInterface() {propertyValues {Curve {keys [{arriveTangent=1.0817, interpMode=RCIM_Cubic, leaveTangent=1.0817, tangentMode=RCTM_User, time=0, value=1}, <at 1 1.5>]}}})
       Z=&FloatFromCurve(
        FloatCurve=DataInterface() {propertyValues {Curve {keys [{interpMode=RCIM_Cubic, tangentMode=RCTM_User, time=0, value=1.5}, {tangentMode=RCTM_User, time=1, value=1}]}}}
        Scale Curve=&FloatFromCurve(
         FloatCurve=DataInterface() {propertyValues {Curve {keys [{interpMode=RCIM_Cubic, tangentMode=RCTM_User, time=0, value=1.5}, {arriveTangent=-0.1559, tangentMode=RCTM_User, time=0.5, value=0.5}]}}}
         CurveIndex=$Emitter.Age
         Scale Curve=&RandomRangeFloat(Minimum=0.8, Maximum=1)
        )
       )
      )
      Float=0.8
     )
     Float=&FloatFromCurve(
      FloatCurve=DataInterface() {propertyValues {Curve {keys [
          {interpMode=RCIM_Cubic, tangentMode=RCTM_User, time=0, value=0.3}
          {arriveTangent=0.7, interpMode=RCIM_Cubic, leaveTangent=0.7, time=0.453, value=0.6985}
          {interpMode=RCIM_Cubic, tangentMode=RCTM_User, time=1, value=1}
         ]}}}
      Scale Curve=3
     )
    ))
   &UpdateMeshOrientation(
    Rotation Rate=&MergeHelper_FloatSwitch(
     0=&FloatFromCurve(
      FloatCurve=DataInterface() {propertyValues {Curve {keys [{arriveTangent=-2.1843, interpMode=RCIM_Cubic, leaveTangent=-2.1843, tangentMode=RCTM_User, time=0, value=1}, <at 1 0>]}}}
      Scale Curve=15
     )
     1=0
     2=?
     3=?
     Index=&SpawnGroup(Max Spawn Group=3)
     Num=::NewEnumerator0
    )
    Rotation Vector=<0 0 1>
   )
   &ScaleColor(
    Scale Alpha=&Multiply_Float(
     A=&FloatFromCurve(FloatCurve=DataInterface() {propertyValues {Curve {keys [
          {interpMode=RCIM_Cubic, tangentMode=RCTM_User, time=0, value=0}
          {interpMode=RCIM_Cubic, tangentMode=RCTM_User, time=0.1, value=1}
          {tangentMode=RCTM_User, time=1, value=0}
         ]}}})
     B=0.2
    )
    Scale RGB=&Multiply_VectorByFloat(
     Vector=&Multiply_VectorByFloat(
      Vector=&MakeVectorFromLinearColorRGB(Color=&ColorFromCurve(
        ColorCurve=DataInterface() {propertyValues {
          AlphaCurve {defaultValue=0, keys []}
          BlueCurve {keys [<at 1.25 2 RCIM_Cubic>,<at 2 1>]}
          GreenCurve {keys [<at 1.25 1 RCIM_Cubic>,<at 2 1>]}
          RedCurve {keys [<at 1.25 0.43 RCIM_Cubic>,<at 2 1>]}
         }}
        CurveIndex=$Emitter.Age
       ))
      Float=&FN_FrontEndFloatModifier(Gameplay=0.7, Front End=0.7)
     )
     Float=$System.TOD
    )
    ScaleA=true
    ScaleRGB=true
   )
   /SetVariables(Particles.GroundDynP00:Vector4f=&MakeVector4(
     X=&FloatFromCurve(FloatCurve=DataInterface() {propertyValues {Curve {keys [
          <at 0.5 0 RCIM_Cubic>
          {arriveTangent=2.6293, arriveTangentWeight=0.1667, interpMode=RCIM_Cubic, leaveTangent=2.6293, tangentMode=RCTM_User, time=1, value=1}
         ]}}})
     Y=&FloatFromCurve(FloatCurve=DataInterface() {propertyValues {Curve {
         defaultValue=1
         keys [
          {arriveTangentWeight=0.3072, interpMode=RCIM_Cubic, leaveTangentWeight=0.3071, tangentMode=RCTM_User, time=0.0787, value=1}
          {arriveTangentWeight=0.3071, interpMode=RCIM_Cubic, leaveTangentWeight=0.8807, tangentMode=RCTM_User, time=1, value=0.5}
         ]
        }}})
     Z=&FloatFromCurve(FloatCurve=DataInterface() {propertyValues {Curve {
         defaultValue=1
         keys [
          {
           arriveTangent=-1.4436
           arriveTangentWeight=0.3333
           interpMode=RCIM_Cubic
           leaveTangent=-1.4436
           leaveTangentWeight=0.5854
           tangentMode=RCTM_User
           time=0
           value=3
          }
          {arriveTangentWeight=0.3333, interpMode=RCIM_Cubic, tangentMode=RCTM_User, time=1, value=0}
         ]
        }}})
     W=&FloatFromCurve(FloatCurve=DataInterface() {propertyValues {Curve {
         defaultValue=1
         keys [{arriveTangent=3.1821, interpMode=RCIM_Cubic, leaveTangent=3.1821, tangentMode=RCTM_User, time=0, value=0}, <at 1 2 RCIM_Cubic>]
        }}})
    )) {isSetParameters=true}
  ]
  rendererClasses [&NiagaraMeshRendererProperties]
  renderers [
   {rendererClass=&NiagaraMeshRendererProperties, rendererIndex=0}
   {rendererClass=&NiagaraMeshRendererProperties, rendererIndex=1}
  ]
 }
}
summary {
 systemName=NS_Emote_DuneBrief_Shot_WindBurst
 emitters [
  {emitterName=Shockwave_Wind, rendererClasses [&NiagaraMeshRendererProperties]}
  {emitterName=Shockwave_Air, rendererClasses [&NiagaraMeshRendererProperties]}
  {emitterName=Inward_Wind, rendererClasses [&NiagaraMeshRendererProperties]}
  {emitterName=Burst_spikes, rendererClasses [&NiagaraSpriteRendererProperties, &NiagaraLightRendererProperties]}
  {emitterName=Glow_Body, rendererClasses [&NiagaraMeshRendererProperties]}
  {
   bEnabled=false
   emitterName=Burst_spikes001
   rendererClasses [&NiagaraSpriteRendererProperties, &NiagaraLightRendererProperties]
  }
 ]
 userVariables [
  User.bIsFrontEnd {type=NiagaraBool}
  User.bIsFrontEndPreview {type=NiagaraBool}
 ]
}
