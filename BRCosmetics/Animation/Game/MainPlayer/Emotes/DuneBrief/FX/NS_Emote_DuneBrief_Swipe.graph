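# Niagara effect for the swipe in the DuneBrief player emote: a blue glowing spark with a short fading trail that follows the character's skeletal mesh.
# Core bursts short-lived sprites pinned to the mesh each frame, with three renderers and curves that rotate, flicker, then flare size and brightness before expiry.
# Ribbon continuously spawns trail points at the same mesh offset, with two renderers, derived dark color and front-end-preview-dependent alpha.
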NiagaraSystem /BRCosmetics/Animation/Game/MainPlayer/Emotes/DuneBrief/FX/NS_Emote_DuneBrief_Swipe
dependencies {
 usedDataInterfaces [&NiagaraDataInterfaceCurve, &NiagaraDataInterfaceSkeletalMesh]
 usedDynamicInputs []
 usedModules [
  &SystemState
  &self:SystemSpawnScript.SystemScriptSource.SystemScriptGraph.NiagaraNodeAssignment_0.SetVariables_#1
 ]
}
emitters {
 Core {
  emitterUpdateScript [
   &EmitterState()
   &SpawnBurst_Instantaneous(Use Spawn Probability=false, UseLoopCountLimit=false)
   &SpawnRate(SpawnGroup=1, SpawnRate=50, Use Spawn Probability=false) {enabled=false}
  ]
  particleSpawnScript [
   &InitializeParticle(
    Color=<rgba 0.7 2.2977 7 1>
    Lifetime=0.47
    Ribbon Width Mode=::NewEnumerator1
    Sprite Size Mode=::NewEnumerator3
    Uniform Sprite Size=75
    UsePositionOffset=false
    Write Material Random=true
   )
   &AddVelocity() {enabled=false}
  ]
  particleUpdateScript [
   &ParticleState()
   &SpriteRotationRate(Rotation Rate=360) {enabled=false}
   &ScaleSpriteSize(
    Scale Sprite Size Mode=::NewEnumerator0
    Uniform Scale Factor=&FloatFromCurve(FloatCurve=DataInterface() {propertyValues {Curve {
        defaultValue=1
        keys [
         <at 0 0 RCIM_Cubic>
         {arriveTangentWeight=0.1, interpMode=RCIM_Cubic, leaveTangentWeight=0.1498, tangentMode=RCTM_User, time=0.2, value=1}
         {arriveTangent=0.5883, interpMode=RCIM_Cubic, leaveTangent=0.5883, time=0.235, value=0.6618}
         {arriveTangent=1.6357, interpMode=RCIM_Cubic, leaveTangent=1.6357, time=0.281, value=1.0476}
         {arriveTangent=-0.1974, interpMode=RCIM_Cubic, leaveTangent=-0.1974, time=0.3155, value=0.7935}
         {arriveTangent=-0.3881, interpMode=RCIM_Cubic, leaveTangent=-0.3881, time=0.3763, value=1.0288}
         {arriveTangent=0.6481, interpMode=RCIM_Cubic, leaveTangent=0.6481, time=0.4125, value=0.7559}
         {arriveTangent=0.4404, interpMode=RCIM_Cubic, leaveTangent=0.4404, time=0.4635, value=1.0853}
         {arriveTangent=-0.8642, interpMode=RCIM_Cubic, leaveTangent=-0.8642, time=0.498, value=0.7935}
         {arriveTangent=-0.5205, interpMode=RCIM_Cubic, leaveTangent=-0.5205, time=0.5506, value=1.01}
         {arriveTangent=5.4974, interpMode=RCIM_Cubic, leaveTangent=5.4974, time=0.5884, value=0.7465}
         {arriveTangent=1.7566, interpMode=RCIM_Cubic, leaveTangent=1.7566, time=0.6266, value=1.4277}
         {arriveTangent=3.4807, interpMode=RCIM_Cubic, leaveTangent=3.4807, time=0.6698, value=0.8894}
         {arriveTangent=0.265, interpMode=RCIM_Cubic, leaveTangent=0.265, time=0.7056, value=1.7026}
         {arriveTangent=6.3703, interpMode=RCIM_Cubic, leaveTangent=6.3703, time=0.7469, value=0.9098}
         {
          arriveTangent=1.7175
          arriveTangentWeight=0.1571
          interpMode=RCIM_Cubic
          leaveTangent=1.7175
          leaveTangentWeight=0.0875
          tangentMode=RCTM_User
          time=0.7854
          value=2.2113
         }
         {arriveTangent=5.5533, interpMode=RCIM_Cubic, leaveTangent=5.5533, time=0.8354, value=0.1}
         {arriveTangent=57.7404, interpMode=RCIM_Cubic, leaveTangent=57.7404, time=0.8692, value=2.6768}
         {
          arriveTangent=33.5403
          arriveTangentWeight=0.4371
          interpMode=RCIM_Cubic
          leaveTangent=33.5403
          leaveTangentWeight=0.6591
          tangentMode=RCTM_User
          time=0.9116
          value=4.5
         }
         {arriveTangentWeight=0.7032, interpMode=RCIM_Cubic, leaveTangentWeight=0.2971, tangentMode=RCTM_User, time=0.9779, value=6}
         <at 1 1>
        ]
       }}})
   )
   &ScaleColor(
    Scale Alpha=&FloatFromCurve(FloatCurve=DataInterface() {propertyValues {Curve {keys [<at 0.9116 1>,<at 1 0>]}}})
    Scale RGB=&Multiply_VectorByFloat(
     Vector=<1 1 1>
     Float=&FloatFromCurve(FloatCurve=DataInterface() {propertyValues {Curve {keys [{interpMode=RCIM_Cubic, tangentMode=RCTM_User, time=0.93, value=1}, <at 1 50>]}}})
    )
    ScaleA=false
    ScaleRGB=true
   )
   /SetVariables(
    Particles.RoundSparkColor:LinearColor=&MultiplyLinearColor(Linear Color=$Particles.Color, Scale Factor=&MakeVector4(X=1, Y=1, Z=2, W=1))
    Particles.RoundSparkGlowColor:LinearColor=&MultiplyLinearColor(Linear Color=$Particles.Color, Scale Factor=&MakeVector4(X=0.01, Y=0.01, Z=0.1, W=10))
    Particles.RoundSparkSize:Vector2f=&Multiply_Vector2D_ByFloat(
     Float=&FloatFromCurve(
      FloatCurve=DataInterface() {propertyValues {Curve {keys [
          {arriveTangent=1.9053, interpMode=RCIM_Cubic, leaveTangent=1.9053, tangentMode=RCTM_User, time=0, value=1}
          {arriveTangent=-2.2319, interpMode=RCIM_Cubic, leaveTangent=-2.2319, time=0.0949, value=1.0902}
          {arriveTangent=-0.026, interpMode=RCIM_Cubic, leaveTangent=-0.026, tangentMode=RCTM_User, time=0.1378, value=0.6924}
          {arriveTangent=-0.0642, interpMode=RCIM_Cubic, leaveTangent=-0.0642, time=0.2455, value=1.0846}
          {arriveTangent=-0.113, interpMode=RCIM_Cubic, leaveTangent=-0.113, tangentMode=RCTM_User, time=0.3063, value=0.6816}
          {arriveTangent=0.0572, interpMode=RCIM_Cubic, leaveTangent=0.0572, time=0.4374, value=1.0675}
          {arriveTangent=0.0044, interpMode=RCIM_Cubic, leaveTangent=0.0044, tangentMode=RCTM_User, time=0.5078, value=0.6932}
          {arriveTangent=-0.0847, interpMode=RCIM_Cubic, leaveTangent=-0.0847, time=0.6277, value=1.0734}
          {arriveTangent=-0.0013, interpMode=RCIM_Cubic, leaveTangent=-0.0013, tangentMode=RCTM_User, time=0.7058, value=0.6764}
          {arriveTangent=0.0607, interpMode=RCIM_Cubic, leaveTangent=0.0607, time=0.8481, value=1.079}
          {arriveTangent=-0.5329, interpMode=RCIM_Cubic, leaveTangent=-0.5329, tangentMode=RCTM_User, time=0.8899, value=0.6876}
          <at 1 1>
         ]}}}
      Scale Curve=0.2
     )
     Vector2D=$Particles.SpriteSize
    )
   ) {isSetParameters=true}
   /SetVariables(Particles.Star DynP 00:Vector4f=&MakeVector4(
     X=&FloatFromCurve(FloatCurve=DataInterface() {propertyValues {Curve {keys [
          {interpMode=RCIM_Cubic, tangentMode=RCTM_User, time=0, value=1.5}
          {interpMode=RCIM_Cubic, tangentMode=RCTM_User, time=0.2866, value=3}
          {interpMode=RCIM_Cubic, tangentMode=RCTM_User, time=1, value=3}
         ]}}})
     Y=0
     Z=0
     W=0
    )) {isSetParameters=true}
   &SolveForcesAndVelocity(Clamp Velocity=false, Limit Acceleration=false)
   /SetVariables(Particles.SpriteRotation:NiagaraFloat=&FloatFromCurve(
     FloatCurve=DataInterface() {propertyValues {Curve {keys [<at 0 1 RCIM_Cubic>, {arriveTangent=0.141, interpMode=RCIM_Cubic, leaveTangent=0.141, tangentMode=RCTM_User, time=0.3, value=1}, <at 0.9 0>]}}}
     Scale Curve=-135
    )) {isSetParameters=true}
   &SkeletalMeshLocation(
    MaskSpawnGroup=false
    Mesh Sampling Type=::NewEnumerator5
    OverrideRandomSeed=false
    Sampled Position Offset=<-7 2 0>
    Skeletal Mesh=$System.Skeletal Mesh
   )
  ]
  rendererClasses [&NiagaraSpriteRendererProperties]
  renderers [
   {rendererClass=&NiagaraSpriteRendererProperties}
   {rendererClass=&NiagaraSpriteRendererProperties, rendererIndex=1}
   {rendererClass=&NiagaraSpriteRendererProperties, rendererIndex=2}
  ]
 }
 Ribbon {
  emitterUpdateScript [
   &EmitterState()
   &SpawnRate(SpawnRate=50, Use Spawn Probability=false)
  ]
  particleSpawnScript [
   &InitializeParticle(
    Color=<rgba 1 1 1 0.8>
    Lifetime=0.25
    Ribbon Width=8
    Ribbon Width Mode=::NewEnumerator1
    UsePositionOffset=false
    Write Material Random=true
   )
   &SkeletalMeshLocation(
    MaskSpawnGroup=false
    Mesh Sampling Type=::NewEnumerator5
    OverrideRandomSeed=false
    Sampled Position Offset=<-7 2 0>
    Skeletal Mesh=$System.Skeletal Mesh
   )
   &AddVelocity() {enabled=false}
   &ShapeLocation(MaskSpawnGroup=false, Override Seed=false, Sphere Radius=0, UseOwnerScale=true)
  ]
  particleUpdateScript [
   &ParticleState()
   /SetVariables(Particles.BaseColor:LinearColor=&MakeLinearColorFromVectorAndFloat(
     "Vector (RGB)"=&MakeVectorFromLinearColorRGB(Color=<rgba 5.8906 8.679 15 1>)
     "Float (Alpha)"=&FloatFromCurve(
      FloatCurve=DataInterface() {propertyValues {Curve {keys [
          {arriveTangent=0.3462, interpMode=RCIM_Cubic, leaveTangent=0.3462, tangentMode=RCTM_User, time=0, value=1}
          {arriveTangent=-1.7131, interpMode=RCIM_Cubic, leaveTangent=-1.7131, tangentMode=RCTM_User, time=0.2547, value=0.4}
          {arriveTangent=-0.0168, interpMode=RCIM_Cubic, tangentMode=RCTM_Break, time=0.7, value=0}
         ]}}}
      Scale Curve=&Multiply_Float(A=&MakeCustomFloatFromBool(Bool=$User.bIsFrontEndPreview, True Float=0.8, False Float=0.75), B=1)
     )
    )) {isSetParameters=true}
   /SetVariables(
    Particles.DarkColor:LinearColor=&MultiplyLinearColor(Linear Color=$Particles.BaseColor, Scale Factor=<0.05 0.05 0.1 1>)
    Particles.DarkWidth=10
   ) {isSetParameters=true}
   &SolveForcesAndVelocity(Clamp Velocity=false, Limit Acceleration=false) {enabled=false}
   &SkeletalMeshLocation(
    Bone Sampling Mode=::NewEnumerator4
    MaskSpawnGroup=false
    Position Sampling=::NewEnumerator2
    Sampled Position Offset=<-15 5 0>
    Skeletal Mesh=DataInterface() {propertyValues {
      DefaultMesh=null
      ExcludeBoneName=root
      OptionalComponentClass=null
      PreviewMesh=null
      SkinningMode=SkinOnTheFly
      SoftSourceActor=null
      SourceMode=Default
      UvSetIndex=0
      WholeMeshLOD=-1
      bExcludeBone=false
      bReadDeformedGeometry=false
      bRequireCurrentFrameData=true
      ComponentTags []
      FilteredBones [weapon_r]
      FilteredSockets []
      MeshUserParameter {parameter {name=None, typeDefHandle {registeredTypeIndex=1}, varData []}}
      SamplingRegions []
     }}
   ) {enabled=false, module=SkeletalMeshLocation001}
   /SetVariables(Particles.Position=&Lerp_Vector(
     A=Linked() {linkedVariable {name=Engine.Emitter.SimulationPosition, type=NiagaraPosition}}
     B=Linked() {linkedVariable {name=Particles.Position, type=NiagaraPosition}}
     Alpha=&FloatFromCurve(FloatCurve=DataInterface() {propertyValues {Curve {
         defaultValue=1
         keys [{interpMode=RCIM_Cubic, tangentMode=RCTM_User, time=0, value=0}, {interpMode=RCIM_Cubic, tangentMode=RCTM_User, time=0.1, value=1}]
        }}})
    )) {enabled=false, isSetParameters=true}
  ]
  rendererClasses [&NiagaraRibbonRendererProperties]
  renderers [
   {rendererClass=&NiagaraRibbonRendererProperties}
   {rendererClass=&NiagaraRibbonRendererProperties, rendererIndex=1}
  ]
 }
}
summary {
 systemName=NS_Emote_DuneBrief_Swipe
 emitters [
  {emitterName=Ribbon, rendererClasses [&NiagaraRibbonRendererProperties]}
  {emitterName=Core, rendererClasses [&NiagaraSpriteRendererProperties]}
 ]
 userVariables [
  User.bIsFrontEnd {type=NiagaraBool}
  User.bIsFrontEndPreview {type=NiagaraBool}
  User.RibbonIDOffset {type=NiagaraInt32}
 ]
}
